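# Pre-dynamics Control Rig for the GeminiLink female body, posing jacket flap, collar strap, leg belt and ankle bones from body motion.
# Parent constraints blend base bones between body bones, and a shared swing/twist DeltaRotation function turns joint rotations into 0-1 weights.
# Those weights, a spherical pose reader and gameplay variables (crouch, slides, strafe, rail grind, pickaxe swing) drive additive ModifyTransforms correctives.
# The main graph references a DeformRig function.
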ControlRigBlueprint /BRCosmetics/Characters/Player/Female/Medium/Bodies/F_MED_GeminiLink/Meshes/F_MED_GeminiLink_PREDynamics_CtrlRig
graphs [
 New Function_ContainedGraph {nodes [
   RigVMFunctionEntryNode()
   > DeltaRotation v4 = RigVMCollapseNode(
    Rotation Order=ZYX
    child=(Type=Bone,Name="spine_04")
    debugRotation=false
    outputMinMax=(X=0,Y=1)
    parent=(Type=Null,Name="spine_04_Spine03_DeltaRef_null")
    premult=1
    rotationMinMax=(X=0,Y=30)
    triggerAxis=Z
    twistAxis=(X=1,Y=0,Z=0)
   )
   > &RigUnit_ModifyTransforms(
    ItemToModify="((Item=(Type=Bone,Name=\"base_jacketFlap_01_F_l\"),Transform=(Rotation=(X=0.0114,Y=-0.0864,Z=0.13,W=0.9877),Translation=(X=0,Y=0,Z=0),Scale3D=(X=1,Y=1,Z=1))),(Item=(Type=Bone,Name=\"base_jacketFlap_01_S_l\"),Transform=(Rotation=(X=-0.052,Y=0.0331,Z=0.2179,W=0. ...(truncated)"
    WeightMinimum=0
    Weight=@"DeltaRotation v4".outputWeight
   )
   > DeltaRotation v4_1 = RigVMCollapseNode(
    Rotation Order=ZYX
    child=(Type=Bone,Name="spine_04")
    debugRotation=false
    outputMinMax=(X=0,Y=1)
    parent=(Type=Null,Name="spine_04_Spine03_DeltaRef_null")
    premult=1
    rotationMinMax=(X=0,Y=-30)
    triggerAxis=Z
    twistAxis=(X=1,Y=0,Z=0)
   )
   > &RigUnit_ModifyTransforms(
    ItemToModify="((Item=(Type=Bone,Name=\"base_jacketFlap_01_F_l\"),Transform=(Rotation=(X=0.0114,Y=0.0864,Z=-0.13,W=0.9877),Translation=(X=0,Y=0,Z=0),Scale3D=(X=1,Y=1,Z=1))),(Item=(Type=Bone,Name=\"base_jacketFlap_01_S_l\"),Transform=(Rotation=(X=0.052,Y=0.0331,Z=-0.2179,W=0. ...(truncated)"
    WeightMinimum=0
    Weight=@"DeltaRotation v4_1".outputWeight
   )
   > RigVMFunctionReturnNode()
  ]}
 New Function_ContainedGraph {nodes [
   n1 = RigVMFunctionEntryNode()
   RigVMFunction_MathBoolToInteger = &RigVMFunction_MathBoolToInteger(Value=@n1.onlyTwist)
   RigVMFunction_MathQuaternionSwingTwist = &RigVMFunction_MathQuaternionSwingTwist(
    TwistAxis=@n1.twistAxis
    Input=&RigUnit_GetRelativeTransformForItem(bChildInitial=False, bParentInitial=False, Child=@n1.child, Parent=@n1.parent).RelativeTransform.Rotation
   )
   n2 = RigVMDispatchNode(
    Values=((),())
    Index=@RigVMFunction_MathBoolToInteger.Result
    Values.0=@RigVMFunction_MathQuaternionSwingTwist.Swing
    Values.1=@RigVMFunction_MathQuaternionSwingTwist.Twist
   ) {template="DISPATCH_RigVMDispatch_SelectInt32(in Index,in Values,out Result)"}
   RigVMFunction_MathQuaternionToEuler = &RigVMFunction_MathQuaternionToEuler(Value=@n2.Result, RotationOrder=@n1.Rotation Order)
   > DISPATCH_RigVMDispatch_Print = RigVMDispatchNode(ScreenColor=(R=1,G=1,B=1,A=1), ScreenDuration=1.5, Enabled=@n1.debugRotation, Value=@RigVMFunction_MathQuaternionToEuler.Result) {template="DISPATCH_RigVMDispatch_Print(in Prefix,in Value,in Enabled,in ScreenDuration,in ScreenColor)"}
   Cast = RigVMDispatchNode(Value=@n1.triggerAxis) {template="DISPATCH_RigVMDispatch_CastEnumToInt(in Value,out Result)"}
   DISPATCH_RigVMDispatch_SelectInt32 = RigVMDispatchNode(
    Values=(0,0,0)
    Index=@Cast.Result
    Values.0=@RigVMFunction_MathQuaternionToEuler.Result.X
    Values.1=@RigVMFunction_MathQuaternionToEuler.Result.Y
    Values.2=@RigVMFunction_MathQuaternionToEuler.Result.Z
   ) {template="DISPATCH_RigVMDispatch_SelectInt32(in Index,in Values,out Result)"}
   Remap = &RigVMFunction_MathDoubleRemap(
    bClamp=False
    Value=@DISPATCH_RigVMDispatch_SelectInt32.Result
    TargetMinimum=@n1.outputMinMax.X
    TargetMaximum=@n1.outputMinMax.Y
    SourceMinimum=@n1.rotationMinMax.X
    SourceMaximum=@n1.rotationMinMax.Y
   )
   Absolute = &RigVMFunction_MathDoubleAbs(Value=@Remap.Result)
   Multiply = &RigVMFunction_MathDoubleMul(
    A=&RigVMFunction_MathDoubleClamp(Value=RigVMDispatchNode(True=@Absolute.Result, Condition=@n1.absoluteValue, False=@Remap.Result) {template="DISPATCH_RigVMDispatch_If(in Condition,in True,in False,out Result)"}.Result).Result
    B=@n1.premult
   )
   > RigVMFunctionReturnNode(outputQuat=@n2.Result, outputWeight=@Multiply.Result)
  ]}
 New Function_ContainedGraph {nodes [
   n1 = RigVMFunctionEntryNode()
   RigVMFunction_MathBoolToInteger = &RigVMFunction_MathBoolToInteger(Value=@n1.onlyTwist)
   RigVMFunction_MathQuaternionSwingTwist = &RigVMFunction_MathQuaternionSwingTwist(
    TwistAxis=@n1.twistAxis
    Input=&RigUnit_GetRelativeTransformForItem(bChildInitial=False, bParentInitial=False, Child=@n1.child, Parent=@n1.parent).RelativeTransform.Rotation
   )
   n2 = RigVMDispatchNode(
    Values=((),())
    Index=@RigVMFunction_MathBoolToInteger.Result
    Values.0=@RigVMFunction_MathQuaternionSwingTwist.Swing
    Values.1=@RigVMFunction_MathQuaternionSwingTwist.Twist
   ) {template="DISPATCH_RigVMDispatch_SelectInt32(in Index,in Values,out Result)"}
   RigVMFunction_MathQuaternionToEuler = &RigVMFunction_MathQuaternionToEuler(Value=@n2.Result, RotationOrder=@n1.Rotation Order)
   > DISPATCH_RigVMDispatch_Print = RigVMDispatchNode(ScreenColor=(R=1,G=1,B=1,A=1), ScreenDuration=1.5, Enabled=@n1.debugRotation, Value=@RigVMFunction_MathQuaternionToEuler.Result) {template="DISPATCH_RigVMDispatch_Print(in Prefix,in Value,in Enabled,in ScreenDuration,in ScreenColor)"}
   Cast = RigVMDispatchNode(Value=@n1.triggerAxis) {template="DISPATCH_RigVMDispatch_CastEnumToInt(in Value,out Result)"}
   DISPATCH_RigVMDispatch_SelectInt32 = RigVMDispatchNode(
    Values=(0,0,0)
    Index=@Cast.Result
    Values.0=@RigVMFunction_MathQuaternionToEuler.Result.X
    Values.1=@RigVMFunction_MathQuaternionToEuler.Result.Y
    Values.2=@RigVMFunction_MathQuaternionToEuler.Result.Z
   ) {template="DISPATCH_RigVMDispatch_SelectInt32(in Index,in Values,out Result)"}
   Remap = &RigVMFunction_MathDoubleRemap(
    bClamp=False
    Value=@DISPATCH_RigVMDispatch_SelectInt32.Result
    TargetMinimum=@n1.outputMinMax.X
    TargetMaximum=@n1.outputMinMax.Y
    SourceMinimum=@n1.rotationMinMax.X
    SourceMaximum=@n1.rotationMinMax.Y
   )
   Absolute = &RigVMFunction_MathDoubleAbs(Value=@Remap.Result)
   Multiply = &RigVMFunction_MathDoubleMul(
    A=&RigVMFunction_MathDoubleClamp(Value=RigVMDispatchNode(True=@Absolute.Result, Condition=@n1.absoluteValue, False=@Remap.Result) {template="DISPATCH_RigVMDispatch_If(in Condition,in True,in False,out Result)"}.Result).Result
    B=@n1.premult
   )
   > RigVMFunctionReturnNode(outputQuat=@n2.Result, outputWeight=@Multiply.Result)
  ]}
 New Function_ContainedGraph {nodes [
   RigVMFunctionEntryNode()
   > n1 = &RigUnit_SphericalPoseReader(
    ActiveRegionScaleFactors=(PositiveWidth=0.2,NegativeWidth=0.3,PositiveHeight=0.2,NegativeHeight=0.2)
    ActiveRegionSize=0.1
    Debug=(bDrawDebug=false,bDraw2D=False,bDrawLocalAxes=False,DebugScale=25,DebugSegments=20,DebugThickness=0.25)
    DriverAxis=(X=1,Y=0,Z=0)
    DriverItem=(Type=Bone,Name="upperarm_l")
    FalloffRegionScaleFactors=(PositiveWidth=1,NegativeWidth=1,PositiveHeight=1,NegativeHeight=1)
    FalloffSize=0.2
    FlipHeightScaling=False
    FlipWidthScaling=False
    OptionalParentItem=(Type=Bone,Name="None")
    RotationOffset=(X=90,Y=0,Z=30)
   )
   > &RigUnit_ModifyTransforms(
    ItemToModify=((Item=(Type=Bone,Name="base_collarStrap_01_l"),Transform=(Rotation=(X=-0.0308,Y=-0.0534,Z=0.2565,W=0.9646),Translation=(X=0,Y=0.5,Z=0),Scale3D=(X=1,Y=1,Z=1))))
    WeightMinimum=0
    Weight=@n1.OutputParam
   )
   n2 = &RigVMFunction_AnimEvalRichCurve(
    Curve=(EditorCurveData=(Keys=((InterpMode=RCIM_Cubic),(InterpMode=RCIM_Cubic,Time=0.5777,Value=0.3237,ArriveTangent=1,LeaveTangent=1),(InterpMode=RCIM_Cubic,TangentMode=RCTM_User,Time=1,Value=1,ArriveTangent=1.65,LeaveTangent=1.65))))
    SourceMinimum=0
    TargetMinimum=0
    Value=@n1.OutputParam
   )
   > &RigUnit_ModifyTransforms(
    ItemToModify=((Item=(Type=Bone,Name="dyn_collarStrap_02_l"),Transform=(Rotation=(X=0,Y=0,Z=0.1736,W=0.9848),Translation=(X=0,Y=0.5,Z=0),Scale3D=(X=1,Y=1,Z=1))))
    WeightMinimum=0
    Weight=@n2.Result
   )
   > RigVMFunctionReturnNode()
  ]}
 New Function_ContainedGraph {nodes [
   n1 = RigVMFunctionEntryNode()
   RigVMFunction_Sequence = &RigVMFunction_Sequence(ExecuteContext=@n1.ExecuteContext)
   &RigUnit_ParentConstraint(
    AdvancedSettings=(RotationOrderForFilter=XZY)
    Child=(Type=Bone,Name="base_jacketFlap_01_F_l")
    Parents="((Item=(Type=Bone,Name=\"upperarm_twist_01_l\"),Weight=0.0052),(Item=(Type=Bone,Name=\"clavicle_l\"),Weight=0.0265),(Item=(Type=Bone,Name=\"spine_03\"),Weight=0.0374),(Item=(Type=Bone,Name=\"spine_02\"),Weight=0.0415),(Item=(Type=Bone,Name=\"spine_05\"),Weight=0.046 ...(truncated)"
    Weight=1
    ExecutePin=@RigVMFunction_Sequence.A
   )
   > &RigUnit_ParentConstraint(
    AdvancedSettings=(RotationOrderForFilter=XZY)
    Child=(Type=Bone,Name="base_jacketFlap_01_S_l")
    Parents="((Item=(Type=Bone,Name=\"spine_01\"),Weight=0.0155),(Item=(Type=Bone,Name=\"spine_03\"),Weight=0.0667),(Item=(Type=Bone,Name=\"spine_02\"),Weight=0.0771),(Item=(Type=Bone,Name=\"upperarm_twist_01_l\"),Weight=0.0857),(Item=(Type=Bone,Name=\"clavicle_l\"),Weight=0.109 ...(truncated)"
    Weight=1
   )
   > &RigUnit_ParentConstraint(
    AdvancedSettings=(RotationOrderForFilter=XZY)
    Child=(Type=Bone,Name="base_jacketFlap_01_SB_l")
    Parents="((Item=(Type=Bone,Name=\"clavicle_r\"),Weight=0.0125),(Item=(Type=Bone,Name=\"upperarm_twist_01_l\"),Weight=0.0291),(Item=(Type=Bone,Name=\"spine_01\"),Weight=0.0471),(Item=(Type=Bone,Name=\"spine_05\"),Weight=0.119),(Item=(Type=Bone,Name=\"clavicle_l\"),Weight=0.12 ...(truncated)"
    Weight=1
   )
   > &RigUnit_ParentConstraint(
    AdvancedSettings=(RotationOrderForFilter=XZY)
    Child=(Type=Bone,Name="base_jacketFlap_01_B")
    Parents="((Item=(Type=Bone,Name=\"pelvis\"),Weight=0.0121),(Item=(Type=Bone,Name=\"spine_01\"),Weight=0.0541),(Item=(Type=Bone,Name=\"clavicle_l\"),Weight=0.0601),(Item=(Type=Bone,Name=\"clavicle_r\"),Weight=0.0601),(Item=(Type=Bone,Name=\"spine_03\"),Weight=0.1405),(Item=(T ...(truncated)"
    Weight=1
   )
   > &RigUnit_ParentConstraint(
    AdvancedSettings=(RotationOrderForFilter=XZY)
    Child=(Type=Bone,Name="base_jacketFlap_01_SB_r")
    Parents="((Item=(Type=Bone,Name=\"clavicle_l\"),Weight=0.0125),(Item=(Type=Bone,Name=\"upperarm_twist_01_r\"),Weight=0.0291),(Item=(Type=Bone,Name=\"spine_01\"),Weight=0.0471),(Item=(Type=Bone,Name=\"spine_05\"),Weight=0.119),(Item=(Type=Bone,Name=\"clavicle_r\"),Weight=0.12 ...(truncated)"
    Weight=1
   )
   > &RigUnit_ParentConstraint(
    AdvancedSettings=(RotationOrderForFilter=XZY)
    Child=(Type=Bone,Name="base_jacketFlap_01_S_r")
    Parents="((Item=(Type=Bone,Name=\"spine_01\"),Weight=0.0155),(Item=(Type=Bone,Name=\"spine_03\"),Weight=0.0667),(Item=(Type=Bone,Name=\"spine_02\"),Weight=0.0771),(Item=(Type=Bone,Name=\"upperarm_twist_01_r\"),Weight=0.0857),(Item=(Type=Bone,Name=\"clavicle_r\"),Weight=0.109 ...(truncated)"
    Weight=1
   )
   > &RigUnit_ParentConstraint(
    AdvancedSettings=(RotationOrderForFilter=XZY)
    Child=(Type=Bone,Name="base_jacketFlap_01_F_r")
    Parents="((Item=(Type=Bone,Name=\"upperarm_twist_01_r\"),Weight=0.0052),(Item=(Type=Bone,Name=\"clavicle_r\"),Weight=0.0265),(Item=(Type=Bone,Name=\"spine_03\"),Weight=0.0374),(Item=(Type=Bone,Name=\"spine_02\"),Weight=0.0415),(Item=(Type=Bone,Name=\"spine_05\"),Weight=0.046 ...(truncated)"
    Weight=1
   )
   > &RigUnit_ParentConstraint(
    AdvancedSettings=(RotationOrderForFilter=XZY)
    Child=(Type=Bone,Name="base_legBelt_01_B_l")
    Parents=((Item=(Type=Bone,Name="thigh_twist_01_l"),Weight=0.1878),(Item=(Type=Bone,Name="deform_glute_l"),Weight=0.3956),(Item=(Type=Bone,Name="pelvis"),Weight=0.4167))
    Weight=1
   )
   > &RigUnit_ParentConstraint(
    AdvancedSettings=(RotationOrderForFilter=XZY)
    Child=(Type=Bone,Name="base_regBelt_01_B_r")
    Parents=((Item=(Type=Bone,Name="thigh_twist_01_r"),Weight=0.1878),(Item=(Type=Bone,Name="deform_glute_r"),Weight=0.3956),(Item=(Type=Bone,Name="pelvis"),Weight=0.4167))
    Weight=1
   )
   RigVMFunctionReturnNode(ExecuteContext=@RigVMFunction_Sequence.B)
  ]}
 New Function_ContainedGraph {nodes [
   n1 = RigVMFunctionEntryNode()
   RigVMFunction_Sequence = &RigVMFunction_Sequence(ExecuteContext=@n1.ExecuteContext)
   DeltaRotation v4_1 = RigVMFunctionReferenceNode(
    Rotation Order=ZYX
    child=(Type=Bone,Name="thigh_l")
    debugRotation=false
    outputMinMax=(X=0,Y=1)
    parent=(Type=Null,Name="thigh_l_null")
    premult=1
    rotationMinMax=(X=0,Y=-120)
    triggerAxis=Z
    twistAxis=(X=1,Y=0,Z=0)
    ExecuteContext=@RigVMFunction_Sequence.A
   )
   > RigVMVariableNode(Value=@"DeltaRotation v4_1".outputWeight) {variable=SPR L Leg Fwd}
   > DeltaRotation v4_2 = RigVMFunctionReferenceNode(
    Rotation Order=ZYX
    child=(Type=Bone,Name="thigh_r")
    debugRotation=false
    outputMinMax=(X=0,Y=1)
    parent=(Type=Null,Name="thigh_r_null")
    premult=1
    rotationMinMax=(X=0,Y=-120)
    triggerAxis=Z
    twistAxis=(X=1,Y=0,Z=0)
   )
   > RigVMVariableNode(Value=@"DeltaRotation v4_2".outputWeight) {variable=SPR R Leg Fwd}
   > DeltaRotation v4_3 = RigVMFunctionReferenceNode(
    Rotation Order=ZYX
    child=(Type=Bone,Name="thigh_l")
    debugRotation=false
    outputMinMax=(X=0,Y=1)
    parent=(Type=Null,Name="thigh_l_null")
    premult=1
    rotationMinMax=(X=0,Y=120)
    triggerAxis=Z
    twistAxis=(X=1,Y=0,Z=0)
   )
   > RigVMVariableNode(Value=@"DeltaRotation v4_3".outputWeight) {variable=SPR L Leg Bkd}
   > DeltaRotation v4_4 = RigVMFunctionReferenceNode(
    Rotation Order=ZYX
    child=(Type=Bone,Name="thigh_r")
    debugRotation=false
    outputMinMax=(X=0,Y=1)
    parent=(Type=Null,Name="thigh_r_null")
    premult=1
    rotationMinMax=(X=0,Y=120)
    triggerAxis=Z
    twistAxis=(X=1,Y=0,Z=0)
   )
   > RigVMVariableNode(Value=@"DeltaRotation v4_4".outputWeight) {variable=SPR R Leg Bkd}
   > DeltaRotation v4_5 = RigVMFunctionReferenceNode(
    Rotation Order=ZYX
    absoluteValue=false
    child=(Type=Bone,Name="thigh_l")
    debugRotation=false
    outputMinMax=(X=0,Y=1)
    parent=(Type=Null,Name="thigh_l_null")
    premult=1
    rotationMinMax=(X=-8,Y=-90)
    triggerAxis=Y
    twistAxis=(X=1,Y=0,Z=0)
   )
   > RigVMVariableNode(Value=@"DeltaRotation v4_5".outputWeight) {variable=SPR L Leg Otw}
   > DeltaRotation v4_6 = RigVMFunctionReferenceNode(
    Rotation Order=ZYX
    child=(Type=Bone,Name="thigh_r")
    debugRotation=false
    outputMinMax=(X=0,Y=1)
    parent=(Type=Null,Name="thigh_r_null")
    premult=1
    rotationMinMax=(X=-8,Y=-90)
    triggerAxis=Y
    twistAxis=(X=1,Y=0,Z=0)
   )
   > RigVMVariableNode(Value=@"DeltaRotation v4_6".outputWeight) {variable=SPR R Leg Otw}
   n2 = &RigVMFunction_AnimEvalRichCurve(
    Curve="(EditorCurveData=(Keys=((InterpMode=RCIM_Cubic,TangentMode=RCTM_Break,TangentWeightMode=RCTWM_WeightedNone,Time=0,Value=0,ArriveTangent=0,ArriveTangentWeight=0,LeaveTangent=0,LeaveTangentWeight=0),(InterpMode=RCIM_Cubic,TangentMode=RCTM_Break,TangentWeight ...(truncated)"
    SourceMinimum=0.5
    TargetMinimum=0
    Value=&RigVMFunction_MathDoubleRemap(
     SourceMaximum=30
     SourceMinimum=60
     TargetMaximum=0
     TargetMinimum=1
     bClamp=False
     Value=&RigUnit_GetRelativeTransformForItem(Child=(Type=Bone,Name="pelvis"), Parent=(Type=Bone,Name="root"), bChildInitial=False, bParentInitial=False).RelativeTransform.Translation.Z
    ).Result
   )
   > RigVMVariableNode(Value=@n2.Result) {variable=PelvisProx}
   > DeltaRotation v4_7 = RigVMFunctionReferenceNode(
    Rotation Order=ZXY
    child=(Type=Bone,Name="calf_l")
    debugRotation=false
    outputMinMax=(X=0,Y=1)
    parent=(Type=Null,Name="calf_l_null")
    premult=1
    rotationMinMax=(X=0,Y=120)
    triggerAxis=Z
    twistAxis=(X=1,Y=0,Z=0)
   )
   > RigVMVariableNode(Value=@"DeltaRotation v4_7".outputWeight) {variable=SPR L Calf Bkd}
   > DeltaRotation v4_8 = RigVMFunctionReferenceNode(
    Rotation Order=ZXY
    child=(Type=Bone,Name="calf_r")
    debugRotation=false
    outputMinMax=(X=0,Y=1)
    parent=(Type=Null,Name="calf_r_null")
    premult=1
    rotationMinMax=(X=0,Y=120)
    triggerAxis=Z
    twistAxis=(X=1,Y=0,Z=0)
   )
   > RigVMVariableNode(Value=@"DeltaRotation v4_8".outputWeight) {variable=SPR R Calf Bkd}
   > DeltaRotation v4_9 = RigVMFunctionReferenceNode(
    Rotation Order=ZXY
    child=(Type=Bone,Name="calf_l")
    debugRotation=false
    outputMinMax=(X=0,Y=1)
    parent=(Type=Null,Name="calf_l_null")
    premult=1
    rotationMinMax=(X=120,Y=140)
    triggerAxis=Z
    twistAxis=(X=1,Y=0,Z=0)
   )
   > RigVMVariableNode(Value=@"DeltaRotation v4_9".outputWeight) {variable=SPR L Calf Bkd_Max}
   > DeltaRotation v4_10 = RigVMFunctionReferenceNode(
    Rotation Order=ZXY
    child=(Type=Bone,Name="calf_r")
    debugRotation=false
    outputMinMax=(X=0,Y=1)
    parent=(Type=Null,Name="calf_r_null")
    premult=1
    rotationMinMax=(X=120,Y=140)
    triggerAxis=Z
    twistAxis=(X=1,Y=0,Z=0)
   )
   > RigVMVariableNode(Value=@"DeltaRotation v4_10".outputWeight) {variable=SPR R Calf Bkd_Max}
   > DeltaRotation v4_11 = RigVMFunctionReferenceNode(
    Rotation Order=ZYX
    absoluteValue=false
    child=(Type=Bone,Name="thigh_l")
    debugRotation=false
    outputMinMax=(X=0,Y=1)
    parent=(Type=Null,Name="thigh_l_null")
    premult=1
    rotationMinMax=(X=0,Y=-20)
    triggerAxis=Z
    twistAxis=(X=1,Y=0,Z=0)
   )
   > RigVMVariableNode(Value=@"DeltaRotation v4_11".outputWeight) {variable=IdleThigh_L}
   > DeltaRotation v4_12 = RigVMFunctionReferenceNode(
    Rotation Order=ZYX
    absoluteValue=false
    child=(Type=Bone,Name="thigh_r")
    debugRotation=false
    outputMinMax=(X=0,Y=1)
    parent=(Type=Null,Name="thigh_r_null")
    premult=1
    rotationMinMax=(X=0,Y=-10)
    triggerAxis=Y
    twistAxis=(X=1,Y=0,Z=0)
   )
   > RigVMVariableNode(Value=@"DeltaRotation v4_12".outputWeight) {variable=IdleThigh_R}
   RigVMFunctionReturnNode(ExecuteContext=@RigVMFunction_Sequence.B)
  ]}
 New Function_ContainedGraph {nodes [
   n1 = RigVMFunctionEntryNode()
   RigVMFunction_Sequence = &RigVMFunction_Sequence(ExecuteContext=@n1.ExecuteContext)
   &RigUnit_ModifyTransforms(
    ItemToModify="((Item=(Type=Bone,Name=\"dyn_legBelt_02_B_l\"),Transform=(Rotation=(X=0.0663,Y=0.0866,Z=0.4933,W=0.863),Translation=(X=0,Y=0,Z=0),Scale3D=(X=1,Y=1,Z=1))),(Item=(Type=Bone,Name=\"dyn_legBelt_03_F_l\"),Transform=(Rotation=(X=-0.177,Y=-0.3237,Z=-0.4459,W=0.8155), ...(truncated)"
    WeightMinimum=0
    ExecutePin=@RigVMFunction_Sequence.A
    Weight=RigVMVariableNode() {variable=SPR L Leg Fwd}.Value
   )
   > &RigUnit_ModifyTransforms(
    ItemToModify="((Item=(Type=Bone,Name=\"dyn_regBelt_03_B_r\"),Transform=(Rotation=(X=0.115,Y=-0.1927,Z=0.3817,W=-0.8966),Translation=(X=0,Y=0,Z=0),Scale3D=(X=1,Y=1,Z=1))),(Item=(Type=Bone,Name=\"dyn_regBelt_06_B_r\"),Transform=(Rotation=(X=0.054,Y=0.4318,Z=-0.093,W=-0.8955), ...(truncated)"
    WeightMinimum=0
    Weight=RigVMVariableNode() {variable=SPR R Leg Fwd}.Value
   )
   > &RigUnit_ModifyTransforms(
    ItemToModify="((Item=(Type=Bone,Name=\"dyn_legBelt_02_B_l\"),Transform=(Rotation=(X=0.2122,Y=0.3752,Z=0.6681,W=-0.6065),Translation=(X=-4.8586,Y=0.7721,Z=3.8301),Scale3D=(X=1,Y=1,Z=1))),(Item=(Type=Bone,Name=\"dyn_legBelt_05_B_l\"),Transform=(Rotation=(X=0,Y=0,Z=0,W=1),Tran ...(truncated)"
    WeightMinimum=0
    Weight=RigVMVariableNode() {variable=SPR L Leg Bkd}.Value
   )
   > &RigUnit_ModifyTransforms(
    ItemToModify="((Item=(Type=Bone,Name=\"dyn_regBelt_01_B_r\"),Transform=(Rotation=(X=0.2597,Y=-0.4674,Z=0.535,W=0.6542),Translation=(X=-2.3573,Y=0.6444,Z=5.5454),Scale3D=(X=1,Y=1,Z=1))),(Item=(Type=Bone,Name=\"dyn_regBelt_03_F_r\"),Transform=(Rotation=(X=0.004,Y=-0.0635,Z=-0 ...(truncated)"
    WeightMinimum=0
    Weight=RigVMVariableNode() {variable=SPR R Leg Bkd}.Value
   )
   > &RigUnit_ModifyTransforms(
    ItemToModify="((Item=(Type=Bone,Name=\"dyn_legBelt_02_B_l\"),Transform=(Rotation=(X=0.1507,Y=-0.4725,Z=0.2366,W=0.8355),Translation=(X=0,Y=0,Z=0),Scale3D=(X=1,Y=1,Z=1))),(Item=(Type=Bone,Name=\"dyn_legBelt_03_F_l\"),Transform=(Rotation=(X=-0.0584,Y=0.0523,Z=0.0637,W=0.9949) ...(truncated)"
    WeightMinimum=0
    Weight=RigVMVariableNode() {variable=SPR L Leg Otw}.Value
   )
   > &RigUnit_ModifyTransforms(
    ItemToModify="((Item=(Type=Bone,Name=\"dyn_regBelt_01_B_r\"),Transform=(Rotation=(X=-0.3637,Y=-0.3757,Z=-0.2367,W=0.8189),Translation=(X=0,Y=0,Z=0),Scale3D=(X=1,Y=1,Z=1))),(Item=(Type=Bone,Name=\"dyn_beltStrap_01\"),Transform=(Rotation=(X=-0.1659,Y=-0.0834,Z=-0.1571,W=0.97) ...(truncated)"
    WeightMinimum=0
    Weight=RigVMVariableNode() {variable=SPR R Leg Otw}.Value
   )
   > &RigUnit_ModifyTransforms(
    ItemToModify="((Item=(Type=Bone,Name=\"dyn_legBelt_03_B_l\"),Transform=(Rotation=(X=0,Y=-0.066,Z=0,W=0.9978),Translation=(X=0,Y=0,Z=0),Scale3D=(X=1,Y=1,Z=1))),(Item=(Type=Bone,Name=\"dyn_legBelt_04_F_l\"),Transform=(Rotation=(X=0,Y=0,Z=0,W=1),Translation=(X=0,Y=0,Z=0),Scale ...(truncated)"
    WeightMinimum=0
    Weight=RigVMVariableNode() {variable=SPR L Calf Bkd}.Value
   )
   > &RigUnit_ModifyTransforms(
    ItemToModify=((Item=(Type=Bone,Name="dyn_regBelt_02_B_r"),Transform=(Rotation=(X=0,Y=0,Z=0.1112,W=-0.9938),Translation=(X=0,Y=0,Z=0),Scale3D=(X=1,Y=1,Z=1))))
    WeightMinimum=0
    Weight=RigVMVariableNode() {variable=SPR R Calf Bkd}.Value
   )
   > &RigUnit_ModifyTransforms(
    ItemToModify="((Item=(Type=Bone,Name=\"dyn_legBelt_01_B_l\"),Transform=(Rotation=(X=0.0737,Y=-0.1121,Z=0.022,W=0.9907),Translation=(X=0,Y=0,Z=0),Scale3D=(X=1,Y=1,Z=1))),(Item=(Type=Bone,Name=\"dyn_legBelt_04_B_l\"),Transform=(Rotation=(X=0.0759,Y=-0.2226,Z=-0.0589,W=-0.9702 ...(truncated)"
    WeightMinimum=0
    Weight=RigVMVariableNode() {variable=SPR L Calf Bkd_Max}.Value
   )
   > &RigUnit_ModifyTransforms(
    ItemToModify="((Item=(Type=Bone,Name=\"dyn_regBelt_02_B_r\"),Transform=(Rotation=(X=0.0088,Y=0.0642,Z=0.1355,W=-0.9887),Translation=(X=0,Y=0,Z=0),Scale3D=(X=1,Y=1,Z=1))),(Item=(Type=Bone,Name=\"dyn_regBelt_04_B_r\"),Transform=(Rotation=(X=0.0081,Y=-0.054,Z=-0.1489,W=-0.9873 ...(truncated)"
    WeightMinimum=0
    Weight=RigVMVariableNode() {variable=SPR R Calf Bkd_Max}.Value
   )
   > &RigUnit_ModifyTransforms(
    ItemToModify=((Item=(Type=Bone,Name="dyn_legBelt_03_F_l"),Transform=(Rotation=(X=0,Y=0.0698,Z=0,W=0.9976),Translation=(X=0,Y=0,Z=0),Scale3D=(X=1,Y=1,Z=1))))
    WeightMinimum=0
    Weight=RigVMVariableNode() {variable=IdleThigh_L}.Value
   )
   > &RigUnit_ModifyTransforms(
    ItemToModify=((Item=(Type=Bone,Name="dyn_regBelt_03_F_r"),Transform=(Rotation=(X=0,Y=0.0698,Z=0,W=0.9976),Translation=(X=0,Y=0,Z=0),Scale3D=(X=1,Y=1,Z=1))))
    WeightMinimum=0
    Weight=RigVMVariableNode() {variable=IdleThigh_R}.Value
   )
   > DeltaRotation v4_1 = RigVMFunctionReferenceNode(
    Rotation Order=ZYX
    child=(Type=Bone,Name="thigh_r")
    debugRotation=false
    outputMinMax=(X=0,Y=1)
    parent=(Type=Null,Name="thigh_r_null")
    premult=1
    rotationMinMax=(X=0,Y=-55)
    triggerAxis=Z
    twistAxis=(X=1,Y=0,Z=0)
   )
   > &RigUnit_ModifyTransforms(
    ItemToModify=((Item=(Type=Bone,Name="dyn_beltStrap_01"),Transform=(Rotation=(X=0.1818,Y=0.1283,Z=-0.0095,W=-0.9749),Translation=(X=0,Y=0,Z=0),Scale3D=(X=1,Y=1,Z=1))))
    WeightMinimum=0
    Weight=@"DeltaRotation v4_1".outputWeight
   )
   > DeltaRotation v4_2 = RigVMFunctionReferenceNode(
    Rotation Order=ZYX
    child=(Type=Bone,Name="thigh_r")
    debugRotation=false
    outputMinMax=(X=0,Y=1)
    parent=(Type=Null,Name="thigh_r_null")
    premult=1
    rotationMinMax=(X=-55,Y=-90)
    triggerAxis=Z
    twistAxis=(X=1,Y=0,Z=0)
   )
   > &RigUnit_ModifyTransforms(
    ItemToModify=((Item=(Type=Bone,Name="dyn_beltStrap_01"),Transform=(Rotation=(X=-0.3454,Y=-0.0265,Z=-0.0467,W=0.9369),Translation=(X=-0.6445,Y=-1.0747,Z=3.0961),Scale3D=(X=1,Y=1,Z=1))))
    WeightMinimum=0
    Weight=@"DeltaRotation v4_2".outputWeight
   )
   > DeltaRotation v4_3 = RigVMFunctionReferenceNode(
    Rotation Order=ZYX
    child=(Type=Bone,Name="thigh_r")
    debugRotation=false
    outputMinMax=(X=0,Y=1)
    parent=(Type=Null,Name="thigh_r_null")
    premult=1
    rotationMinMax=(X=-90,Y=-120)
    triggerAxis=Z
    twistAxis=(X=1,Y=0,Z=0)
   )
   > &RigUnit_ModifyTransforms(
    ItemToModify=((Item=(Type=Bone,Name="dyn_beltStrap_01"),Transform=(Rotation=(X=0.2983,Y=0.0782,Z=0.1959,W=-0.9309),Translation=(X=-0.3826,Y=-3.4573,Z=3.6825),Scale3D=(X=1,Y=1,Z=1))))
    WeightMinimum=0
    Weight=@"DeltaRotation v4_3".outputWeight
   )
   > DeltaRotation v4_4 = RigVMFunctionReferenceNode(
    Rotation Order=ZYX
    child=(Type=Bone,Name="spine_05")
    debugRotation=false
    outputMinMax=(X=0,Y=1)
    parent=(Type=Null,Name="spine_05_null")
    premult=1
    rotationMinMax=(X=0,Y=10)
    triggerAxis=Z
    twistAxis=(X=1,Y=0,Z=0)
   )
   Multiply = &RigVMFunction_MathDoubleMul(A=@"DeltaRotation v4_4".outputWeight, B=RigVMVariableNode() {variable=Pub_SwingPickaxe}.Value)
   > &RigUnit_ModifyTransforms(
    ItemToModify="((Item=(Type=Bone,Name=\"base_regBelt_01_B_r\"),Transform=(Rotation=(X=0,Y=-0.0436,Z=0,W=0.999),Translation=(X=0,Y=0,Z=0),Scale3D=(X=1,Y=1,Z=1))),(Item=(Type=Bone,Name=\"base_regBelt_01_F_r\"),Transform=(Rotation=(X=0,Y=-0.0436,Z=0,W=0.999),Translation=(X=0,Y= ...(truncated)"
    WeightMinimum=0
    Weight=@Multiply.Result
   )
   > DeltaRotation v4_5 = RigVMFunctionReferenceNode(
    Rotation Order=ZYX
    child=(Type=Bone,Name="spine_05")
    debugRotation=false
    onlyTwist=true
    outputMinMax=(X=0,Y=1)
    parent=(Type=Null,Name="spine_05_null")
    premult=1
    rotationMinMax=(X=0,Y=-65)
    triggerAxis=X
    twistAxis=(X=1,Y=0,Z=0)
   )
   > &RigUnit_ModifyTransforms(
    ItemToModify="((Item=(Type=Bone,Name=\"dyn_jacketFlap_01_SB_r\"),Transform=(Rotation=(X=0,Y=-0.2164,Z=0,W=0.9763),Translation=(X=0,Y=0,Z=0),Scale3D=(X=1,Y=1,Z=1))),(Item=(Type=Bone,Name=\"dyn_jacketFlap_01_S_r\"),Transform=(Rotation=(X=0.0175,Y=0,Z=0,W=0.9998),Translation=( ...(truncated)"
    WeightMinimum=0
    Weight=&RigVMFunction_MathDoubleMul(A=@"DeltaRotation v4_5".outputWeight, B=RigVMVariableNode() {variable=Pub_SwingPickaxe}.Value).Result
   )
   RigVMFunctionReturnNode(ExecuteContext=@RigVMFunction_Sequence.B)
  ]}
 New Function_ContainedGraph {nodes [
   n1 = RigVMFunctionEntryNode()
   RigVMFunction_MathBoolToInteger = &RigVMFunction_MathBoolToInteger(Value=@n1.onlyTwist)
   RigVMFunction_MathQuaternionSwingTwist = &RigVMFunction_MathQuaternionSwingTwist(
    TwistAxis=@n1.twistAxis
    Input=&RigUnit_GetRelativeTransformForItem(bChildInitial=False, bParentInitial=False, Child=@n1.child, Parent=@n1.parent).RelativeTransform.Rotation
   )
   n2 = RigVMDispatchNode(
    Values=((),())
    Index=@RigVMFunction_MathBoolToInteger.Result
    Values.0=@RigVMFunction_MathQuaternionSwingTwist.Swing
    Values.1=@RigVMFunction_MathQuaternionSwingTwist.Twist
   ) {template="DISPATCH_RigVMDispatch_SelectInt32(in Index,in Values,out Result)"}
   RigVMFunction_MathQuaternionToEuler = &RigVMFunction_MathQuaternionToEuler(Value=@n2.Result, RotationOrder=@n1.Rotation Order)
   > DISPATCH_RigVMDispatch_Print = RigVMDispatchNode(ScreenColor=(R=1,G=1,B=1,A=1), ScreenDuration=1.5, Enabled=@n1.debugRotation, Value=@RigVMFunction_MathQuaternionToEuler.Result) {template="DISPATCH_RigVMDispatch_Print(in Prefix,in Value,in Enabled,in ScreenDuration,in ScreenColor)"}
   Cast = RigVMDispatchNode(Value=@n1.triggerAxis) {template="DISPATCH_RigVMDispatch_CastEnumToInt(in Value,out Result)"}
   DISPATCH_RigVMDispatch_SelectInt32 = RigVMDispatchNode(
    Values=(0,0,0)
    Index=@Cast.Result
    Values.0=@RigVMFunction_MathQuaternionToEuler.Result.X
    Values.1=@RigVMFunction_MathQuaternionToEuler.Result.Y
    Values.2=@RigVMFunction_MathQuaternionToEuler.Result.Z
   ) {template="DISPATCH_RigVMDispatch_SelectInt32(in Index,in Values,out Result)"}
   Remap = &RigVMFunction_MathDoubleRemap(
    bClamp=False
    Value=@DISPATCH_RigVMDispatch_SelectInt32.Result
    TargetMinimum=@n1.outputMinMax.X
    TargetMaximum=@n1.outputMinMax.Y
    SourceMinimum=@n1.rotationMinMax.X
    SourceMaximum=@n1.rotationMinMax.Y
   )
   Absolute = &RigVMFunction_MathDoubleAbs(Value=@Remap.Result)
   Multiply = &RigVMFunction_MathDoubleMul(
    A=&RigVMFunction_MathDoubleClamp(Value=RigVMDispatchNode(True=@Absolute.Result, Condition=@n1.absoluteValue, False=@Remap.Result) {template="DISPATCH_RigVMDispatch_If(in Condition,in True,in False,out Result)"}.Result).Result
    B=@n1.premult
   )
   > RigVMFunctionReturnNode(outputQuat=@n2.Result, outputWeight=@Multiply.Result)
  ]}
 New Function_ContainedGraph {nodes [
   n1 = RigVMFunctionEntryNode()
   RigVMFunction_MathBoolToInteger = &RigVMFunction_MathBoolToInteger(Value=@n1.onlyTwist)
   RigVMFunction_MathQuaternionSwingTwist = &RigVMFunction_MathQuaternionSwingTwist(
    TwistAxis=@n1.twistAxis
    Input=&RigUnit_GetRelativeTransformForItem(bChildInitial=False, bParentInitial=False, Child=@n1.child, Parent=@n1.parent).RelativeTransform.Rotation
   )
   n2 = RigVMDispatchNode(
    Values=((),())
    Index=@RigVMFunction_MathBoolToInteger.Result
    Values.0=@RigVMFunction_MathQuaternionSwingTwist.Swing
    Values.1=@RigVMFunction_MathQuaternionSwingTwist.Twist
   ) {template="DISPATCH_RigVMDispatch_SelectInt32(in Index,in Values,out Result)"}
   RigVMFunction_MathQuaternionToEuler = &RigVMFunction_MathQuaternionToEuler(Value=@n2.Result, RotationOrder=@n1.Rotation Order)
   > DISPATCH_RigVMDispatch_Print = RigVMDispatchNode(ScreenColor=(R=1,G=1,B=1,A=1), ScreenDuration=1.5, Enabled=@n1.debugRotation, Value=@RigVMFunction_MathQuaternionToEuler.Result) {template="DISPATCH_RigVMDispatch_Print(in Prefix,in Value,in Enabled,in ScreenDuration,in ScreenColor)"}
   Cast = RigVMDispatchNode(Value=@n1.triggerAxis) {template="DISPATCH_RigVMDispatch_CastEnumToInt(in Value,out Result)"}
   DISPATCH_RigVMDispatch_SelectInt32 = RigVMDispatchNode(
    Values=(0,0,0)
    Index=@Cast.Result
    Values.0=@RigVMFunction_MathQuaternionToEuler.Result.X
    Values.1=@RigVMFunction_MathQuaternionToEuler.Result.Y
    Values.2=@RigVMFunction_MathQuaternionToEuler.Result.Z
   ) {template="DISPATCH_RigVMDispatch_SelectInt32(in Index,in Values,out Result)"}
   Remap = &RigVMFunction_MathDoubleRemap(
    bClamp=False
    Value=@DISPATCH_RigVMDispatch_SelectInt32.Result
    TargetMinimum=@n1.outputMinMax.X
    TargetMaximum=@n1.outputMinMax.Y
    SourceMinimum=@n1.rotationMinMax.X
    SourceMaximum=@n1.rotationMinMax.Y
   )
   Absolute = &RigVMFunction_MathDoubleAbs(Value=@Remap.Result)
   Multiply = &RigVMFunction_MathDoubleMul(
    A=&RigVMFunction_MathDoubleClamp(Value=RigVMDispatchNode(True=@Absolute.Result, Condition=@n1.absoluteValue, False=@Remap.Result) {template="DISPATCH_RigVMDispatch_If(in Condition,in True,in False,out Result)"}.Result).Result
    B=@n1.premult
   )
   > RigVMFunctionReturnNode(outputQuat=@n2.Result, outputWeight=@Multiply.Result)
  ]}
 New Function_ContainedGraph {nodes [
   n1 = RigVMFunctionEntryNode()
   RigVMFunction_Sequence = &RigVMFunction_Sequence(ExecuteContext=@n1.ExecuteContext)
   &RigUnit_ModifyTransforms(
    ItemToModify="((Item=(Type=Bone,Name=\"dyn_legBelt_01_B_l\"),Transform=(Rotation=(X=-0.1205,Y=0.1509,Z=-0.4009,W=0.8955),Translation=(X=0,Y=0,Z=0),Scale3D=(X=1,Y=1,Z=1))),(Item=(Type=Bone,Name=\"dyn_legBelt_05_B_l\"),Transform=(Rotation=(X=0.0864,Y=-0.13,Z=0.0114,W=0.9877), ...(truncated)"
    WeightMinimum=0
    ExecutePin=@RigVMFunction_Sequence.A
    Weight=RigVMVariableNode() {variable=Pub_Is_Crouching}.Value
   )
   AlphaInterp = &RigVMFunction_AlphaInterp(InterpSpeedDecreasing=5, InterpSpeedIncreasing=5, Scale=1, bClampResult=False, Value=RigVMVariableNode() {variable=Pub_Is_KneeSliding}.Value)
   > &RigUnit_ModifyTransforms(
    ItemToModify="((Item=(Type=Bone,Name=\"dyn_legBelt_01_B_l\"),Transform=(Rotation=(X=0.0595,Y=-0.5794,Z=0.1823,W=0.7921),Translation=(X=0,Y=0,Z=0),Scale3D=(X=1,Y=1,Z=1))),(Item=(Type=Bone,Name=\"dyn_legBelt_05_B_l\"),Transform=(Rotation=(X=0,Y=0,Z=0,W=1),Translation=(X=0,Y=0 ...(truncated)"
    WeightMinimum=0
    Weight=@AlphaInterp.Result
   )
   > &RigUnit_ModifyTransforms(
    ItemToModify="((Item=(Type=Bone,Name=\"dyn_regBelt_01_B_r\"),Transform=(Rotation=(X=0,Y=0,Z=0.1736,W=0.9848),Translation=(X=0,Y=0,Z=0),Scale3D=(X=1,Y=1,Z=1))),(Item=(Type=Bone,Name=\"dyn_legBelt_03_F_l\"),Transform=(Rotation=(X=0,Y=-0.1736,Z=0,W=0.9848),Translation=(X=0,Y=0 ...(truncated)"
    WeightMinimum=0
    Weight=RigVMVariableNode() {variable=Pub_Is_BaseballSliding}.Value
   )
   > DeltaRotation v4_12 = RigVMFunctionReferenceNode(
    Rotation Order=ZYX
    absoluteValue=false
    child=(Type=Bone,Name="thigh_r")
    debugRotation=false
    outputMinMax=(X=0,Y=1)
    parent=(Type=Null,Name="thigh_r_null")
    premult=1
    rotationMinMax=(X=0,Y=-18)
    triggerAxis=Y
    twistAxis=(X=1,Y=0,Z=0)
   )
   Multiply = &RigVMFunction_MathDoubleMul(A=@"DeltaRotation v4_12".outputWeight, B=RigVMVariableNode() {variable=Pub_GrindRail}.Value)
   > &RigUnit_ModifyTransforms(
    ItemToModify=((Item=(Type=Bone,Name="dyn_regBelt_01_B_r"),Transform=(Rotation=(X=0,Y=-0.2588,Z=0,W=0.9659),Translation=(X=0,Y=0,Z=0),Scale3D=(X=1,Y=1,Z=1))))
    WeightMinimum=0
    Weight=@Multiply.Result
   )
   n2 = &RigVMFunction_AlphaInterp(
    InterpSpeedDecreasing=5
    InterpSpeedIncreasing=5
    Scale=1
    bClampResult=False
    Value=&RigVMFunction_MathDoubleMul(A=RigVMVariableNode() {variable=Pub_Is_Crouching}.Value, B=RigVMVariableNode() {variable=Pub_LeftStrafe}.Value).Result
   )
   > &RigUnit_ModifyTransforms(
    ItemToModify="((Item=(Type=Bone,Name=\"dyn_legBelt_03_F_l\"),Transform=(Rotation=(X=0,Y=-0.1305,Z=0,W=0.9914),Translation=(X=0,Y=0,Z=0),Scale3D=(X=1,Y=1,Z=1))),(Item=(Type=Bone,Name=\"dyn_legBelt_03_B_l\"),Transform=(Rotation=(X=0,Y=-0.1736,Z=0,W=0.9848),Translation=(X=0,Y= ...(truncated)"
    WeightMinimum=0
    Weight=@n2.Result
   )
   RigVMFunctionReturnNode(ExecuteContext=@RigVMFunction_Sequence.B)
  ]}
 New Function_ContainedGraph {nodes [
   n1 = RigVMFunctionEntryNode()
   RigVMFunction_Sequence = &RigVMFunction_Sequence(ExecuteContext=@n1.ExecuteContext)
   DeltaRotation v4_1 = RigVMFunctionReferenceNode(
    Rotation Order=ZYX
    child=(Type=Bone,Name="thigh_l")
    debugRotation=false
    onlyTwist=false
    outputMinMax=(X=0,Y=1)
    parent=(Type=Null,Name="thigh_l_null")
    premult=1
    rotationMinMax=(X=0,Y=-113)
    twistAxis=(X=1,Y=0,Z=0)
    ExecuteContext=@RigVMFunction_Sequence.A
   )
   > DeltaRotation v4_2 = RigVMFunctionReferenceNode(
    Rotation Order=ZYX
    child=(Type=Bone,Name="thigh_l")
    debugRotation=false
    onlyTwist=false
    outputMinMax=(X=0,Y=1)
    parent=(Type=Null,Name="thigh_l_null")
    premult=1
    rotationMinMax=(X=0,Y=-72)
    triggerAxis=Y
    twistAxis=(X=1,Y=0,Z=0)
   )
   > DeltaRotation v4_3 = RigVMFunctionReferenceNode(
    Rotation Order=ZYX
    child=(Type=Bone,Name="thigh_l")
    debugRotation=false
    onlyTwist=false
    outputMinMax=(X=0,Y=1)
    parent=(Type=Null,Name="thigh_l_null")
    premult=1
    rotationMinMax=(X=0,Y=-128)
    triggerAxis=Z
    twistAxis=(X=1,Y=0,Z=0)
   )
   Multiply = &RigVMFunction_MathDoubleMul(A=@"DeltaRotation v4_1".outputWeight, B=@"DeltaRotation v4_2".outputWeight)
   > &RigUnit_ModifyTransforms(
    ItemToModify="((Item=(Type=Bone,Name=\"dyn_legBelt_03_B_l\"),Transform=(Rotation=(X=-0.2393,Y=-0.1893,Z=0.0381,W=0.9515),Translation=(X=0,Y=0,Z=0),Scale3D=(X=1,Y=1,Z=1))),(Item=(Type=Bone,Name=\"dyn_legBelt_01_F_l\"),Transform=(Rotation=(X=-0.1076,Y=0.2389,Z=0.2607,W=0.9292 ...(truncated)"
    WeightMinimum=0
    Weight=&RigVMFunction_MathDoubleMul(A=@"DeltaRotation v4_3".outputWeight, B=@Multiply.Result).Result
   )
   > DeltaRotation v4_4 = RigVMFunctionReferenceNode(
    Rotation Order=ZYX
    child=(Type=Bone,Name="thigh_r")
    debugRotation=false
    onlyTwist=false
    outputMinMax=(X=0,Y=1)
    parent=(Type=Null,Name="thigh_r_null")
    premult=1
    rotationMinMax=(X=0,Y=-98)
    twistAxis=(X=1,Y=0,Z=0)
   )
   > DeltaRotation v4_5 = RigVMFunctionReferenceNode(
    Rotation Order=ZYX
    child=(Type=Bone,Name="thigh_r")
    debugRotation=false
    onlyTwist=false
    outputMinMax=(X=0,Y=1)
    parent=(Type=Null,Name="thigh_r_null")
    premult=1
    rotationMinMax=(X=0,Y=-63)
    triggerAxis=Y
    twistAxis=(X=1,Y=0,Z=0)
   )
   > DeltaRotation v4_6 = RigVMFunctionReferenceNode(
    Rotation Order=ZYX
    child=(Type=Bone,Name="thigh_r")
    debugRotation=false
    onlyTwist=false
    outputMinMax=(X=0,Y=1)
    parent=(Type=Null,Name="thigh_r_null")
    premult=1
    rotationMinMax=(X=0,Y=-125)
    triggerAxis=Z
    twistAxis=(X=1,Y=0,Z=0)
   )
   > &RigUnit_ModifyTransforms(
    ItemToModify="((Item=(Type=Bone,Name=\"dyn_regBelt_03_B_r\"),Transform=(Rotation=(X=0.3712,Y=-0.3185,Z=0.0821,W=0.8684),Translation=(X=0,Y=0,Z=0),Scale3D=(X=1,Y=1,Z=1))),(Item=(Type=Bone,Name=\"dyn_beltStrap_01\"),Transform=(Rotation=(X=-0.2743,Y=0,Z=0,W=-0.9616),Translatio ...(truncated)"
    WeightMinimum=0
    Weight=&RigVMFunction_MathDoubleMul(
     A=@"DeltaRotation v4_6".outputWeight
     B=&RigVMFunction_MathDoubleMul(B=@"DeltaRotation v4_4".outputWeight, A=@"DeltaRotation v4_5".outputWeight).Result
    ).Result
   )
   RigVMFunctionReturnNode(ExecuteContext=@RigVMFunction_Sequence.B)
  ]}
 New Function_ContainedGraph {nodes [
   n1 = RigVMFunctionEntryNode()
   RigVMFunction_Sequence = &RigVMFunction_Sequence(ExecuteContext=@n1.ExecuteContext)
   DeltaRotation v4_1 = RigVMFunctionReferenceNode(
    Rotation Order=ZYX
    child=(Type=Bone,Name="foot_l")
    debugRotation=false
    outputMinMax=(X=0,Y=1)
    parent=(Type=Null,Name="foot_l_null")
    premult=1
    rotationMinMax=(X=0,Y=70)
    triggerAxis=Z
    twistAxis=(X=1,Y=0,Z=0)
    ExecuteContext=@RigVMFunction_Sequence.A
   )
   > &RigUnit_ModifyTransforms(
    ItemToModify=((Item=(Type=Bone,Name="dyn_push_ankle_l"),Transform=(Rotation=(X=-0.1474,Y=0.069,Z=0.0103,W=0.9866),Translation=(X=19.4599,Y=8.4322,Z=-5.336),Scale3D=(X=1,Y=1,Z=1))))
    WeightMinimum=0
    Weight=@"DeltaRotation v4_1".outputWeight
   )
   > DeltaRotation v4_2 = RigVMFunctionReferenceNode(
    Rotation Order=ZYX
    child=(Type=Bone,Name="foot_r")
    debugRotation=false
    outputMinMax=(X=0,Y=1)
    parent=(Type=Null,Name="foot_r_null")
    premult=1
    rotationMinMax=(X=0,Y=70)
    triggerAxis=Z
    twistAxis=(X=1,Y=0,Z=0)
   )
   > &RigUnit_ModifyTransforms(
    ItemToModify=((Item=(Type=Bone,Name="dyn_push_ankle_r"),Transform=(Rotation=(X=-0.1474,Y=0.069,Z=0.0103,W=0.9866),Translation=(X=14.0969,Y=7.9775,Z=-4.4411),Scale3D=(X=1,Y=1,Z=1))))
    WeightMinimum=0
    Weight=@"DeltaRotation v4_2".outputWeight
   )
   RigVMFunctionReturnNode(ExecuteContext=@RigVMFunction_Sequence.B)
  ]}
 RigVMFunctionLibrary {nodes [
   Ankle_Fix = RigVMCollapseNode(ExecuteContext=())
   Constraints = RigVMCollapseNode(ExecuteContext=())
   DeltaRotation v4 = RigVMCollapseNode(
    ExecuteContext=()
    Rotation Order=ZYX
    child=(Type=Bone,Name="None")
    outputMinMax=(X=0,Y=1)
    parent=(Type=Bone,Name="None")
    premult=1
    rotationMinMax=(X=0,Y=0)
    twistAxis=(X=1,Y=0,Z=0)
   )
   DeltaRotation v4_1 = RigVMCollapseNode(
    ExecuteContext=()
    Rotation Order=ZYX
    child=(Type=Bone,Name="None")
    outputMinMax=(X=0,Y=1)
    parent=(Type=Bone,Name="None")
    premult=1
    rotationMinMax=(X=0,Y=0)
    twistAxis=(X=1,Y=0,Z=0)
   )
   EasternBloc_Fix = RigVMCollapseNode(ExecuteContext=())
   IsCrouching_Fix = RigVMCollapseNode(ExecuteContext=())
   JacketEnds_SpineRotCompensator = RigVMCollapseNode(ExecuteContext=())
   NeckStap_Corrector = RigVMCollapseNode(ExecuteContext=())
   StrapsBelt_Fix = RigVMCollapseNode(ExecuteContext=())
   Variables = RigVMCollapseNode(ExecuteContext=())
  ]}
 RigVMModel {nodes [
   RigUnit_BeginExecution = &RigUnit_BeginExecution()
   FN_DeformRig_FNC = RigVMFunctionReferenceNode(
    DeformRig_ Curve="(deformRig_l_upperarm_up_alpha_72_#1=true,deformRig_r_upperarm_up_alpha_73_#2=true,deformRig_l_upperarm_fr_alpha_74_#3=true,deformRig_r_upperarm_fr_alpha_75_CB51B80C4 ...(truncated)"
    DeformRig_Bool="(upperarm_1_#4=True,elbow_3_#5=True,wrist_7_#6=True,groin_8_#7=True,glute_17_#8=True,knee_9_40C408324CA261 ...(truncated)"
    DeformRig_Debug=(upperarm_debug_18_#4=False,wrist_debug_20_#6=False,lat_debug_26_#9=False,groinGlute_debug_27_#7=False)
    DeformRig_Value="(upperarm_up_18_#10=(Rotation=(X=0.0,Y=-0.0,Z=0.0,W=1.0),Translation=(X=0.0,Y=0.0,Z=8.0),Scale3D=(X=0.70,Y=1.0,Z=1.0)),upperarm_fwd_fr_19_#11=(Rotation=(X=0.0,Y=-0.0,Z=0.0,W=1.0),Translation=(X=0.0, ...(truncated)"
    ExecuteContext=@RigUnit_BeginExecution.ExecutePin
   )
   > Constraints = RigVMFunctionReferenceNode()
   > Variables = RigVMFunctionReferenceNode()
   > StrapsBelt_Fix = RigVMFunctionReferenceNode()
   > IsCrouching_Fix = RigVMFunctionReferenceNode()
   > EasternBloc_Fix = RigVMFunctionReferenceNode()
   > NeckStap_Corrector = RigVMFunctionReferenceNode()
   > Ankle_Fix = RigVMFunctionReferenceNode()
  ]}
]
hierarchy [
 root {loc=<0 0 0>, rot=<0 0 0 1>}
 pelvis {parent=root, loc=<0 -0.5301 98.0571>, rot=<0.0127 0.707 -0.0127 -0.707>}
 spine_01 {parent=pelvis, loc=<6.1208 -0.9328 0>, rot=<0 0 0.034 -0.9994>}
 spine_02 {parent=spine_01, loc=<6.5997 -0.2095 0>, rot=<0 0 0.0238 -0.9997>}
 spine_03 {parent=spine_02, loc=<6.4914 -0.1237 0>, rot=<0 0 0.0149 -0.9999>}
 spine_04 {parent=spine_03, loc=<6.9598 0.0794 0>, rot=<0 0 -0.0132 -0.9999>}
 spine_05 {parent=spine_04, loc=<6.9987 0 0>, rot=<0 0 -0.0015 -1>}
 neck_01 {parent=spine_05, loc=<11.3138 0.1529 0>, rot=<0 0 -0.2003 -0.9797>}
 neck_02 {parent=neck_01, loc=<4.5627 0 0>, rot=<0 0 0.0504 -0.9987>}
 head {parent=neck_02, loc=<5.3914 0 0>, rot=<0 0 0.115 -0.9934>}
 HairMain_01 {parent=head, loc=<7.1837 -2.5907 0>, rot=<0.101 -0.6999 0.6999 -0.101>}
 HairMain_02 {parent=HairMain_01, loc=<6.2714 0 0>, rot=<0 -0.1224 0 -0.9925>}
 HairMain_03 {parent=HairMain_02, loc=<5.8192 0 0>, rot=<0 0.0295 0 -0.9996>}
 base_HairSpike_A_01_l {parent=head, loc=<6.5319 1.5352 -7.4061>, rot=<0.0041 -1 0 0>}
 dyn_HairSpike_A_01_l {parent=base_HairSpike_A_01_l, loc=<0 0 0>, rot=<0 0 0 1>}
 base_HairSpike_B_01_l {parent=head, loc=<9.9494 0.1681 -7.3092>, rot=<0.0041 -1 0 0>}
 dyn_HairSpike_B_01_l {parent=base_HairSpike_B_01_l, loc=<0 0 0>, rot=<0 0 0 1>}
 base_HairSpike_C_01_l {parent=head, loc=<3.358 6.9803 -5.9>, rot=<0.0041 -1 0 0>}
 dyn_HairSpike_C_01_l {parent=base_HairSpike_C_01_l, loc=<0 0 0>, rot=<0 0 0 1>}
 base_HairSpike_D_01_l {parent=head, loc=<5.9304 9.9965 -4.1098>, rot=<0.0041 -1 0 0>}
 dyn_HairSpike_D_01_l {parent=base_HairSpike_D_01_l, loc=<0 0 0>, rot=<0 0 0 1>}
 base_HairSpike_E_01_l {parent=head, loc=<8.4906 10.8221 -1.9717>, rot=<0.0041 -1 0 0>}
 dyn_HairSpike_E_01_l {parent=base_HairSpike_E_01_l, loc=<0 0 0>, rot=<0 0 0 1>}
 base_HairSpike_F_01_c {parent=head, loc=<8.3628 10.6369 0.201>, rot=<0.0041 -1 0 0>}
 dyn_HairSpike_F_01_c {parent=base_HairSpike_F_01_c, loc=<0 0 0>, rot=<0 0 0 1>}
 base_HairSpike_G_01_l {parent=head, loc=<12.4869 5.0443 -4.8291>, rot=<0.0041 -1 0 0>}
 dyn_HairSpike_G_01_l {parent=base_HairSpike_G_01_l, loc=<0 0 0>, rot=<0 0 0 1>}
 base_HairSpike_H_01_l {parent=head, loc=<14.0049 5.3461 0.8886>, rot=<0.0041 -1 0 0>}
 dyn_HairSpike_H_01_l {parent=base_HairSpike_H_01_l, loc=<0 0 0>, rot=<0 0 0 1>}
 base_HairSpike_A_01_r {parent=head, loc=<4.4458 4.2077 6.6507>, rot=<0.0041 -1 0 0>}
 dyn_HairSpike_A_01_r {parent=base_HairSpike_A_01_r, loc=<0 0 0>, rot=<0 0 0 1>}
 base_HairSpike_B_01_r {parent=head, loc=<8.1868 2.9158 7.4996>, rot=<0.0041 -1 0 0>}
 dyn_HairSpike_B_01_r {parent=base_HairSpike_B_01_r, loc=<0 0 0>, rot=<0 0 0 1>}
 base_HairSpike_C_01_r {parent=head, loc=<6.212 7.0495 6.1391>, rot=<0.0041 -1 0 0>}
 dyn_HairSpike_C_01_r {parent=base_HairSpike_C_01_r, loc=<0 0 0>, rot=<0 0 0 1>}
 base_HairSpike_D_01_r {parent=head, loc=<5.7791 9.1564 4.9875>, rot=<0.0041 -1 0 0>}
 dyn_HairSpike_D_01_r {parent=base_HairSpike_D_01_r, loc=<0 0 0>, rot=<0 0 0 1>}
 base_HairSpike_E_01_r {parent=head, loc=<8.3614 10.0474 3.6366>, rot=<0.0041 -1 0 0>}
 dyn_HairSpike_E_01_r {parent=base_HairSpike_E_01_r, loc=<0 0 0>, rot=<0 0 0 1>}
 base_HairSpike_G_01_r {parent=head, loc=<15.7106 2.1727 2.9611>, rot=<0.0041 -1 0 0>}
 dyn_HairSpike_G_01_r {parent=base_HairSpike_G_01_r, loc=<0 0 0>, rot=<0 0 0 1>}
 base_hair_back_01 {parent=head, loc=<2.5256 -2.4689 -9.5681>, rot=<0.2165 -0.4721 0.0815 -0.8506>}
 dyn_hair_back_01 {parent=base_hair_back_01, loc=<0 0 0>, rot=<0 0 0 1>}
 base_hair_back_02 {parent=head, loc=<-0.9334 -3.9817 -8.794>, rot=<0.1124 -0.5631 -0.1037 -0.8121>}
 dyn_hair_back_02 {parent=base_hair_back_02, loc=<0 0 0>, rot=<0 0 0 1>}
 base_hair_back_03 {parent=head, loc=<-4.0775 -2.167 -8.4862>, rot=<0.1796 -0.6892 0.3038 -0.6328>}
 dyn_hair_back_03 {parent=base_hair_back_03, loc=<0 0 0>, rot=<0 0 0 1>}
 base_hair_back_04 {parent=head, loc=<-4.8 -5.8375 -6.967>, rot=<0.1958 -0.6671 0.1944 -0.692>}
 dyn_hair_back_04 {parent=base_hair_back_04, loc=<0 0 0>, rot=<0 0 0 1>}
 base_hair_back_05 {parent=head, loc=<4.9269 -7.2765 -5.2449>, rot=<0.4781 -0.5256 0.3614 -0.6038>}
 dyn_hair_back_05 {parent=base_hair_back_05, loc=<0 0 0>, rot=<0 0 0 1>}
 base_hair_back_06 {parent=head, loc=<0.8901 -7.7521 -6.3243>, rot=<0.3451 -0.683 0.1502 -0.626>}
 dyn_hair_back_06 {parent=base_hair_back_06, loc=<0 0 0>, rot=<0 0 0 1>}
 base_hair_back_07 {parent=head, loc=<-4.1951 -7.7694 -2.3913>, rot=<0.2003 -0.7561 0.4162 -0.4636>}
 dyn_hair_back_07 {parent=base_hair_back_07, loc=<0 0 0>, rot=<0 0 0 1>}
 base_hair_back_08 {parent=head, loc=<-1.5209 -8.513 -5.7025>, rot=<0.2228 -0.5753 0.2518 -0.7457>}
 dyn_hair_back_08 {parent=base_hair_back_08, loc=<0 0 0>, rot=<0 0 0 1>}
 base_hair_back_09 {parent=head, loc=<-1.992 -9.7902 -1.2837>, rot=<0.3717 -0.5796 0.3447 -0.638>}
 dyn_hair_back_09 {parent=base_hair_back_09, loc=<0 0 0>, rot=<0 0 0 1>}
 base_hair_back_10 {parent=head, loc=<-2.2882 -9.3973 2.4334>, rot=<0.5689 -0.3867 0.589 -0.4241>}
 dyn_hair_back_10 {parent=base_hair_back_10, loc=<0 0 0>, rot=<0 0 0 1>}
 base_hair_back_11 {parent=head, loc=<1.338 -10.3157 2.2132>, rot=<0.6511 -0.3906 0.5224 -0.388>}
 dyn_hair_back_11 {parent=base_hair_back_11, loc=<0 0 0>, rot=<0 0 0 1>}
 base_hair_back_12 {parent=head, loc=<-1.3085 -8.0283 5.9286>, rot=<0.687 -0.2997 0.5517 -0.366>}
 dyn_hair_back_12 {parent=base_hair_back_12, loc=<0 0 0>, rot=<0 0 0 1>}
 base_hair_back_13 {parent=head, loc=<-3.6263 -9.8413 4.359>, rot=<0.7511 -0.1815 0.4588 -0.4386>}
 dyn_hair_back_13 {parent=base_hair_back_13, loc=<0 0 0>, rot=<0 0 0 1>}
 base_hair_back_14 {parent=head, loc=<-4.3382 -4.041 7.8882>, rot=<0.7081 -0.1208 0.6892 -0.0946>}
 dyn_hair_back_14 {parent=base_hair_back_14, loc=<0 0 0>, rot=<0 0 0 1>}
 base_hair_back_15 {parent=head, loc=<-4.8102 -1.3637 7.6811>, rot=<0.7065 0.0069 0.7075 -0.0153>}
 dyn_hair_back_15 {parent=base_hair_back_15, loc=<0 0 0>, rot=<0 0 0 1>}
 base_hair_back_16 {parent=head, loc=<-0.4323 -4.0032 9.4716>, rot=<0.8191 0.1102 0.5606 -0.0506>}
 dyn_hair_back_16 {parent=base_hair_back_16, loc=<0 0 0>, rot=<0 0 0 1>}
 base_hair_back_17 {parent=head, loc=<2.0354 -2.4983 9.9434>, rot=<0.8483 -0.0221 0.5174 -0.1104>}
 dyn_hair_back_17 {parent=base_hair_back_17, loc=<0 0 0>, rot=<0 0 0 1>}
 base_hair_back_18 {parent=head, loc=<5.1142 -0.1204 9.5347>, rot=<0.8082 -0.0145 0.5077 -0.2981>}
 dyn_hair_back_18 {parent=base_hair_back_18, loc=<0 0 0>, rot=<0 0 0 1>}
 base_hair_back_19 {parent=head, loc=<4.8961 -9.3994 3.5997>, rot=<0.5427 -0.3952 0.6616 -0.3341>}
 dyn_hair_back_19 {parent=base_hair_back_19, loc=<0 0 0>, rot=<0 0 0 1>}
 base_hair_back_20 {parent=head, loc=<2.1551 -8.4085 7.1647>, rot=<0.6996 -0.187 0.6073 -0.3269>}
 dyn_hair_back_20 {parent=base_hair_back_20, loc=<0 0 0>, rot=<0 0 0 1>}
 base_hair_back_21 {parent=head, loc=<5.751 -7.9395 6.4144>, rot=<0.6421 -0.3186 0.5298 -0.4533>}
 dyn_hair_back_21 {parent=base_hair_back_21, loc=<0 0 0>, rot=<0 0 0 1>}
 clavicle_l {parent=spine_05, loc=<5.4831 -0.1053 -3.2946>, rot=<0.0465 -0.7289 -0.009 -0.6829>}
 upperarm_l {parent=clavicle_l, loc=<9.8658 -1.2136 -0.4105>, rot=<0.0164 -0.3245 -0.0269 -0.9454>}
 lowerarm_l {parent=upperarm_l, loc=<26.465 -0.8345 0.259>, rot=<0.0036 0.0161 -0.2298 -0.9731>}
 hand_l {parent=lowerarm_l, loc=<22.7355 -0.3203 -0.6105>, rot=<-0.6784 -0.0223 0.0735 -0.7307>}
 thumb_01_l {parent=hand_l, loc=<1.3118 -0.3191 -2.459>, rot=<0.6546 -0.4042 0.0664 -0.6354>}
 thumb_02_l {parent=thumb_01_l, loc=<5.9336 -0.0448 0.0401>, rot=<-0.0037 0.0213 0.298 -0.9543>}
 thumb_03_l {parent=thumb_02_l, loc=<2.2223 0.3682 -0.0907>, rot=<0.0118 -0.013 0.0066 -0.9998>}
 index_metacarpal_l {parent=hand_l, loc=<1.6535 0.0183 -2.96>, rot=<0.0158 -0.0124 -0.0173 -0.9996>}
 index_01_l {parent=index_metacarpal_l, loc=<8.3008 0.0357 -0.0524>, rot=<-0.0044 0.0485 0.0924 -0.9945>}
 index_02_l {parent=index_01_l, loc=<3.3587 -0.3004 0.0452>, rot=<-0.002 0.0027 0.3225 -0.9466>}
 index_03_l {parent=index_02_l, loc=<2.5578 -0.123 -0.0252>, rot=<-0.002 0.0067 0.2558 -0.9667>}
 middle_metacarpal_l {parent=hand_l, loc=<2.1856 -0.1725 -0.5984>, rot=<0.006 0.0168 -0.0329 -0.9993>}
 middle_01_l {parent=middle_metacarpal_l, loc=<7.3275 -0.2447 0.0129>, rot=<-0.0208 0.0042 0.0858 -0.9961>}
 middle_02_l {parent=middle_01_l, loc=<4.0123 0.1401 0.1175>, rot=<-0.0218 0.0258 0.3512 -0.9357>}
 middle_03_l {parent=middle_02_l, loc=<2.6077 0.0534 0.1373>, rot=<0.0082 0.0188 0.1708 -0.9851>}
 ring_metacarpal_l {parent=hand_l, loc=<2.2074 -0.3625 1.3638>, rot=<-0.0284 0.0722 -0.0458 -0.9959>}
 ring_01_l {parent=ring_metacarpal_l, loc=<7.0295 -0.6197 -0.1446>, rot=<-0.0489 -0.0252 0.1892 -0.9804>}
 ring_02_l {parent=ring_01_l, loc=<3.6174 0.4647 -0.1027>, rot=<0 0 0.2806 -0.9598>}
 ring_03_l {parent=ring_02_l, loc=<2.4152 0.114 0.0199>, rot=<0 0 0.1189 -0.9929>}
 pinky_metacarpal_l {parent=hand_l, loc=<1.7616 -0.9753 3.3814>, rot=<-0.0397 0.0814 -0.069 -0.9935>}
 pinky_01_l {parent=pinky_metacarpal_l, loc=<7.1709 -0.4831 -0.1094>, rot=<-0.0654 -0.0035 0.2242 -0.9723>}
 pinky_02_l {parent=pinky_01_l, loc=<2.8274 0.0893 -0.1043>, rot=<0.0037 0.0011 0.2854 -0.9584>}
 pinky_03_l {parent=pinky_02_l, loc=<2.2046 0.0307 -0.0078>, rot=<-0.0002 0.0098 0.1445 -0.9895>}
 deform_wrist_dn_base_l {parent=hand_l, loc=<0 -1.5326 0>, rot=<0 0 0 -1>}
 deform_wrist_dn_l {parent=deform_wrist_dn_base_l, loc=<0 0 0>, rot=<0 0 0 1>}
 deform_wrist_up_base_l {parent=hand_l, loc=<0 2.3009 0>, rot=<0 0 0 -1>}
 deform_wrist_up_l {parent=deform_wrist_up_base_l, loc=<0 0 0>, rot=<0 0 0 1>}
 hand_attach_l {parent=hand_l, loc=<8.5 -3 0>, rot=<0.7071 0.7071 0 0>}
 weapon_l {parent=hand_l, loc=<-0.1974 -3.4111 0.3811>, rot=<0 0 0 1>}
 lowerarm_twist_02_l {parent=lowerarm_l, loc=<10.6707 0 0>, rot=<0 0 0 1>}
 lowerarm_twist_01_l {parent=lowerarm_l, loc=<18.316 0 0>, rot=<0 0 0 1>}
 deform_elbow_in_base_l {parent=lowerarm_l, loc=<0.1834 3.6062 -0.2572>, rot=<0 0 0 -1>}
 deform_elbow_in_l {parent=deform_elbow_in_base_l, loc=<0 0 0>, rot=<0 0 0 1>}
 deform_elbow_out_base_l {parent=lowerarm_l, loc=<0 -4 0>, rot=<0 0 0 1>}
 deform_elbow_out_l {parent=deform_elbow_out_base_l, loc=<0 0 0>, rot=<0 0 0 1>}
 upperarm_twist_01_l {parent=upperarm_l, loc=<7.2135 -0.1134 0>, rot=<0 0 0 1>}
 deform_upperarm_bk_base_l {parent=upperarm_twist_01_l, loc=<-6.4053 -4.3746 0.314>, rot=<0 0 0 -1>}
 deform_upperarm_bk_l {parent=deform_upperarm_bk_base_l, loc=<0 0 0>, rot=<0 0 0 1>}
 deform_upperarm_fr_base_l {parent=upperarm_twist_01_l, loc=<-5.6312 5.092 0.3354>, rot=<0 0 0 -1>}
 deform_upperarm_fr_l {parent=deform_upperarm_fr_base_l, loc=<0 0 0>, rot=<0 0 0 1>}
 deform_upperarm_up_base_l {parent=upperarm_twist_01_l, loc=<-8.891 0.4825 3.9517>, rot=<0 0 0 -1>}
 deform_upperarm_up_l {parent=deform_upperarm_up_base_l, loc=<0 0 0>, rot=<0 0 0 1>}
 upperarm_twist_02_l {parent=upperarm_l, loc=<19.7998 -0.1134 0>, rot=<0 0 0 1>}
 clavicle_r {parent=spine_05, loc=<5.4831 -0.1053 3.2946>, rot=<0.7289 0.0465 -0.6829 0.009>}
 upperarm_r {parent=clavicle_r, loc=<-9.8658 1.2136 0.4105>, rot=<0.0164 -0.3245 -0.0269 -0.9454>}
 lowerarm_r {parent=upperarm_r, loc=<-26.465 0.8345 -0.259>, rot=<0.0036 0.0161 -0.2298 -0.9731>}
 hand_r {parent=lowerarm_r, loc=<-22.7355 0.3203 0.6106>, rot=<-0.6784 -0.0223 0.0735 -0.7307>}
 thumb_01_r {parent=hand_r, loc=<-1.3118 0.3191 2.459>, rot=<0.6546 -0.4042 0.0664 -0.6354>}
 thumb_02_r {parent=thumb_01_r, loc=<-5.9336 0.0448 -0.0401>, rot=<-0.0037 0.0213 0.298 -0.9543>}
 thumb_03_r {parent=thumb_02_r, loc=<-2.2223 -0.3682 0.0907>, rot=<0.0118 -0.013 0.0066 -0.9998>}
 index_metacarpal_r {parent=hand_r, loc=<-1.6535 -0.0183 2.96>, rot=<0.0158 -0.0124 -0.0173 -0.9996>}
 index_01_r {parent=index_metacarpal_r, loc=<-8.3008 -0.0357 0.0524>, rot=<-0.0044 0.0485 0.0924 -0.9945>}
 index_02_r {parent=index_01_r, loc=<-3.3587 0.3004 -0.0452>, rot=<-0.002 0.0027 0.3225 -0.9466>}
 index_03_r {parent=index_02_r, loc=<-2.5578 0.123 0.0252>, rot=<-0.002 0.0067 0.2558 -0.9667>}
 middle_metacarpal_r {parent=hand_r, loc=<-2.1856 0.1725 0.5984>, rot=<0.006 0.0168 -0.0329 -0.9993>}
 middle_01_r {parent=middle_metacarpal_r, loc=<-7.3275 0.2447 -0.0129>, rot=<-0.0208 0.0042 0.0858 -0.9961>}
 middle_02_r {parent=middle_01_r, loc=<-4.0123 -0.1401 -0.1175>, rot=<-0.0218 0.0258 0.3512 -0.9357>}
 middle_03_r {parent=middle_02_r, loc=<-2.6077 -0.0534 -0.1373>, rot=<0.0082 0.0188 0.1708 -0.9851>}
 ring_metacarpal_r {parent=hand_r, loc=<-2.2074 0.3625 -1.3638>, rot=<-0.0284 0.0722 -0.0458 -0.9959>}
 ring_01_r {parent=ring_metacarpal_r, loc=<-7.0295 0.6197 0.1446>, rot=<-0.0489 -0.0252 0.1892 -0.9804>}
 ring_02_r {parent=ring_01_r, loc=<-3.6174 -0.4647 0.1027>, rot=<0 0 0.2806 -0.9598>}
 ring_03_r {parent=ring_02_r, loc=<-2.4152 -0.114 -0.0199>, rot=<0 0 0.1189 -0.9929>}
 pinky_metacarpal_r {parent=hand_r, loc=<-1.7616 0.9753 -3.3814>, rot=<-0.0397 0.0814 -0.069 -0.9935>}
 pinky_01_r {parent=pinky_metacarpal_r, loc=<-7.1709 0.4831 0.1094>, rot=<-0.0654 -0.0035 0.2242 -0.9723>}
 pinky_02_r {parent=pinky_01_r, loc=<-2.8274 -0.0893 0.1043>, rot=<0.0037 0.0011 0.2854 -0.9584>}
 pinky_03_r {parent=pinky_02_r, loc=<-2.2046 -0.0307 0.0078>, rot=<-0.0002 0.0098 0.1445 -0.9895>}
 deform_wrist_dn_base_r {parent=hand_r, loc=<-0.0001 1.5328 0>, rot=<0 0 0 -1>}
 deform_wrist_dn_r {parent=deform_wrist_dn_base_r, loc=<0 0 0>, rot=<0 0 0 1>}
 deform_wrist_up_base_r {parent=hand_r, loc=<-0.0002 -2.3007 -0.0001>, rot=<0 0 0 -1>}
 deform_wrist_up_r {parent=deform_wrist_up_base_r, loc=<0 0 0>, rot=<0 0 0 1>}
 weapon_r {parent=hand_r, loc=<0.1974 3.4111 -0.3811>, rot=<0 0 0 1>}
 hand_attach_r {parent=hand_r, loc=<-8.5001 2.9999 0>, rot=<0 0 -0.7071 -0.7071>}
 lowerarm_twist_02_r {parent=lowerarm_r, loc=<-10.6707 0 0>, rot=<0 0 0 -1>}
 lowerarm_twist_01_r {parent=lowerarm_r, loc=<-18.316 0 0>, rot=<0 0 0 -1>}
 deform_elbow_in_base_r {parent=lowerarm_r, loc=<-0.1834 -3.6062 0.2572>, rot=<0 0 0 -1>}
 deform_elbow_in_r {parent=deform_elbow_in_base_r, loc=<0 0 0>, rot=<0 0 0 1>}
 deform_elbow_out_base_r {parent=lowerarm_r, loc=<0 4 -0.0001>, rot=<0 0 0 -1>}
 deform_elbow_out_r {parent=deform_elbow_out_base_r, loc=<0 0 0>, rot=<0 0 0 1>}
 upperarm_twist_01_r {parent=upperarm_r, loc=<-7.2135 0.1134 0>, rot=<0 0 0 -1>}
 deform_upperarm_bk_base_r {parent=upperarm_twist_01_r, loc=<6.4053 4.3746 -0.314>, rot=<0 0 0 -1>}
 deform_upperarm_bk_r {parent=deform_upperarm_bk_base_r, loc=<0 0 0>, rot=<0 0 0 1>}
 deform_upperarm_fr_base_r {parent=upperarm_twist_01_r, loc=<5.6312 -5.092 -0.3354>, rot=<0 0 0 -1>}
 deform_upperarm_fr_r {parent=deform_upperarm_fr_base_r, loc=<0 0 0>, rot=<0 0 0 1>}
 deform_upperarm_up_base_r {parent=upperarm_twist_01_r, loc=<8.891 -0.4825 -3.9517>, rot=<0 0 0 -1>}
 deform_upperarm_up_r {parent=deform_upperarm_up_base_r, loc=<0 0 0>, rot=<0 0 0 1>}
 upperarm_twist_02_r {parent=upperarm_r, loc=<-19.7998 0.1134 0>, rot=<0 0 0 -1>}
 attach_backpack {parent=spine_05, loc=<-0.1227 -13.4731 0>, rot=<-0.0093 -0.707 0.0093 -0.707>}
 Backpack_BR {parent=attach_backpack, type=Null, loc=<0 3 0>, rot=<0 0 0 1>}
 attach_cape {parent=spine_05, loc=<9.0172 -11.714 0>, rot=<0.0616 -0.7044 -0.0616 -0.7044>}
 cape {parent=attach_cape, type=Null, loc=<0 -2 0>, rot=<0 0 0 1>}
 base_collarStrap_01_l {parent=spine_05, loc=<10.5931 2.1457 -6.332>, rot=<-0.3486 -0.6045 -0.5278 -0.4843>}
 dyn_collarStrap_01_l {parent=base_collarStrap_01_l, loc=<0 0 0>, rot=<0 0 0 1>}
 dyn_collarStrap_02_l {parent=dyn_collarStrap_01_l, loc=<4.0281 0 0>, rot=<-0.004 0.0765 0.1171 -0.9902>}
 dyn_collarStrap_03_l {parent=dyn_collarStrap_02_l, loc=<2.7847 0 0>, rot=<0 0 0 1>}
 base_collarStrap_01_r {parent=spine_05, loc=<10.9687 2.8062 6.9699>, rot=<-0.2137 -0.7412 -0.5096 -0.3812>}
 dyn_collarStrap_01_r {parent=base_collarStrap_01_r, loc=<0 0 0>, rot=<0 0 0 1>}
 dyn_collarStrap_02_r {parent=dyn_collarStrap_01_r, loc=<4.0281 0 0.0002>, rot=<-0.2409 -0.0125 0.0832 -0.9669>}
 dyn_collarStrap_03_r {parent=dyn_collarStrap_02_r, loc=<2.7847 0 -0.0003>, rot=<-0.1567 0.0004 0.0022 -0.9876>}
 lowerarm_ArmFrontDist_null {parent=spine_05, type=Null, loc=<6.5101 24.8367 -13.1587>, rot=<-0.6404 -0.4384 -0.5489 -0.3105>}
 upperarm_l_Spine05DeltaRef_null {parent=spine_05, type=Null, loc=<4.5519 -1.8922 -13.0371>, rot=<-0.0385 0.9118 0.0238 0.4081>}
 deform_pec_aim_base_l {parent=spine_04, loc=<3.2456 9.3309 -1.2237>, rot=<0.0603 -0.4033 0.3318 -0.8507>}
 deform_pec_aim_l {parent=deform_pec_aim_base_l, loc=<0 0 0>, rot=<0 0 0 1>}
 deform_pec_base_l {parent=deform_pec_aim_l, loc=<9.3665 0 0>, rot=<0 0 0 1>}
 deform_pec_l {parent=deform_pec_base_l, loc=<0 0 0>, rot=<0 0 0 1>}
 deform_pec_aim_base_r {parent=spine_04, loc=<3.2452 9.3309 1.2237>, rot=<-0.4033 -0.0603 0.8507 0.3318>}
 deform_pec_aim_r {parent=deform_pec_aim_base_r, loc=<0 0 0>, rot=<0 0 0 1>}
 deform_pec_base_r {parent=deform_pec_aim_r, loc=<-9.3667 -0.0003 -0.0004>, rot=<0 0 0 1>}
 deform_pec_r {parent=deform_pec_base_r, loc=<0 0 0>, rot=<0 0 0 1>}
 base_jacketFlap_01_F_l {parent=spine_04, loc=<-0.7357 9.7981 -8.8831>, rot=<-0.0295 -0.9341 -0.3535 -0.0411>}
 dyn_jacketFlap_01_F_l {parent=base_jacketFlap_01_F_l, loc=<0 0 0>, rot=<0 0 0 1>}
 dyn_jacketFlap_01_F_l_End {parent=dyn_jacketFlap_01_F_l, loc=<5.453 0 0>, rot=<0 0 0 1>}
 base_jacketFlap_01_S_l {parent=spine_04, loc=<-1.6161 -0.0841 -12.4302>, rot=<-0.0459 -0.991 0.1247 -0.0177>}
 dyn_jacketFlap_01_S_l {parent=base_jacketFlap_01_S_l, loc=<0 0 0>, rot=<0 0 0 1>}
 dyn_jacketFlap_01_S_l_End {parent=dyn_jacketFlap_01_S_l, loc=<5.453 0 0>, rot=<0 0 0 1>}
 base_jacketFlap_01_SB_l {parent=spine_04, loc=<-2.5575 -7.4654 -8.0891>, rot=<-0.0845 -0.8342 0.5406 0.069>}
 dyn_jacketFlap_01_SB_l {parent=base_jacketFlap_01_SB_l, loc=<0 0 0>, rot=<0 0 0 1>}
 dyn_jacketFlap_01_SB_l_End {parent=dyn_jacketFlap_01_SB_l, loc=<5.453 0 0>, rot=<0 0 0 1>}
 base_jacketFlap_01_B {parent=spine_04, loc=<-2.7909 -8.8592 0>, rot=<-0.0176 -0.7069 0.7069 0.0176>}
 dyn_jacketFlap_01_B {parent=base_jacketFlap_01_B, loc=<0 0 0>, rot=<0 0 0 1>}
 dyn_jacketFlap_01_B_End {parent=dyn_jacketFlap_01_B, loc=<5.453 0 0>, rot=<0 0 0 1>}
 base_jacketFlap_01_SB_r {parent=spine_04, loc=<-2.5571 -7.4653 8.0891>, rot=<-0.069 -0.5406 0.8342 0.0845>}
 dyn_jacketFlap_01_SB_r {parent=base_jacketFlap_01_SB_r, loc=<0 0 0>, rot=<0 0 0 1>}
 dyn_jacketFlap_01_SB_r_End {parent=dyn_jacketFlap_01_SB_r, loc=<5.4529 0 0>, rot=<0 0 0 1>}
 base_jacketFlap_01_S_r {parent=spine_04, loc=<-1.6164 -0.0842 12.4301>, rot=<0.0177 -0.1247 0.991 0.0459>}
 dyn_jacketFlap_01_S_r {parent=base_jacketFlap_01_S_r, loc=<0 0 0>, rot=<0 0 0 1>}
 dyn_jacketFlap_01_S_r_End {parent=dyn_jacketFlap_01_S_r, loc=<5.4525 -0.0001 0.0001>, rot=<0 0 0 1>}
 base_jacketFlap_01_F_r {parent=spine_04, loc=<-0.7359 9.7981 8.8831>, rot=<0.0411 0.3535 0.9341 0.0295>}
 dyn_jacketFlap_01_F_r {parent=base_jacketFlap_01_F_r, loc=<0 0 0>, rot=<0 0 0 1>}
 dyn_jacketFlap_01_F_r_End {parent=dyn_jacketFlap_01_F_r, loc=<5.4528 0 0>, rot=<0 0 0 -1>}
 spine_04_Spine03_DeltaRef_null {parent=spine_03, type=Null, loc=<6.9598 0.0794 0>, rot=<0 0 -0.0132 -0.9999>}
 thigh_l {parent=pelvis, loc=<-1.0855 -0.2109 -8.7053>, rot=<0.0009 0.0355 0.0155 -0.9993>}
 calf_l {parent=thigh_l, loc=<-42.5284 -0.0019 -0.7297>, rot=<0.0006 -0.0161 -0.0391 -0.9991>}
 foot_l_null {parent=calf_l, type=Null, loc=<-46.7402 -2.5014 -0.592>, rot=<0.0427 -0.0238 0.0353 -0.9982>}
 foot_l {parent=calf_l, loc=<-46.7401 -2.5014 -0.592>, rot=<0.0427 -0.0238 0.0353 -0.9982>}
 ball_l {parent=foot_l, loc=<-3.5658 10.8066 -0.1544>, rot=<0.0001 -0.0001 -0.7063 -0.7079>}
 base_push_ankle_l {parent=foot_l, loc=<-0.0446 -4.1332 0>, rot=<0 0 1 0.0054>}
 dyn_push_ankle_l {parent=base_push_ankle_l, loc=<0 0 0>, rot=<0 0 0 1>}
 calf_twist_02_l {parent=calf_l, loc=<-19.1546 -1.7943 -0.2638>, rot=<0.0237 0.0095 -0.0035 -0.9997>}
 calf_twist_01_l {parent=calf_l, loc=<-35.8483 -1.9197 -0.5794>, rot=<0.0237 0.0095 -0.0035 -0.9997>}
 deform_knee_in_base_l {parent=calf_l, loc=<0.8031 -5.0231 0.2456>, rot=<0 0 0 -1>}
 deform_knee_in_l {parent=deform_knee_in_base_l, loc=<0 0 0>, rot=<0 0 0 1>}
 deform_knee_out_base_l {parent=calf_l, loc=<0 4 0>, rot=<0 0 0 -1>}
 deform_knee_out_l {parent=deform_knee_out_base_l, loc=<0 0 0>, rot=<0 0 0 1>}
 thigh_twist_01_l {parent=thigh_l, loc=<-27.952 0 0>, rot=<0.0013 0 0 -1>}
 deform_groin_base_l {parent=thigh_l, loc=<0.4801 5.6776 -0.2486>, rot=<0 0 0 -1>}
 deform_groin_l {parent=deform_groin_base_l, loc=<0 0 0>, rot=<0 0 0 1>}
 base_legBelt_01_F_l {parent=thigh_l, loc=<-15.2241 6.5295 -3.6446>, rot=<0.0626 -0.9501 0.2508 -0.1749>}
 dyn_legBelt_01_F_l {parent=base_legBelt_01_F_l, loc=<0 0 0>, rot=<0 0 0 1>}
 dyn_legBelt_02_F_l {parent=dyn_legBelt_01_F_l, loc=<1.0447 0.7131 0.3693>, rot=<0 0 0 1>}
 dyn_legBelt_03_F_l {parent=dyn_legBelt_02_F_l, loc=<4.503 0.2327 0.7424>, rot=<0 0 0 1>}
 dyn_legBelt_04_F_l {parent=dyn_legBelt_03_F_l, loc=<7.208 0.2187 -0.9661>, rot=<0 0 0 1>}
 dyn_legBelt_05_F_l {parent=dyn_legBelt_04_F_l, loc=<8.0574 -1.6008 -1.1691>, rot=<0 0 0 1>}
 dyn_legBelt_06_F_l {parent=dyn_legBelt_05_F_l, loc=<4.2282 -6.2485 0.6332>, rot=<0 0 0 1>}
 base_legBelt_01_B_l {parent=thigh_l, loc=<-3.6062 -10.205 -0.4868>, rot=<0.1876 -0.7073 0.1779 -0.658>}
 dyn_legBelt_01_B_l {parent=base_legBelt_01_B_l, loc=<0 0 0>, rot=<0 -0.5373 0 -0.8434>}
 dyn_legBelt_02_B_l {parent=dyn_legBelt_01_B_l, loc=<1.1094 -0.3553 0.1966>, rot=<0 0 0 1>}
 dyn_legBelt_03_B_l {parent=dyn_legBelt_02_B_l, loc=<4.059 -0.3273 -1.3318>, rot=<0 0 0 1>}
 dyn_legBelt_04_B_l {parent=dyn_legBelt_03_B_l, loc=<8.1003 -0.0008 -2.2249>, rot=<0 0 0 1>}
 dyn_legBelt_05_B_l {parent=dyn_legBelt_04_B_l, loc=<7.0454 0.6268 -2.1984>, rot=<0 0 0 1>}
 dyn_legBelt_06_B_l {parent=dyn_legBelt_05_B_l, loc=<7.9553 1.8725 -2.4476>, rot=<0 0 0 1>}
 dyn_legBelt_07_B_l {parent=dyn_legBelt_06_B_l, loc=<5.0796 7.9343 -1.6503>, rot=<0 0 0 1>}
 calf_l_null {parent=thigh_l, type=Null, loc=<-42.5284 -0.0019 -0.7297>, rot=<0.0006 -0.0161 -0.0391 -0.9991>}
 base_legBelt_aimChain_01_l {parent=thigh_l, loc=<-8.8591 -2.3226 -2.068>, rot=<0.0169 -0.9969 0.0015 -0.0771>}
 dyn_legBelt_aimChain_01_l {parent=base_legBelt_aimChain_01_l, loc=<0 0 0>, rot=<0 0 0 1>}
 dyn_legBelt_aimChain_02_l {parent=dyn_legBelt_aimChain_01_l, loc=<10.5405 0 -0.0286>, rot=<0 -0.0039 0 -1>}
 dyn_legBelt_aimChain_03_l {parent=dyn_legBelt_aimChain_02_l, loc=<7.8626 0 -0.0061>, rot=<0 0.0157 0 -0.9999>}
 dyn_legBelt_aimChain_04_l {parent=dyn_legBelt_aimChain_03_l, loc=<6.8952 0 -0.0188>, rot=<0 0.0301 0 -0.9995>}
 dyn_legBelt_aimChain_05_l {parent=dyn_legBelt_aimChain_04_l, loc=<4.6646 0 0>, rot=<0 0 0 1>}
 thigh_r {parent=pelvis, loc=<-1.0855 -0.2109 8.7053>, rot=<0.0355 -0.0009 0.9993 0.0155>}
 calf_r {parent=thigh_r, loc=<42.5284 0.0019 0.7297>, rot=<0.0006 -0.0161 -0.0391 -0.9991>}
 foot_r {parent=calf_r, loc=<46.7401 2.5014 0.592>, rot=<0.0427 -0.0238 0.0353 -0.9982>}
 ball_r {parent=foot_r, loc=<3.5658 -10.8066 0.1544>, rot=<0.0001 -0.0001 -0.7063 -0.7079>}
 base_push_ankle_r {parent=foot_r, loc=<0.0446 4.1332 0>, rot=<-0.0905 0.0052 0.0054 -0.9959>}
 dyn_push_ankle_r {parent=base_push_ankle_r, loc=<0 0 0>, rot=<0 0 0 1>}
 calf_twist_02_r {parent=calf_r, loc=<19.1546 1.7943 0.2638>, rot=<0.0237 0.0095 -0.0035 -0.9997>}
 calf_twist_01_r {parent=calf_r, loc=<35.8483 1.9197 0.5794>, rot=<0.0237 0.0095 -0.0035 -0.9997>}
 deform_knee_in_base_r {parent=calf_r, loc=<-0.8031 5.0231 -0.2457>, rot=<0 0 0 -1>}
 deform_knee_in_r {parent=deform_knee_in_base_r, loc=<0 0 0>, rot=<0 0 0 1>}
 deform_knee_out_base_r {parent=calf_r, loc=<0 -4 0>, rot=<0 0 0 -1>}
 deform_knee_out_r {parent=deform_knee_out_base_r, loc=<0 0 0>, rot=<0 0 0 1>}
 foot_r_null {parent=calf_r, type=Null, loc=<46.7401 2.5014 0.592>, rot=<0.0427 -0.0238 0.0353 -0.9982>}
 thigh_twist_01_r {parent=thigh_r, loc=<27.952 0 0>, rot=<0.0013 0 0 -1>}
 deform_groin_base_r {parent=thigh_r, loc=<-0.4801 -5.6776 0.2486>, rot=<0 0 0 -1>}
 deform_groin_r {parent=deform_groin_base_r, loc=<0 0 0>, rot=<0 0 0 1>}
 base_regBelt_01_F_r {parent=thigh_r, loc=<15.2242 -6.5295 3.6446>, rot=<0.2487 0.2044 -0.0704 -0.9442>}
 dyn_regBelt_01_F_r {parent=base_regBelt_01_F_r, loc=<0 0 0>, rot=<0 0 0 1>}
 dyn_regBelt_02_F_r {parent=dyn_regBelt_01_F_r, loc=<1.0656 -0.7131 0.3037>, rot=<0 0 0 1>}
 dyn_regBelt_03_F_r {parent=dyn_regBelt_02_F_r, loc=<4.5404 -0.2326 0.4608>, rot=<0 0 0 1>}
 dyn_regBelt_04_F_r {parent=dyn_regBelt_03_F_r, loc=<7.134 -0.2188 -1.4126>, rot=<0 0 0 1>}
 dyn_regBelt_05_F_r {parent=dyn_regBelt_04_F_r, loc=<7.9691 1.6008 -1.6681>, rot=<0 0 0 1>}
 dyn_regBelt_06_F_r {parent=dyn_regBelt_05_F_r, loc=<4.2595 6.2485 0.3689>, rot=<0 0 0 1>}
 base_regBelt_01_B_r {parent=thigh_r, loc=<3.6062 10.205 0.4868>, rot=<0.2487 0.2044 -0.0704 -0.9442>}
 dyn_regBelt_01_B_r {parent=base_regBelt_01_B_r, loc=<0 0 0>, rot=<0 0 0 1>}
 dyn_regBelt_02_B_r {parent=dyn_regBelt_01_B_r, loc=<1.1194 0.3553 0.1273>, rot=<0 0 0 1>}
 dyn_regBelt_03_B_r {parent=dyn_regBelt_02_B_r, loc=<3.9682 0.3273 -1.5817>, rot=<0 0 0 1>}
 dyn_regBelt_04_B_r {parent=dyn_regBelt_03_B_r, loc=<7.9462 0.0008 -2.7245>, rot=<0 0 0 1>}
 dyn_regBelt_05_B_r {parent=dyn_regBelt_04_B_r, loc=<6.895 -0.6268 -2.6324>, rot=<0 0 0 1>}
 dyn_regBelt_06_B_r {parent=dyn_regBelt_05_B_r, loc=<7.7876 -1.8726 -2.9377>, rot=<0 0 0 1>}
 dyn_regBelt_07_B_r {parent=dyn_regBelt_06_B_r, loc=<4.9672 -7.9344 -1.9631>, rot=<0 0 0 1>}
 calf_r_null {parent=thigh_r, type=Null, loc=<42.5284 0.0019 0.7297>, rot=<0.0006 -0.0161 -0.0391 -0.9991>}
 base_regBelt_aimChain_01_r {parent=thigh_r, loc=<8.8592 2.3226 2.068>, rot=<0.0015 0.0771 -0.0169 -0.9969>}
 dyn_regBelt_aimChain_01_r {parent=base_regBelt_aimChain_01_r, loc=<0 0 0>, rot=<0 0 0 1>}
 dyn_regBelt_aimChain_02_r {parent=dyn_regBelt_aimChain_01_r, loc=<10.5404 0 -0.0285>, rot=<0 -0.0039 0 -1>}
 dyn_regBelt_aimChain_03_r {parent=dyn_regBelt_aimChain_02_r, loc=<7.8627 0 -0.0061>, rot=<0 0.0157 0 -0.9999>}
 dyn_regBelt_aimChain_04_r {parent=dyn_regBelt_aimChain_03_r, loc=<6.8951 0 -0.0188>, rot=<0 0.0301 0 -0.9995>}
 dyn_regBelt_aimChain_05_r {parent=dyn_regBelt_aimChain_04_r, loc=<4.6647 0 -0.0001>, rot=<0 0 0 1>}
 deform_glute_base_l {parent=pelvis, loc=<-7.3466 -9.0213 -9.1245>, rot=<0.0172 0.031 -0.4477 -0.8935>}
 deform_glute_l {parent=deform_glute_base_l, loc=<0 0 0>, rot=<0 0 0 1>}
 deform_glute_base_r {parent=pelvis, loc=<-7.3464 -9.0213 9.1245>, rot=<0.031 -0.0172 0.8935 -0.4477>}
 deform_glute_r {parent=deform_glute_base_r, loc=<0 0 0>, rot=<0 0 0 1>}
 base_beltStrap_01 {parent=pelvis, loc=<0.6723 6.7325 8.2183>, rot=<0.7141 -0.0661 0.6908 -0.0922>}
 dyn_beltStrap_01 {parent=base_beltStrap_01, loc=<0 0 0>, rot=<0 0 0 1>}
 dyn_beltStrap_02 {parent=dyn_beltStrap_01, loc=<4.3695 -0.051 -1.2633>, rot=<-0.0731 -0.256 0.2933 -0.9182>}
 dyn_beltStrap_03 {parent=dyn_beltStrap_02, loc=<2.9635 0 0>, rot=<-0.0821 -0.0291 0.3331 -0.9389>}
 thigh_r_PelvisDelta_null {parent=pelvis, type=Null, loc=<-1.0855 -0.2109 8.7053>, rot=<0.0355 -0.0009 0.9993 0.0155>}
 thigh_l_null {parent=pelvis, type=Null, loc=<-1.0855 -0.2109 -8.7053>, rot=<0.0009 0.0355 0.0155 -0.9993>}
 thigh_r_null {parent=pelvis, type=Null, loc=<-1.0855 -0.2109 8.7053>, rot=<0.0355 -0.0009 0.9993 0.0155>}
 spine_05_null {parent=pelvis, type=Null, loc=<32.9717 -4.2232 0>, rot=<0 0 0.058 -0.9983>}
 pelvis_null {parent=root, type=Null, loc=<0 -0.5301 98.0571>, rot=<0.0127 0.707 -0.0127 -0.707>}
 ik_hand_root {parent=root, loc=<0 0 0>, rot=<0 0 0 1>}
 ik_hand_gun {parent=ik_hand_root, loc=<-48.1743 3.3545 104.4381>, rot=<0.6276 0.1309 0.4183 -0.6434>}
 ik_hand_r {parent=ik_hand_gun, loc=<0 0 0>, rot=<0 0 0 1>}
 ik_hand_l {parent=ik_hand_gun, loc=<59.3271 67.6997 34.3524>, rot=<-0.6158 -0.7027 -0.3565 0>}
 ik_foot_root {parent=root, loc=<0 0 0>, rot=<0 0 0 1>}
 ik_foot_l {parent=ik_foot_root, loc=<14.8262 -7.3545 8.296>, rot=<0.0357 0.7045 0.0284 -0.7082>}
 ik_foot_r {parent=ik_foot_root, loc=<-14.8262 -7.3545 8.296>, rot=<0.7082 0.0284 -0.7045 0.0357>}
 root_fp {parent=root, loc=<0 0 0>, rot=<0 0 0 1>}
 attach_fp {parent=root_fp, loc=<0 0 0>, rot=<0 0 0 1>}
 camera_root {parent=root_fp, loc=<0 0 0>, rot=<0 0 0 1>}
 camera {parent=camera_root, loc=<0 0 0>, rot=<0 0 0 1>}
]
variables [
 IdleThigh_L {type=double, default=0}
 IdleThigh_R {type=double, default=0}
 PelvisProx {type=double, default=0}
 Pub_GrindRail {type=double, default=0}
 Pub_Is_BaseballSliding {type=double, default=0}
 Pub_Is_Crouching {type=double, default=0}
 Pub_Is_KneeSliding {type=double, default=0}
 Pub_LeftStrafe {type=double, default=0}
 Pub_SwingPickaxe {type=double, default=0}
 SPR L Calf Bkd {type=double, default=0}
 SPR L Calf Bkd_Max {type=double, default=0}
 SPR L Leg Bkd {type=double, default=0}
 SPR L Leg Fwd {type=double, default=0}
 SPR L Leg Otw {type=double, default=0}
 SPR R Calf Bkd {type=double, default=0}
 SPR R Calf Bkd_Max {type=double, default=0}
 SPR R Leg Bkd {type=double, default=0}
 SPR R Leg Fwd {type=double, default=0}
 SPR R Leg Otw {type=double, default=0}
]
