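# Physics asset of cape colliders for the GeminiLink female medium character body, named as the collision set for its cape.
# Capsules cover the head, clavicles, spine, pelvis and feet, while the arm, thigh, calf and neck bodies carry no shapes.
# Two six-link dynamic leg-belt chains, one per thigh, add small capsules along the back of each side.
# Constraints link bones parent to child with all swing and twist axes free.
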PhysicsAsset /BRCosmetics/Characters/Player/Female/Medium/Bodies/F_MED_GeminiLink/Meshes/F_MED_GeminiLink_CapeColliders
bodies {
 calf_l []
 calf_r []
 clavicle_l [{shapeType=Capsule, length=6.5623, radius=7.5194, center=<3.9566 -1.0862 -2.1127>, rotation=<rot -77.5566 177.3915 7.7314>}]
 clavicle_r [{shapeType=Capsule, length=6.5623, radius=7.5194, center=<-3.9566 1.0862 2.1127>, rotation=<rot -77.5566 177.3915 -172.2686>}]
 dyn_legBelt_01_B_l [{shapeType=Capsule, length=0.505, radius=3, shapeName=dyn_legBelt_01_B_l_capsule, rotation=<rot 0 -90 0>}]
 dyn_legBelt_02_B_l [{
   shapeType=Capsule
   length=5.0021
   radius=3
   shapeName=dyn_legBelt_02_B_l_capsule
   center=<4.1563 -0.1648 -1.2468>
   rotation=<rot 0 -94.9036 -73.0891>
  }]
 dyn_legBelt_03_B_l [{
   shapeType=Capsule
   length=1.0997
   radius=3
   shapeName=dyn_legBelt_03_B_l_capsule
   center=<6.6423 -0.082 -1.7997>
   rotation=<rot 0 166.5673 -46.736>
  }]
 dyn_legBelt_04_B_l [{
   shapeType=Capsule
   length=3.1447
   radius=3
   shapeName=dyn_legBelt_04_B_l_capsule
   center=<3.4516 0.3747 -1.2239>
   rotation=<rot 0 -83.4381 -71.7746>
  }]
 dyn_legBelt_05_B_l [{
   shapeType=Capsule
   length=7.522
   radius=3
   shapeName=dyn_legBelt_05_B_l_capsule
   center=<6.9483 3.0262 -2.1128>
   rotation=<rot 0 -61.6311 -76.0366>
  }]
 dyn_legBelt_06_B_l [{
   shapeType=Capsule
   length=0.505
   radius=3
   shapeName=dyn_legBelt_06_B_l_capsule
   center=<4.7295 6.3168 -1.4921>
   rotation=<rot 6.8953 -90 -5.1262>
  }]
 dyn_regBelt_01_B_r [{shapeType=Capsule, length=0.505, radius=3, shapeName=dyn_regBelt_01_B_r_capsule, rotation=<rot 0 -90 0>}]
 dyn_regBelt_02_B_r [{
   shapeType=Capsule
   length=7.9666
   radius=3
   shapeName=dyn_regBelt_02_B_r_capsule
   center=<6.2798 0.1963 -2.3339>
   rotation=<rot 0 -86.9901 -69.4371>
  }]
 dyn_regBelt_03_B_r [{shapeType=Capsule, length=0.505, radius=3, shapeName=dyn_regBelt_03_B_r_capsule, rotation=<rot 0 -90 0>}]
 dyn_regBelt_04_B_r [{shapeType=Capsule, length=6, radius=3, shapeName=dyn_regBelt_04_B_r_capsule, rotation=<rot 0.0001 -90.0002 -68.5535>}]
 dyn_regBelt_05_B_r [{
   shapeType=Capsule
   length=12
   radius=3
   shapeName=dyn_regBelt_05_B_r_capsule
   center=<4.0345 -1.907 -1.5387>
   rotation=<rot 0 -109.8786 -71.8954>
  }]
 dyn_regBelt_06_B_r [{
   shapeType=Capsule
   length=6
   radius=3
   shapeName=dyn_regBelt_06_B_r_capsule
   center=<4.1501 -8.1142 -1.8102>
   rotation=<rot 0 173.6783 -86.0744>
  }]
 foot_l [{shapeType=Capsule, length=11.0765, radius=8.0344, center=<-4.1811 6.0027 -0.338>, rotation=<rot 0.0388 1.4701 -89.3982>}]
 foot_r [{shapeType=Capsule, length=11.0765, radius=8.0344, center=<4.1811 -6.0027 0.338>, rotation=<rot 0.0388 1.4701 90.6018>}]
 head [{shapeType=Capsule, length=0, radius=18, center=<9.0697 0.0746 0>, rotation=<rot 0 -90 0>}]
 lowerarm_l []
 lowerarm_r []
 neck_02 []
 pelvis [{shapeType=Capsule, length=9.6042, radius=11.1441, center=<-2.4471 -2.5289 -0.1981>}]
 spine_01 [{shapeType=Capsule, length=4.3888, radius=14.3638, center=<-8.9021 -0.8973 0>}]
 spine_02 [
  {shapeType=Capsule, length=2.5219, radius=9.7317, center=<-2.6885 -1.2641 0>}
  {shapeType=Capsule, length=20.6875, radius=9.3125, center=<-1.8317 1.9362 0>, rotation=<rot 90 0 0>}
 ]
 spine_03 [
  {shapeType=Capsule, length=10.3089, radius=8.4854, center=<0.0574 0.4416 0>}
  {shapeType=Capsule, length=20.7188, radius=9.2812, center=<-1.2163 0.549 0>, rotation=<rot 90 180 -175.0732>}
 ]
 spine_04 [
  {shapeType=Capsule, length=8.8875, radius=10.5117, center=<5.6234 -0.6779 0>}
  {shapeType=Capsule, length=7.6026, radius=11.7099, center=<-1.2326 1.308 0>, rotation=<rot 90 0 0>}
 ]
 spine_05 [{shapeType=Capsule, length=9.1103, radius=12.9248, center=<0.6969 1.91 0>}]
 thigh_l []
 thigh_r []
 upperarm_l []
 upperarm_r []
}
constraints [
 {
  bone1Name=spine_01
  bone2Name=pelvis
  swing1LimitDegrees=45
  swing1Motion=Free
  swing2LimitDegrees=45
  swing2Motion=Free
  twistLimitDegrees=45
  twistMotion=Free
 }
 {
  bone1Name=spine_02
  bone2Name=spine_01
  swing1LimitDegrees=45
  swing1Motion=Free
  swing2LimitDegrees=45
  swing2Motion=Free
  twistLimitDegrees=45
  twistMotion=Free
 }
 {
  bone1Name=spine_03
  bone2Name=spine_02
  swing1LimitDegrees=45
  swing1Motion=Free
  swing2LimitDegrees=45
  swing2Motion=Free
  twistLimitDegrees=45
  twistMotion=Free
 }
 {
  bone1Name=spine_04
  bone2Name=spine_03
  swing1LimitDegrees=45
  swing1Motion=Free
  swing2LimitDegrees=45
  swing2Motion=Free
  twistLimitDegrees=45
  twistMotion=Free
 }
 {
  bone1Name=spine_05
  bone2Name=spine_04
  swing1LimitDegrees=45
  swing1Motion=Free
  swing2LimitDegrees=45
  swing2Motion=Free
  twistLimitDegrees=45
  twistMotion=Free
 }
 {
  bone1Name=clavicle_l
  bone2Name=spine_05
  swing1LimitDegrees=45
  swing1Motion=Free
  swing2LimitDegrees=45
  swing2Motion=Free
  twistLimitDegrees=45
  twistMotion=Free
 }
 {
  bone1Name=upperarm_l
  bone2Name=clavicle_l
  swing1LimitDegrees=45
  swing1Motion=Free
  swing2LimitDegrees=45
  swing2Motion=Free
  twistLimitDegrees=45
  twistMotion=Free
 }
 {
  bone1Name=lowerarm_l
  bone2Name=upperarm_l
  swing1LimitDegrees=45
  swing1Motion=Free
  swing2LimitDegrees=45
  swing2Motion=Free
  twistLimitDegrees=45
  twistMotion=Free
 }
 {
  bone1Name=neck_02
  bone2Name=spine_05
  swing1LimitDegrees=45
  swing1Motion=Free
  swing2LimitDegrees=45
  swing2Motion=Free
  twistLimitDegrees=45
  twistMotion=Free
 }
 {
  bone1Name=head
  bone2Name=neck_02
  swing1LimitDegrees=45
  swing1Motion=Free
  swing2LimitDegrees=45
  swing2Motion=Free
  twistLimitDegrees=45
  twistMotion=Free
 }
 {
  bone1Name=thigh_l
  bone2Name=pelvis
  swing1LimitDegrees=45
  swing1Motion=Free
  swing2LimitDegrees=45
  swing2Motion=Free
  twistLimitDegrees=45
  twistMotion=Free
 }
 {
  bone1Name=foot_l
  bone2Name=calf_l
  swing1LimitDegrees=45
  swing1Motion=Free
  swing2LimitDegrees=45
  swing2Motion=Free
  twistLimitDegrees=45
  twistMotion=Free
 }
 {
  bone1Name=lowerarm_r
  bone2Name=upperarm_r
  swing1LimitDegrees=45
  swing1Motion=Free
  swing2LimitDegrees=45
  swing2Motion=Free
  twistLimitDegrees=45
  twistMotion=Free
 }
 {
  bone1Name=upperarm_r
  bone2Name=clavicle_r
  swing1LimitDegrees=45
  swing1Motion=Free
  swing2LimitDegrees=45
  swing2Motion=Free
  twistLimitDegrees=45
  twistMotion=Free
 }
 {
  bone1Name=clavicle_r
  bone2Name=spine_05
  swing1LimitDegrees=45
  swing1Motion=Free
  swing2LimitDegrees=45
  swing2Motion=Free
  twistLimitDegrees=45
  twistMotion=Free
 }
 {
  bone1Name=thigh_r
  bone2Name=pelvis
  swing1LimitDegrees=45
  swing1Motion=Free
  swing2LimitDegrees=45
  swing2Motion=Free
  twistLimitDegrees=45
  twistMotion=Free
 }
 {
  bone1Name=foot_r
  bone2Name=calf_r
  swing1LimitDegrees=45
  swing1Motion=Free
  swing2LimitDegrees=45
  swing2Motion=Free
  twistLimitDegrees=45
  twistMotion=Free
 }
 {
  bone1Name=calf_r
  bone2Name=thigh_r
  swing1LimitDegrees=45
  swing1Motion=Free
  swing2LimitDegrees=45
  swing2Motion=Free
  twistLimitDegrees=45
  twistMotion=Free
 }
 {
  bone1Name=calf_l
  bone2Name=thigh_l
  swing1LimitDegrees=45
  swing1Motion=Free
  swing2LimitDegrees=45
  swing2Motion=Free
  twistLimitDegrees=45
  twistMotion=Free
 }
 {
  bone1Name=dyn_legBelt_01_B_l
  bone2Name=thigh_l
  swing1LimitDegrees=45
  swing1Motion=Free
  swing2LimitDegrees=45
  swing2Motion=Free
  twistLimitDegrees=45
  twistMotion=Free
 }
 {
  bone1Name=dyn_legBelt_02_B_l
  bone2Name=dyn_legBelt_01_B_l
  swing1LimitDegrees=45
  swing1Motion=Free
  swing2LimitDegrees=45
  swing2Motion=Free
  twistLimitDegrees=45
  twistMotion=Free
 }
 {
  bone1Name=dyn_legBelt_03_B_l
  bone2Name=dyn_legBelt_02_B_l
  swing1LimitDegrees=45
  swing1Motion=Free
  swing2LimitDegrees=45
  swing2Motion=Free
  twistLimitDegrees=45
  twistMotion=Free
 }
 {
  bone1Name=dyn_legBelt_04_B_l
  bone2Name=dyn_legBelt_03_B_l
  swing1LimitDegrees=45
  swing1Motion=Free
  swing2LimitDegrees=45
  swing2Motion=Free
  twistLimitDegrees=45
  twistMotion=Free
 }
 {
  bone1Name=dyn_legBelt_05_B_l
  bone2Name=dyn_legBelt_04_B_l
  swing1LimitDegrees=45
  swing1Motion=Free
  swing2LimitDegrees=45
  swing2Motion=Free
  twistLimitDegrees=45
  twistMotion=Free
 }
 {
  bone1Name=dyn_legBelt_06_B_l
  bone2Name=dyn_legBelt_05_B_l
  swing1LimitDegrees=45
  swing1Motion=Free
  swing2LimitDegrees=45
  swing2Motion=Free
  twistLimitDegrees=45
  twistMotion=Free
 }
 {
  bone1Name=dyn_regBelt_01_B_r
  bone2Name=thigh_r
  swing1LimitDegrees=45
  swing1Motion=Free
  swing2LimitDegrees=45
  swing2Motion=Free
  twistLimitDegrees=45
  twistMotion=Free
 }
 {
  bone1Name=dyn_regBelt_02_B_r
  bone2Name=dyn_regBelt_01_B_r
  swing1LimitDegrees=45
  swing1Motion=Free
  swing2LimitDegrees=45
  swing2Motion=Free
  twistLimitDegrees=45
  twistMotion=Free
 }
 {
  bone1Name=dyn_regBelt_03_B_r
  bone2Name=dyn_regBelt_02_B_r
  swing1LimitDegrees=45
  swing1Motion=Free
  swing2LimitDegrees=45
  swing2Motion=Free
  twistLimitDegrees=45
  twistMotion=Free
 }
 {
  bone1Name=dyn_regBelt_04_B_r
  bone2Name=dyn_regBelt_03_B_r
  swing1LimitDegrees=45
  swing1Motion=Free
  swing2LimitDegrees=45
  swing2Motion=Free
  twistLimitDegrees=45
  twistMotion=Free
 }
 {
  bone1Name=dyn_regBelt_05_B_r
  bone2Name=dyn_regBelt_04_B_r
  swing1LimitDegrees=45
  swing1Motion=Free
  swing2LimitDegrees=45
  swing2Motion=Free
  twistLimitDegrees=45
  twistMotion=Free
 }
 {
  bone1Name=dyn_regBelt_06_B_r
  bone2Name=dyn_regBelt_05_B_r
  swing1LimitDegrees=45
  swing1Motion=Free
  swing2LimitDegrees=45
  swing2Motion=Free
  twistLimitDegrees=45
  twistMotion=Free
 }
]
