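# Post-dynamics Control Rig for the female GeminiLink body that shapes the front and back belt straps on the left leg.
# Construction spawns a control joint and control per listed belt bone.
# Swing-twist delta functions turn thigh and calf rotation into weights that offset an aim chain.
# Control joints follow that chain, receive thigh-back and crouch-weighted corrections, then drive B-spline Spline IK on the belt bones.
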ControlRigBlueprint /BRCosmetics/Characters/Player/Female/Medium/Bodies/F_MED_GeminiLink/Meshes/F_MED_GeminiLink_POSTDynamics_CtrlRig
graphs [
 New Function_ContainedGraph {nodes [
   n1 = RigVMFunctionEntryNode()
   RigVMFunction_MathBoolToInteger = &RigVMFunction_MathBoolToInteger(Value=@n1.onlyTwist)
   RigVMFunction_MathQuaternionSwingTwist = &RigVMFunction_MathQuaternionSwingTwist(
    TwistAxis=@n1.twistAxis
    Input=&RigUnit_GetRelativeTransformForItem(bChildInitial=False, bParentInitial=False, Child=@n1.child, Parent=@n1.parent).RelativeTransform.Rotation
   )
   n2 = RigVMDispatchNode(
    Values=((),())
    Index=@RigVMFunction_MathBoolToInteger.Result
    Values.0=@RigVMFunction_MathQuaternionSwingTwist.Swing
    Values.1=@RigVMFunction_MathQuaternionSwingTwist.Twist
   ) {template="DISPATCH_RigVMDispatch_SelectInt32(in Index,in Values,out Result)"}
   RigVMFunction_MathQuaternionToEuler = &RigVMFunction_MathQuaternionToEuler(Value=@n2.Result, RotationOrder=@n1.Rotation Order)
   > DISPATCH_RigVMDispatch_Print = RigVMDispatchNode(ScreenColor=(R=1,G=1,B=1,A=1), ScreenDuration=1.5, Enabled=@n1.debugRotation, Value=@RigVMFunction_MathQuaternionToEuler.Result) {template="DISPATCH_RigVMDispatch_Print(in Prefix,in Value,in Enabled,in ScreenDuration,in ScreenColor)"}
   Cast = RigVMDispatchNode(Value=@n1.triggerAxis) {template="DISPATCH_RigVMDispatch_CastEnumToInt(in Value,out Result)"}
   DISPATCH_RigVMDispatch_SelectInt32 = RigVMDispatchNode(
    Values=(0,0,0)
    Index=@Cast.Result
    Values.0=@RigVMFunction_MathQuaternionToEuler.Result.X
    Values.1=@RigVMFunction_MathQuaternionToEuler.Result.Y
    Values.2=@RigVMFunction_MathQuaternionToEuler.Result.Z
   ) {template="DISPATCH_RigVMDispatch_SelectInt32(in Index,in Values,out Result)"}
   Remap = &RigVMFunction_MathDoubleRemap(
    bClamp=False
    Value=@DISPATCH_RigVMDispatch_SelectInt32.Result
    TargetMinimum=@n1.outputMinMax.X
    TargetMaximum=@n1.outputMinMax.Y
    SourceMinimum=@n1.rotationMinMax.X
    SourceMaximum=@n1.rotationMinMax.Y
   )
   Absolute = &RigVMFunction_MathDoubleAbs(Value=@Remap.Result)
   Multiply = &RigVMFunction_MathDoubleMul(
    A=&RigVMFunction_MathDoubleClamp(Value=RigVMDispatchNode(True=@Absolute.Result, Condition=@n1.absoluteValue, False=@Remap.Result) {template="DISPATCH_RigVMDispatch_If(in Condition,in True,in False,out Result)"}.Result).Result
    B=@n1.premult
   )
   > RigVMFunctionReturnNode(outputQuat=@n2.Result, outputWeight=@Multiply.Result)
  ]}
 New Function_ContainedGraph {nodes [
   n1 = RigVMFunctionEntryNode()
   RigVMFunction_MathBoolToInteger = &RigVMFunction_MathBoolToInteger(Value=@n1.onlyTwist)
   RigVMFunction_MathQuaternionSwingTwist = &RigVMFunction_MathQuaternionSwingTwist(
    TwistAxis=@n1.twistAxis
    Input=&RigUnit_GetRelativeTransformForItem(bChildInitial=False, bParentInitial=False, Child=@n1.child, Parent=@n1.parent).RelativeTransform.Rotation
   )
   n2 = RigVMDispatchNode(
    Values=((),())
    Index=@RigVMFunction_MathBoolToInteger.Result
    Values.0=@RigVMFunction_MathQuaternionSwingTwist.Swing
    Values.1=@RigVMFunction_MathQuaternionSwingTwist.Twist
   ) {template="DISPATCH_RigVMDispatch_SelectInt32(in Index,in Values,out Result)"}
   RigVMFunction_MathQuaternionToEuler = &RigVMFunction_MathQuaternionToEuler(Value=@n2.Result, RotationOrder=@n1.Rotation Order)
   > DISPATCH_RigVMDispatch_Print = RigVMDispatchNode(ScreenColor=(R=1,G=1,B=1,A=1), ScreenDuration=1.5, Enabled=@n1.debugRotation, Value=@RigVMFunction_MathQuaternionToEuler.Result) {template="DISPATCH_RigVMDispatch_Print(in Prefix,in Value,in Enabled,in ScreenDuration,in ScreenColor)"}
   Cast = RigVMDispatchNode(Value=@n1.triggerAxis) {template="DISPATCH_RigVMDispatch_CastEnumToInt(in Value,out Result)"}
   DISPATCH_RigVMDispatch_SelectInt32 = RigVMDispatchNode(
    Values=(0,0,0)
    Index=@Cast.Result
    Values.0=@RigVMFunction_MathQuaternionToEuler.Result.X
    Values.1=@RigVMFunction_MathQuaternionToEuler.Result.Y
    Values.2=@RigVMFunction_MathQuaternionToEuler.Result.Z
   ) {template="DISPATCH_RigVMDispatch_SelectInt32(in Index,in Values,out Result)"}
   Remap = &RigVMFunction_MathDoubleRemap(
    bClamp=False
    Value=@DISPATCH_RigVMDispatch_SelectInt32.Result
    TargetMinimum=@n1.outputMinMax.X
    TargetMaximum=@n1.outputMinMax.Y
    SourceMinimum=@n1.rotationMinMax.X
    SourceMaximum=@n1.rotationMinMax.Y
   )
   Absolute = &RigVMFunction_MathDoubleAbs(Value=@Remap.Result)
   Multiply = &RigVMFunction_MathDoubleMul(
    A=&RigVMFunction_MathDoubleClamp(Value=RigVMDispatchNode(True=@Absolute.Result, Condition=@n1.absoluteValue, False=@Remap.Result) {template="DISPATCH_RigVMDispatch_If(in Condition,in True,in False,out Result)"}.Result).Result
    B=@n1.premult
   )
   > RigVMFunctionReturnNode(outputQuat=@n2.Result, outputWeight=@Multiply.Result)
  ]}
 New Function_ContainedGraph {nodes [
   n1 = RigVMFunctionEntryNode()
   RigVMFunction_MathBoolToInteger = &RigVMFunction_MathBoolToInteger(Value=@n1.onlyTwist)
   RigVMFunction_MathQuaternionSwingTwist = &RigVMFunction_MathQuaternionSwingTwist(
    TwistAxis=@n1.twistAxis
    Input=&RigUnit_GetRelativeTransformForItem(bChildInitial=False, bParentInitial=False, Child=@n1.child, Parent=@n1.parent).RelativeTransform.Rotation
   )
   n2 = RigVMDispatchNode(
    Values=((),())
    Index=@RigVMFunction_MathBoolToInteger.Result
    Values.0=@RigVMFunction_MathQuaternionSwingTwist.Swing
    Values.1=@RigVMFunction_MathQuaternionSwingTwist.Twist
   ) {template="DISPATCH_RigVMDispatch_SelectInt32(in Index,in Values,out Result)"}
   RigVMFunction_MathQuaternionToEuler = &RigVMFunction_MathQuaternionToEuler(Value=@n2.Result, RotationOrder=@n1.Rotation Order)
   > DISPATCH_RigVMDispatch_Print = RigVMDispatchNode(ScreenColor=(R=1,G=1,B=1,A=1), ScreenDuration=1.5, Enabled=@n1.debugRotation, Value=@RigVMFunction_MathQuaternionToEuler.Result) {template="DISPATCH_RigVMDispatch_Print(in Prefix,in Value,in Enabled,in ScreenDuration,in ScreenColor)"}
   Cast = RigVMDispatchNode(Value=@n1.triggerAxis) {template="DISPATCH_RigVMDispatch_CastEnumToInt(in Value,out Result)"}
   DISPATCH_RigVMDispatch_SelectInt32 = RigVMDispatchNode(
    Values=(0,0,0)
    Index=@Cast.Result
    Values.0=@RigVMFunction_MathQuaternionToEuler.Result.X
    Values.1=@RigVMFunction_MathQuaternionToEuler.Result.Y
    Values.2=@RigVMFunction_MathQuaternionToEuler.Result.Z
   ) {template="DISPATCH_RigVMDispatch_SelectInt32(in Index,in Values,out Result)"}
   Remap = &RigVMFunction_MathDoubleRemap(
    bClamp=False
    Value=@DISPATCH_RigVMDispatch_SelectInt32.Result
    TargetMinimum=@n1.outputMinMax.X
    TargetMaximum=@n1.outputMinMax.Y
    SourceMinimum=@n1.rotationMinMax.X
    SourceMaximum=@n1.rotationMinMax.Y
   )
   Absolute = &RigVMFunction_MathDoubleAbs(Value=@Remap.Result)
   Multiply = &RigVMFunction_MathDoubleMul(
    A=&RigVMFunction_MathDoubleClamp(Value=RigVMDispatchNode(True=@Absolute.Result, Condition=@n1.absoluteValue, False=@Remap.Result) {template="DISPATCH_RigVMDispatch_If(in Condition,in True,in False,out Result)"}.Result).Result
    B=@n1.premult
   )
   > RigVMFunctionReturnNode(outputQuat=@n2.Result, outputWeight=@Multiply.Result)
  ]}
 RigVMFunctionLibrary {links [], nodes []}
 RigVMFunction_Sequence_1_ContainedGraph {nodes [
   n1 = RigVMFunctionEntryNode()
   RigVMFunction_Sequence = &RigVMFunction_Sequence(ExecuteContext=@n1.ExecuteContext)
   n2 = &RigVMFunction_Sequence(ExecuteContext=@RigVMFunction_Sequence.B)
   n3 = &RigVMFunction_Sequence(ExecuteContext=@n2.B)
   RigVMFunctionReturnNode(A=@RigVMFunction_Sequence.A, B=@n2.A, C=@n3.A, D=@n3.B)
  ]}
 RigVMFunction_Sequence_3_ContainedGraph {nodes [
   n1 = RigVMFunctionEntryNode()
   n2 = &RigVMFunction_Sequence(ExecuteContext=@n1.ExecuteContext)
   n3 = &RigVMFunction_Sequence(ExecuteContext=@n2.B)
   n4 = &RigVMFunction_Sequence(ExecuteContext=@n3.B)
   RigVMFunctionReturnNode(B=@n3.A, A=@n2.A, C=@n4.A, D=@n4.B)
  ]}
 RigVMFunction_Sequence_4_ContainedGraph {nodes [
   n1 = RigVMFunctionEntryNode()
   n2 = &RigVMFunction_Sequence(ExecuteContext=@n1.ExecuteContext)
   n3 = &RigVMFunction_Sequence(ExecuteContext=@n2.B)
   RigVMFunctionReturnNode(B=@n3.A, C=@n3.B, A=@n2.A)
  ]}
 RigVMModel {nodes [
   RigUnit_BeginExecution = &RigUnit_BeginExecution()
   RigVMFunction_Sequence_3 = RigVMAggregateNode(ExecuteContext=@RigUnit_BeginExecution.ExecutePin)
   n1 = &RigVMFunction_Sequence(ExecuteContext=@RigVMFunction_Sequence_3.B)
   RigVMFunction_Sequence = RigVMAggregateNode(ExecuteContext=@n1.A)
   &RigUnit_ParentConstraint(
    AdvancedSettings=(InterpolationType=Average,RotationOrderForFilter=XZY)
    Child=(Type=Bone,Name="dyn_legBelt_03_F_l_CtrlJnt")
    Parents=((Item=(Type=Bone,Name="dyn_legBelt_aimChain_02_l"),Weight=1))
    Weight=1
    ExecutePin=@RigVMFunction_Sequence.A
   )
   > &RigUnit_ParentConstraint(
    AdvancedSettings=(InterpolationType=Average,RotationOrderForFilter=XZY)
    Child=(Type=Bone,Name="dyn_legBelt_04_F_l_CtrlJnt")
    Parents=((Item=(Type=Bone,Name="dyn_legBelt_aimChain_03_l"),Weight=1))
    Weight=1
   )
   > &RigUnit_ParentConstraint(
    AdvancedSettings=(InterpolationType=Average,RotationOrderForFilter=XZY)
    Child=(Type=Bone,Name="dyn_legBelt_05_F_l_CtrlJnt")
    Parents=((Item=(Type=Bone,Name="dyn_legBelt_aimChain_04_l"),Weight=1))
    Weight=1
   )
   > &RigUnit_ParentConstraint(
    AdvancedSettings=(InterpolationType=Average,RotationOrderForFilter=XZY)
    Child=(Type=Bone,Name="dyn_legBelt_06_F_l_CtrlJnt")
    Parents=((Item=(Type=Bone,Name="dyn_legBelt_aimChain_05_l"),Weight=1))
    Weight=1
   )
   &RigUnit_OffsetTransformForItem(
    Item=(Type=Bone,Name="dyn_legBelt_05_F_l_CtrlJnt")
    OffsetTransform=(Rotation=(X=0,Y=0,Z=0,W=-1),Translation=(X=-3,Y=0,Z=4),Scale3D=(X=1,Y=1,Z=1))
    Weight=1
    bPropagateToChildren=true
    ExecutePin=@RigVMFunction_Sequence.B
   )
   > &RigUnit_OffsetTransformForItem(
    Item=(Type=Bone,Name="dyn_legBelt_04_F_l_CtrlJnt")
    OffsetTransform=(Rotation=(X=0,Y=0,Z=0,W=-1),Translation=(X=0,Y=0,Z=-1),Scale3D=(X=1,Y=1,Z=1))
    Weight=1
    bPropagateToChildren=true
   )
   &RigUnit_ModifyTransforms(
    ItemToModify="((Item=(Type=Bone,Name=\"dyn_legBelt_04_F_l_CtrlJnt\"),Transform=(Rotation=(X=0,Y=0,Z=0,W=1),Translation=(X=-4,Y=0,Z=-5),Scale3D=(X=1,Y=1,Z=1))),(Item=(Type=Bone,Name=\"dyn_legBelt_05_F_l_CtrlJnt\"),Transform=(Rotation=(X=0,Y=0,Z=0,W=1),Translation=(X=3,Y=3,Z= ...(truncated)"
    WeightMinimum=0
    ExecutePin=@RigVMFunction_Sequence.C
    Weight=RigVMVariableNode() {variable=IsCruchingf}.Value
   )
   n2 = &RigUnit_ItemArray(Items=((Type=Bone,Name="dyn_legBelt_03_F_l"),(Type=Bone,Name="dyn_legBelt_04_F_l"),(Type=Bone,Name="dyn_legBelt_05_F_l"),(Type=Bone,Name="dyn_legBelt_06_F_l")))
   n3 = &RigUnit_ItemArray(Items=((Type=Bone,Name="dyn_legBelt_03_F_l_CtrlJnt"),(Type=Bone,Name="dyn_legBelt_04_F_l_CtrlJnt"),(Type=Bone,Name="dyn_legBelt_05_F_l_CtrlJnt"),(Type=Bone,Name="dyn_legBelt_06_F_l_CtrlJnt")))
   SplineIK_1 = RigVMFunctionReferenceNode(
    Primary Axis=(X=1,Y=0,Z=0)
    Samples Per Segment=6
    Secondary Spline Direction=(X=0,Y=5,Z=0)
    Spline Mode=BSpline
    Up Axis=(X=0,Y=0,Z=0)
    Bones=@n2.Items
    Controls=@n3.Items
    ExecuteContext=@RigVMFunction_Sequence.D
   )
   RigVMFunction_Sequence_2 = RigVMAggregateNode(ExecuteContext=@n1.B)
   &RigUnit_ParentConstraint(
    AdvancedSettings=(InterpolationType=Average,RotationOrderForFilter=XZY)
    Child=(Type=Bone,Name="dyn_legBelt_03_B_l_CtrlJnt")
    Filter=(TranslationFilter=(bX=true,bY=true,bZ=true),RotationFilter=(bX=True,bY=True,bZ=True),ScaleFilter=(bX=True,bY=True,bZ=True))
    Parents=((Item=(Type=Bone,Name="dyn_legBelt_aimChain_01_l"),Weight=1))
    Weight=1
    ExecutePin=@RigVMFunction_Sequence_2.A
   )
   > &RigUnit_ParentConstraint(
    AdvancedSettings=(InterpolationType=Average,RotationOrderForFilter=XZY)
    Child=(Type=Bone,Name="dyn_legBelt_04_B_l_CtrlJnt")
    Filter=(TranslationFilter=(bX=true,bY=true,bZ=true),RotationFilter=(bX=True,bY=True,bZ=True),ScaleFilter=(bX=True,bY=True,bZ=True))
    Parents=((Item=(Type=Bone,Name="dyn_legBelt_aimChain_02_l"),Weight=1))
    Weight=1
   )
   > &RigUnit_ParentConstraint(
    AdvancedSettings=(InterpolationType=Average,RotationOrderForFilter=XZY)
    Child=(Type=Bone,Name="dyn_legBelt_05_B_l_CtrlJnt")
    Parents=((Item=(Type=Bone,Name="dyn_legBelt_aimChain_03_l"),Weight=1))
    Weight=1
   )
   > &RigUnit_ParentConstraint(
    AdvancedSettings=(InterpolationType=Average,RotationOrderForFilter=XZY)
    Child=(Type=Bone,Name="dyn_legBelt_06_B_l_CtrlJnt")
    Parents=((Item=(Type=Bone,Name="dyn_legBelt_aimChain_04_l"),Weight=1))
    Weight=1
   )
   > &RigUnit_ParentConstraint(
    AdvancedSettings=(InterpolationType=Average,RotationOrderForFilter=XZY)
    Child=(Type=Bone,Name="dyn_legBelt_07_B_l_CtrlJnt")
    Parents=((Item=(Type=Bone,Name="dyn_legBelt_aimChain_05_l"),Weight=1))
    Weight=1
   )
   &RigUnit_OffsetTransformForItem(
    Item=(Type=Bone,Name="dyn_legBelt_06_B_l_CtrlJnt")
    OffsetTransform=(Rotation=(X=0,Y=0,Z=0,W=-1),Translation=(X=0,Y=0,Z=3),Scale3D=(X=1,Y=1,Z=1))
    Weight=1
    bPropagateToChildren=true
    ExecutePin=@RigVMFunction_Sequence_2.B
   )
   &RigUnit_OffsetTransformForItem(
    Item=(Type=Bone,Name="base_legBelt_01_B_l")
    OffsetTransform=(Rotation=(X=0.1664,Y=-0.1359,Z=-0.0199,W=0.9764),Translation=(X=0,Y=1.5,Z=-6),Scale3D=(X=1,Y=1,Z=1))
    bPropagateToChildren=true
    ExecutePin=@RigVMFunction_Sequence_2.C
    Weight=RigVMVariableNode() {variable=L_Thigh_RotBack}.Value
   )
   > &RigUnit_OffsetTransformForItem(
    Item=(Type=Bone,Name="dyn_legBelt_03_B_l_CtrlJnt")
    OffsetTransform=(Rotation=(X=0,Y=0,Z=0,W=-1),Translation=(X=3.5,Y=0,Z=-3.5),Scale3D=(X=1,Y=1,Z=1))
    bPropagateToChildren=true
    Weight=RigVMVariableNode() {variable=L_Thigh_RotBack}.Value
   )
   > &RigUnit_OffsetTransformForItem(
    Item=(Type=Bone,Name="dyn_legBelt_04_B_l_CtrlJnt")
    OffsetTransform=(Rotation=(X=0,Y=0,Z=0,W=1),Translation=(X=1,Y=0,Z=-2),Scale3D=(X=1,Y=1,Z=1))
    bPropagateToChildren=true
    Weight=RigVMVariableNode() {variable=L_Thigh_RotBack}.Value
   )
   > &RigUnit_OffsetTransformForItem(
    Item=(Type=Bone,Name="dyn_legBelt_05_B_l_CtrlJnt")
    OffsetTransform=(Rotation=(X=0,Y=0,Z=0,W=-1),Translation=(X=3,Y=0,Z=-2),Scale3D=(X=1,Y=1,Z=1))
    bPropagateToChildren=true
    Weight=RigVMVariableNode() {variable=L_Thigh_RotBack}.Value
   )
   > &RigUnit_ModifyTransforms(
    ItemToModify="((Item=(Type=Bone,Name=\"dyn_legBelt_06_B_l_CtrlJnt\"),Transform=(Rotation=(X=0,Y=0,Z=0,W=1),Translation=(X=-2,Y=-1.5,Z=-2.5),Scale3D=(X=1,Y=1,Z=1))),(Item=(Type=Bone,Name=\"dyn_legBelt_07_B_l_CtrlJnt\"),Transform=(Rotation=(X=0,Y=0,Z=0,W=1),Translation=(X=0,Y ...(truncated)"
    WeightMinimum=0
    Weight=RigVMVariableNode() {variable=IsCruchingf}.Value
   )
   RigUnit_ItemArray = &RigUnit_ItemArray(Items=((Type=Bone,Name="dyn_legBelt_03_B_l"),(Type=Bone,Name="dyn_legBelt_04_B_l"),(Type=Bone,Name="dyn_legBelt_05_B_l"),(Type=Bone,Name="dyn_legBelt_06_B_l"),(Type=Bone,Name="dyn_legBelt_07_B_l")))
   n4 = &RigUnit_ItemArray(Items=((Type=Bone,Name="dyn_legBelt_03_B_l_CtrlJnt"),(Type=Bone,Name="dyn_legBelt_04_B_l_CtrlJnt"),(Type=Bone,Name="dyn_legBelt_05_B_l_CtrlJnt"),(Type=Bone,Name="dyn_legBelt_06_B_l_CtrlJnt"),(Type=Bone,Name="dyn_legBelt_07_B_l_CtrlJnt")))
   SplineIK = RigVMFunctionReferenceNode(
    Primary Axis=(X=1,Y=0,Z=0)
    Samples Per Segment=6
    Secondary Spline Direction=(X=0,Y=5,Z=0)
    Spline Mode=BSpline
    Stretch=true
    Up Axis=(X=0,Y=0,Z=0)
    debug=false
    Bones=@RigUnit_ItemArray.Items
    Controls=@n4.Items
    ExecuteContext=@RigVMFunction_Sequence_2.D
   )
   &RigUnit_ParentConstraint(
    AdvancedSettings=(InterpolationType=Average,RotationOrderForFilter=XZY)
    Child=(Type=Bone,Name="base_legBelt_01_B_l")
    Filter=(TranslationFilter=(bX=True,bY=True,bZ=True),RotationFilter=(bX=false,bY=false,bZ=false),ScaleFilter=(bX=True,bY=True,bZ=True))
    Parents=((Item=(Type=Bone,Name="thigh_l"),Weight=0.25),(Item=(Type=Bone,Name="pelvis"),Weight=0.75))
    Weight=1
    ExecutePin=@RigVMFunction_Sequence_3.C
   )
   DeltaRotation v4 = RigVMCollapseNode(
    Rotation Order=ZYX
    child=(Type=Bone,Name="thigh_l")
    debugRotation=false
    onlyTwist=false
    outputMinMax=(X=0,Y=1)
    parent=(Type=Null,Name="thigh_l_pelvisDeltaRef_null")
    premult=1
    rotationMinMax=(X=0,Y=-110)
    triggerAxis=Z
    twistAxis=(X=1,Y=0,Z=0)
    ExecuteContext=@RigVMFunction_Sequence_3.A
   )
   > &RigUnit_OffsetTransformForItem(
    Item=(Type=Bone,Name="base_legBelt_aimChain_01_l")
    OffsetTransform=(Rotation=(X=-0.0489,Y=-0.0375,Z=-0.1322,W=0.9893),Translation=(X=0,Y=0,Z=0),Scale3D=(X=1,Y=1,Z=1))
    bPropagateToChildren=True
    Weight=@"DeltaRotation v4".outputWeight
   )
   > DeltaRotation v4_1 = RigVMCollapseNode(
    Rotation Order=ZYX
    child=(Type=Bone,Name="thigh_l")
    debugRotation=false
    onlyTwist=false
    outputMinMax=(X=0,Y=1)
    parent=(Type=Null,Name="thigh_l_pelvisDeltaRef_null")
    premult=1
    rotationMinMax=(X=0,Y=50)
    triggerAxis=Z
    twistAxis=(X=1,Y=0,Z=0)
   )
   > RigVMVariableNode(Value=@"DeltaRotation v4_1".outputWeight) {variable=L_Thigh_RotBack}
   > &RigUnit_OffsetTransformForItem(
    Item=(Type=Bone,Name="base_legBelt_aimChain_01_l")
    OffsetTransform=(Rotation=(X=-0.0507,Y=-0.0254,Z=0.0685,W=0.996),Translation=(X=0,Y=0,Z=0),Scale3D=(X=1,Y=1,Z=1))
    bPropagateToChildren=True
    Weight=@"DeltaRotation v4_1".outputWeight
   )
   > DeltaRotation v4_2 = RigVMCollapseNode(
    Rotation Order=ZYX
    child=(Type=Bone,Name="calf_l")
    debugRotation=false
    onlyTwist=false
    outputMinMax=(X=0,Y=1)
    parent=(Type=Null,Name="calf_l_ThighDeltaRef_null")
    premult=1
    rotationMinMax=(X=0,Y=100)
    triggerAxis=Z
    twistAxis=(X=1,Y=0,Z=0)
   )
   > &RigUnit_ModifyTransforms(
    ItemToModify="((Item=(Type=Bone,Name=\"dyn_legBelt_aimChain_02_l\"),Transform=(Rotation=(X=-0.0254,Y=-0.0441,Z=0.0163,W=0.9986),Translation=(X=0,Y=0,Z=0),Scale3D=(X=1,Y=1,Z=1))),(Item=(Type=Bone,Name=\"dyn_legBelt_aimChain_03_l\"),Transform=(Rotation=(X=-0.0254,Y=-0.0441,Z= ...(truncated)"
    WeightMinimum=0
    Weight=@"DeltaRotation v4_2".outputWeight
   )
   FN_DeformRig_FNC = RigVMFunctionReferenceNode(
    DeformRig_ Curve="(deformRig_l_upperarm_up_alpha_72_#1=true,deformRig_r_upperarm_up_alpha_73_#2=true,deformRig_l_upperarm_fr_alpha_74_#3=true,deformRig_r_upperarm_fr_alpha_75_CB51B80C4 ...(truncated)"
    DeformRig_Bool="(upperarm_1_#4=True,elbow_3_#5=True,wrist_7_#6=True,groin_8_#7=True,glute_17_#8=True,knee_9_40C408324CA261 ...(truncated)"
    DeformRig_Debug=(upperarm_debug_18_#4=False,wrist_debug_20_#6=False,lat_debug_26_#9=False,groinGlute_debug_27_#7=False)
    DeformRig_Value="(upperarm_up_18_#10=(Rotation=(X=0.0,Y=-0.0,Z=0.0,W=1.0),Translation=(X=0.0,Y=0.0,Z=8.0),Scale3D=(X=0.70,Y=1.0,Z=1.0)),upperarm_fwd_fr_19_#11=(Rotation=(X=0.0,Y=-0.0,Z=0.0,W=1.0),Translation=(X=0.0, ...(truncated)"
    ExecuteContext=()
   )
   setFKCtrl = RigVMFunctionReferenceNode(ExecuteContext=(), setFK_Ctrl=false)
  ]}
 RigVMModel Construction Script {nodes [
   n1 = &RigUnit_PrepareForExecution()
   RigUnit_ItemArray = &RigUnit_ItemArray(Items="((Type=Bone,Name=\"dyn_legBelt_03_F_l\"),(Type=Bone,Name=\"dyn_legBelt_04_F_l\"),(Type=Bone,Name=\"dyn_legBelt_05_F_l\"),(Type=Bone,Name=\"dyn_legBelt_06_F_l\"),(Type=Bone,Name=\"dyn_legBelt_03_B_l\"),(Type=Bone,Name=\"dyn_legBelt_04_B_l\"),(Type=Bone,Name=\"dyn_legBel ...(truncated)")
   For_Each = RigVMDispatchNode(ExecuteContext=@n1.ExecutePin, Array=@RigUnit_ItemArray.Items) {template="DISPATCH_RigVMDispatch_ArrayIterator(in Array,out Element,out Index,out Count,out Ratio)"}
   Concat = /Script/RigVM.RigVMFunction_NameConcat(B=_CtrlJnt, A=@For_Each.Element.Name) {is=struct RigVMUnitNode, template="Concat::Execute(in A,in B,out Result)"}
   n2 = &RigUnit_GetTransform(bInitial=False, Item=@For_Each.Element)
   > n3 = /Script/ControlRig.RigUnit_HierarchyAddBone(Parent=(Type=Bone,Name="None"), Space=GlobalSpace, Name=@Concat.Result, Transform=@n2.Transform) {is=struct RigVMUnitNode}
   n4 = &RigVMFunction_NameConcat(B=_ctrl, A=@Concat.Result) {template="Concat::Execute(in A,in B,out Result)"}
   To_String = RigVMDispatchNode(Value=@n4.Result) {template="DISPATCH_RigDispatch_ToString(in Value,out Result)"}
   RigVMFunction_StringFind = /Script/RigVM.RigVMFunction_StringFind(Search=_F_, Value=@To_String.Result) {is=struct RigVMUnitNode}
   > SpawnControl = /Script/ControlRig.RigUnit_HierarchyAddControlTransform(
    OffsetSpace=GlobalSpace
    OffsetTransform=(Rotation=(X=0,Y=0,Z=0,W=1),Translation=(X=0,Y=0,Z=0),Scale3D=(X=1,Y=1,Z=1))
    Settings="(InitialSpace=LocalSpace,bUsePreferredRotationOrder=False,PreferredRotationOrder=YZX,Limits=(LimitTranslationX=(bMinimum=False,bMaximum=False),LimitTranslationY=(bMinimum=False,bMaximum=False),LimitTranslationZ=(bMinimum=False,bMaximum=False),LimitPitch=(b ...(truncated)"
    Name=@n4.Result
    InitialValue=@n2.Transform
    Parent=@n3.Item
    Settings.Shape.Color=RigVMDispatchNode(False=(R=1,G=0,B=0.13,A=0.2), True=(R=1,G=0.422,B=0.13,A=1), Condition=@RigVMFunction_StringFind.Found) {template="DISPATCH_RigVMDispatch_If(in Condition,in True,in False,out Result)"}.Result
   ) {
    is=struct RigVMUnitNode
    template="SpawnControl::Execute(in InitialValue,in Settings,in OffsetTransform,in OffsetSpace,in Parent,in Name,out Item)"
   }
   addFKCtrl = RigVMFunctionReferenceNode(
    Custom_FKCtrl=()
    ExecuteContext=()
    Options_FKCtrl="(pelvis_1_#12=true,spine_3_#13=true,clavicle_5_#14=true,arm_7_#15=true,leg_9_#16=true,head_11_46DC15EF465CD2979 ...(truncated)"
    Shape_Settings="(pelvis_2_#17=(bVisible=True,Name=\"Hexagon_Thin\",Color=(R=0.0117,G=0.8281,B=1,A=1),Transform=(Rotation=(X=0,Y=-0.7071,Z=0,W=0.7071),Translation=(X=2,Y=0,Z=0),Scale3D=(X=5,Y=5,Z=7))),spine_4_#18=(bVi ...(truncated)"
   )
  ]}
]
hierarchy [
 root {loc=<0 0 0>, rot=<0 0 0 1>}
 pelvis {parent=root, loc=<0 -0.5301 98.0571>, rot=<0.0127 0.707 -0.0127 -0.707>}
 spine_01 {parent=pelvis, loc=<6.1208 -0.9328 0>, rot=<0 0 0.034 -0.9994>}
 spine_02 {parent=spine_01, loc=<6.5997 -0.2095 0>, rot=<0 0 0.0238 -0.9997>}
 spine_03 {parent=spine_02, loc=<6.4914 -0.1237 0>, rot=<0 0 0.0149 -0.9999>}
 spine_04 {parent=spine_03, loc=<6.9598 0.0794 0>, rot=<0 0 -0.0132 -0.9999>}
 spine_05 {parent=spine_04, loc=<6.9987 0 0>, rot=<0 0 -0.0015 -1>}
 neck_01 {parent=spine_05, loc=<11.3138 0.1529 0>, rot=<0 0 -0.2003 -0.9797>}
 neck_02 {parent=neck_01, loc=<4.5627 0 0>, rot=<0 0 0.0504 -0.9987>}
 head {parent=neck_02, loc=<5.3914 0 0>, rot=<0 0 0.115 -0.9934>}
 clavicle_l {parent=spine_05, loc=<5.4831 -0.1053 -3.2946>, rot=<0.0465 -0.7289 -0.009 -0.6829>}
 upperarm_l {parent=clavicle_l, loc=<9.8658 -1.2136 -0.4105>, rot=<0.0164 -0.3245 -0.0269 -0.9454>}
 lowerarm_l {parent=upperarm_l, loc=<26.465 -0.8345 0.259>, rot=<0.0036 0.0161 -0.2298 -0.9731>}
 hand_l {parent=lowerarm_l, loc=<22.7355 -0.3203 -0.6105>, rot=<-0.6784 -0.0223 0.0735 -0.7307>}
 thumb_01_l {parent=hand_l, loc=<1.3118 -0.3191 -2.459>, rot=<0.6546 -0.4042 0.0664 -0.6354>}
 thumb_02_l {parent=thumb_01_l, loc=<5.9336 -0.0448 0.0401>, rot=<-0.0037 0.0213 0.298 -0.9543>}
 thumb_03_l {parent=thumb_02_l, loc=<2.2223 0.3682 -0.0907>, rot=<0.0118 -0.013 0.0066 -0.9998>}
 index_metacarpal_l {parent=hand_l, loc=<1.6535 0.0183 -2.96>, rot=<0.0158 -0.0124 -0.0173 -0.9996>}
 index_01_l {parent=index_metacarpal_l, loc=<8.3008 0.0357 -0.0524>, rot=<-0.0044 0.0485 0.0924 -0.9945>}
 index_02_l {parent=index_01_l, loc=<3.3587 -0.3004 0.0452>, rot=<-0.002 0.0027 0.3225 -0.9466>}
 index_03_l {parent=index_02_l, loc=<2.5578 -0.123 -0.0252>, rot=<-0.002 0.0067 0.2558 -0.9667>}
 middle_metacarpal_l {parent=hand_l, loc=<2.1856 -0.1725 -0.5984>, rot=<0.006 0.0168 -0.0329 -0.9993>}
 middle_01_l {parent=middle_metacarpal_l, loc=<7.3275 -0.2447 0.0129>, rot=<-0.0208 0.0042 0.0858 -0.9961>}
 middle_02_l {parent=middle_01_l, loc=<4.0123 0.1401 0.1175>, rot=<-0.0218 0.0258 0.3512 -0.9357>}
 middle_03_l {parent=middle_02_l, loc=<2.6077 0.0534 0.1373>, rot=<0.0082 0.0188 0.1708 -0.9851>}
 ring_metacarpal_l {parent=hand_l, loc=<2.2074 -0.3625 1.3638>, rot=<-0.0284 0.0722 -0.0458 -0.9959>}
 ring_01_l {parent=ring_metacarpal_l, loc=<7.0295 -0.6197 -0.1446>, rot=<-0.0489 -0.0252 0.1892 -0.9804>}
 ring_02_l {parent=ring_01_l, loc=<3.6174 0.4647 -0.1027>, rot=<0 0 0.2806 -0.9598>}
 ring_03_l {parent=ring_02_l, loc=<2.4152 0.114 0.0199>, rot=<0 0 0.1189 -0.9929>}
 pinky_metacarpal_l {parent=hand_l, loc=<1.7616 -0.9753 3.3814>, rot=<-0.0397 0.0814 -0.069 -0.9935>}
 pinky_01_l {parent=pinky_metacarpal_l, loc=<7.1709 -0.4831 -0.1094>, rot=<-0.0654 -0.0035 0.2242 -0.9723>}
 pinky_02_l {parent=pinky_01_l, loc=<2.8274 0.0893 -0.1043>, rot=<0.0037 0.0011 0.2854 -0.9584>}
 pinky_03_l {parent=pinky_02_l, loc=<2.2046 0.0307 -0.0078>, rot=<-0.0002 0.0098 0.1445 -0.9895>}
 deform_wrist_dn_base_l {parent=hand_l, loc=<0 -1.5326 0>, rot=<0 0 0 -1>}
 deform_wrist_dn_l {parent=deform_wrist_dn_base_l, loc=<0 0 0>, rot=<0 0 0 1>}
 deform_wrist_up_base_l {parent=hand_l, loc=<0 2.3009 0>, rot=<0 0 0 -1>}
 deform_wrist_up_l {parent=deform_wrist_up_base_l, loc=<0 0 0>, rot=<0 0 0 1>}
 lowerarm_twist_02_l {parent=lowerarm_l, loc=<10.6707 0 0>, rot=<0 0 0 1>}
 lowerarm_twist_01_l {parent=lowerarm_l, loc=<18.316 0 0>, rot=<0 0 0 1>}
 deform_elbow_in_base_l {parent=lowerarm_l, loc=<0.1834 3.6062 -0.2572>, rot=<0 0 0 -1>}
 deform_elbow_in_l {parent=deform_elbow_in_base_l, loc=<0 0 0>, rot=<0 0 0 1>}
 deform_elbow_out_base_l {parent=lowerarm_l, loc=<0 -4 0>, rot=<0 0 0 1>}
 deform_elbow_out_l {parent=deform_elbow_out_base_l, loc=<0 0 0>, rot=<0 0 0 1>}
 upperarm_twist_01_l {parent=upperarm_l, loc=<7.2135 -0.1134 0>, rot=<0 0 0 1>}
 deform_upperarm_bk_base_l {parent=upperarm_twist_01_l, loc=<-6.4053 -4.3746 0.314>, rot=<0 0 0 -1>}
 deform_upperarm_bk_l {parent=deform_upperarm_bk_base_l, loc=<0 0 0>, rot=<0 0 0 1>}
 deform_upperarm_fr_base_l {parent=upperarm_twist_01_l, loc=<-5.6312 5.092 0.3354>, rot=<0 0 0 -1>}
 deform_upperarm_fr_l {parent=deform_upperarm_fr_base_l, loc=<0 0 0>, rot=<0 0 0 1>}
 deform_upperarm_up_base_l {parent=upperarm_twist_01_l, loc=<-8.891 0.4825 3.9517>, rot=<0 0 0 -1>}
 deform_upperarm_up_l {parent=deform_upperarm_up_base_l, loc=<0 0 0>, rot=<0 0 0 1>}
 upperarm_twist_02_l {parent=upperarm_l, loc=<19.7998 -0.1134 0>, rot=<0 0 0 1>}
 clavicle_r {parent=spine_05, loc=<5.4831 -0.1053 3.2946>, rot=<0.7289 0.0465 -0.6829 0.009>}
 upperarm_r {parent=clavicle_r, loc=<-9.8658 1.2136 0.4105>, rot=<0.0164 -0.3245 -0.0269 -0.9454>}
 lowerarm_r {parent=upperarm_r, loc=<-26.465 0.8345 -0.259>, rot=<0.0036 0.0161 -0.2298 -0.9731>}
 hand_r {parent=lowerarm_r, loc=<-22.7355 0.3203 0.6106>, rot=<-0.6784 -0.0223 0.0735 -0.7307>}
 thumb_01_r {parent=hand_r, loc=<-1.3118 0.3191 2.459>, rot=<0.6546 -0.4042 0.0664 -0.6354>}
 thumb_02_r {parent=thumb_01_r, loc=<-5.9336 0.0448 -0.0401>, rot=<-0.0037 0.0213 0.298 -0.9543>}
 thumb_03_r {parent=thumb_02_r, loc=<-2.2223 -0.3682 0.0907>, rot=<0.0118 -0.013 0.0066 -0.9998>}
 index_metacarpal_r {parent=hand_r, loc=<-1.6535 -0.0183 2.96>, rot=<0.0158 -0.0124 -0.0173 -0.9996>}
 index_01_r {parent=index_metacarpal_r, loc=<-8.3008 -0.0357 0.0524>, rot=<-0.0044 0.0485 0.0924 -0.9945>}
 index_02_r {parent=index_01_r, loc=<-3.3587 0.3004 -0.0452>, rot=<-0.002 0.0027 0.3225 -0.9466>}
 index_03_r {parent=index_02_r, loc=<-2.5578 0.123 0.0252>, rot=<-0.002 0.0067 0.2558 -0.9667>}
 middle_metacarpal_r {parent=hand_r, loc=<-2.1856 0.1725 0.5984>, rot=<0.006 0.0168 -0.0329 -0.9993>}
 middle_01_r {parent=middle_metacarpal_r, loc=<-7.3275 0.2447 -0.0129>, rot=<-0.0208 0.0042 0.0858 -0.9961>}
 middle_02_r {parent=middle_01_r, loc=<-4.0123 -0.1401 -0.1175>, rot=<-0.0218 0.0258 0.3512 -0.9357>}
 middle_03_r {parent=middle_02_r, loc=<-2.6077 -0.0534 -0.1373>, rot=<0.0082 0.0188 0.1708 -0.9851>}
 ring_metacarpal_r {parent=hand_r, loc=<-2.2074 0.3625 -1.3638>, rot=<-0.0284 0.0722 -0.0458 -0.9959>}
 ring_01_r {parent=ring_metacarpal_r, loc=<-7.0295 0.6197 0.1446>, rot=<-0.0489 -0.0252 0.1892 -0.9804>}
 ring_02_r {parent=ring_01_r, loc=<-3.6174 -0.4647 0.1027>, rot=<0 0 0.2806 -0.9598>}
 ring_03_r {parent=ring_02_r, loc=<-2.4152 -0.114 -0.0199>, rot=<0 0 0.1189 -0.9929>}
 pinky_metacarpal_r {parent=hand_r, loc=<-1.7616 0.9753 -3.3814>, rot=<-0.0397 0.0814 -0.069 -0.9935>}
 pinky_01_r {parent=pinky_metacarpal_r, loc=<-7.1709 0.4831 0.1094>, rot=<-0.0654 -0.0035 0.2242 -0.9723>}
 pinky_02_r {parent=pinky_01_r, loc=<-2.8274 -0.0893 0.1043>, rot=<0.0037 0.0011 0.2854 -0.9584>}
 pinky_03_r {parent=pinky_02_r, loc=<-2.2046 -0.0307 0.0078>, rot=<-0.0002 0.0098 0.1445 -0.9895>}
 deform_wrist_dn_base_r {parent=hand_r, loc=<-0.0001 1.5328 0>, rot=<0 0 0 -1>}
 deform_wrist_dn_r {parent=deform_wrist_dn_base_r, loc=<0 0 0>, rot=<0 0 0 1>}
 deform_wrist_up_base_r {parent=hand_r, loc=<-0.0002 -2.3007 -0.0001>, rot=<0 0 0 -1>}
 deform_wrist_up_r {parent=deform_wrist_up_base_r, loc=<0 0 0>, rot=<0 0 0 1>}
 lowerarm_twist_02_r {parent=lowerarm_r, loc=<-10.6707 0 0>, rot=<0 0 0 -1>}
 lowerarm_twist_01_r {parent=lowerarm_r, loc=<-18.316 0 0>, rot=<0 0 0 -1>}
 deform_elbow_in_base_r {parent=lowerarm_r, loc=<-0.1834 -3.6062 0.2572>, rot=<0 0 0 -1>}
 deform_elbow_in_r {parent=deform_elbow_in_base_r, loc=<0 0 0>, rot=<0 0 0 1>}
 deform_elbow_out_base_r {parent=lowerarm_r, loc=<0 4 -0.0001>, rot=<0 0 0 -1>}
 deform_elbow_out_r {parent=deform_elbow_out_base_r, loc=<0 0 0>, rot=<0 0 0 1>}
 upperarm_twist_01_r {parent=upperarm_r, loc=<-7.2135 0.1134 0>, rot=<0 0 0 -1>}
 deform_upperarm_bk_base_r {parent=upperarm_twist_01_r, loc=<6.4053 4.3746 -0.314>, rot=<0 0 0 -1>}
 deform_upperarm_bk_r {parent=deform_upperarm_bk_base_r, loc=<0 0 0>, rot=<0 0 0 1>}
 deform_upperarm_fr_base_r {parent=upperarm_twist_01_r, loc=<5.6312 -5.092 -0.3354>, rot=<0 0 0 -1>}
 deform_upperarm_fr_r {parent=deform_upperarm_fr_base_r, loc=<0 0 0>, rot=<0 0 0 1>}
 deform_upperarm_up_base_r {parent=upperarm_twist_01_r, loc=<8.891 -0.4825 -3.9517>, rot=<0 0 0 -1>}
 deform_upperarm_up_r {parent=deform_upperarm_up_base_r, loc=<0 0 0>, rot=<0 0 0 1>}
 upperarm_twist_02_r {parent=upperarm_r, loc=<-19.7998 0.1134 0>, rot=<0 0 0 -1>}
 attach_backpack {parent=spine_05, loc=<-0.1227 -13.4731 0>, rot=<-0.0093 -0.707 0.0093 -0.707>}
 Backpack_BR {parent=attach_backpack, type=Null, loc=<0 0 0>, rot=<0 0 0 1>}
 attach_cape {parent=spine_05, loc=<9.0172 -11.714 0>, rot=<0.0616 -0.7044 -0.0616 -0.7044>}
 cape {parent=attach_cape, type=Null, loc=<0 0 0>, rot=<0 0 0 1>}
 base_collarStrap_01_l {parent=spine_05, loc=<10.5931 2.1457 -6.332>, rot=<-0.3486 -0.6045 -0.5278 -0.4843>}
 dyn_collarStrap_01_l {parent=base_collarStrap_01_l, loc=<0 0 0>, rot=<0 0 0 1>}
 dyn_collarStrap_02_l {parent=dyn_collarStrap_01_l, loc=<4.0281 0 0>, rot=<-0.004 0.0765 0.1171 -0.9902>}
 dyn_collarStrap_03_l {parent=dyn_collarStrap_02_l, loc=<2.7847 0 0>, rot=<0 0 0 1>}
 base_collarStrap_01_r {parent=spine_05, loc=<10.9687 2.8062 6.9699>, rot=<-0.2137 -0.7412 -0.5096 -0.3812>}
 dyn_collarStrap_01_r {parent=base_collarStrap_01_r, loc=<0 0 0>, rot=<0 0 0 1>}
 dyn_collarStrap_02_r {parent=dyn_collarStrap_01_r, loc=<4.0281 0 0.0002>, rot=<-0.2409 -0.0125 0.0832 -0.9669>}
 dyn_collarStrap_03_r {parent=dyn_collarStrap_02_r, loc=<2.7847 0 -0.0003>, rot=<-0.1567 0.0004 0.0022 -0.9876>}
 deform_pec_aim_base_l {parent=spine_04, loc=<3.2456 9.3309 -1.2237>, rot=<0.0603 -0.4033 0.3318 -0.8507>}
 deform_pec_aim_l {parent=deform_pec_aim_base_l, loc=<0 0 0>, rot=<0 0 0 1>}
 deform_pec_base_l {parent=deform_pec_aim_l, loc=<9.3665 0 0>, rot=<0 0 0 1>}
 deform_pec_l {parent=deform_pec_base_l, loc=<0 0 0>, rot=<0 0 0 1>}
 deform_pec_aim_base_r {parent=spine_04, loc=<3.2452 9.3309 1.2237>, rot=<-0.4033 -0.0603 0.8507 0.3318>}
 deform_pec_aim_r {parent=deform_pec_aim_base_r, loc=<0 0 0>, rot=<0 0 0 1>}
 deform_pec_base_r {parent=deform_pec_aim_r, loc=<-9.3667 -0.0003 -0.0004>, rot=<0 0 0 1>}
 deform_pec_r {parent=deform_pec_base_r, loc=<0 0 0>, rot=<0 0 0 1>}
 base_jacketFlap_01_F_l {parent=spine_04, loc=<-0.7357 9.7981 -8.8831>, rot=<-0.0295 -0.9341 -0.3535 -0.0411>}
 dyn_jacketFlap_01_F_l {parent=base_jacketFlap_01_F_l, loc=<0 0 0>, rot=<0 0 0 1>}
 dyn_jacketFlap_01_F_l_End {parent=dyn_jacketFlap_01_F_l, loc=<5.453 0 0>, rot=<0 0 0 1>}
 base_jacketFlap_01_S_l {parent=spine_04, loc=<-1.6161 -0.0841 -12.4302>, rot=<-0.0459 -0.991 0.1247 -0.0177>}
 dyn_jacketFlap_01_S_l {parent=base_jacketFlap_01_S_l, loc=<0 0 0>, rot=<0 0 0 1>}
 dyn_jacketFlap_01_S_l_End {parent=dyn_jacketFlap_01_S_l, loc=<5.453 0 0>, rot=<0 0 0 1>}
 base_jacketFlap_01_SB_l {parent=spine_04, loc=<-2.5575 -7.4654 -8.0891>, rot=<-0.0845 -0.8342 0.5406 0.069>}
 dyn_jacketFlap_01_SB_l {parent=base_jacketFlap_01_SB_l, loc=<0 0 0>, rot=<0 0 0 1>}
 dyn_jacketFlap_01_SB_l_End {parent=dyn_jacketFlap_01_SB_l, loc=<5.453 0 0>, rot=<0 0 0 1>}
 base_jacketFlap_01_B {parent=spine_04, loc=<-2.7909 -8.8592 0>, rot=<-0.0176 -0.7069 0.7069 0.0176>}
 dyn_jacketFlap_01_B {parent=base_jacketFlap_01_B, loc=<0 0 0>, rot=<0 0 0 1>}
 dyn_jacketFlap_01_B_End {parent=dyn_jacketFlap_01_B, loc=<5.453 0 0>, rot=<0 0 0 1>}
 base_jacketFlap_01_SB_r {parent=spine_04, loc=<-2.5571 -7.4653 8.0891>, rot=<-0.069 -0.5406 0.8342 0.0845>}
 dyn_jacketFlap_01_SB_r {parent=base_jacketFlap_01_SB_r, loc=<0 0 0>, rot=<0 0 0 1>}
 dyn_jacketFlap_01_SB_r_End {parent=dyn_jacketFlap_01_SB_r, loc=<5.4529 0 0>, rot=<0 0 0 1>}
 base_jacketFlap_01_S_r {parent=spine_04, loc=<-1.6164 -0.0842 12.4301>, rot=<0.0177 -0.1247 0.991 0.0459>}
 dyn_jacketFlap_01_S_r {parent=base_jacketFlap_01_S_r, loc=<0 0 0>, rot=<0 0 0 1>}
 dyn_jacketFlap_01_S_r_End {parent=dyn_jacketFlap_01_S_r, loc=<5.4525 -0.0001 0.0001>, rot=<0 0 0 1>}
 base_jacketFlap_01_F_r {parent=spine_04, loc=<-0.7359 9.7981 8.8831>, rot=<0.0411 0.3535 0.9341 0.0295>}
 dyn_jacketFlap_01_F_r {parent=base_jacketFlap_01_F_r, loc=<0 0 0>, rot=<0 0 0 1>}
 dyn_jacketFlap_01_F_r_End {parent=dyn_jacketFlap_01_F_r, loc=<5.4528 0 0>, rot=<0 0 0 -1>}
 thigh_l {parent=pelvis, loc=<-1.0855 -0.2109 -8.7053>, rot=<0.0009 0.0355 0.0155 -0.9993>}
 calf_l {parent=thigh_l, loc=<-42.5284 -0.0019 -0.7297>, rot=<0.0006 -0.0161 -0.0391 -0.9991>}
 foot_l {parent=calf_l, loc=<-46.7401 -2.5014 -0.592>, rot=<0.0427 -0.0238 0.0353 -0.9982>}
 ball_l {parent=foot_l, loc=<-3.5658 10.8066 -0.1544>, rot=<0.0001 -0.0001 -0.7063 -0.7079>}
 calf_twist_02_l {parent=calf_l, loc=<-19.1546 -1.7943 -0.2638>, rot=<0.0237 0.0095 -0.0035 -0.9997>}
 calf_twist_01_l {parent=calf_l, loc=<-35.8483 -1.9197 -0.5794>, rot=<0.0237 0.0095 -0.0035 -0.9997>}
 deform_knee_in_base_l {parent=calf_l, loc=<0.8031 -5.0231 0.2456>, rot=<0 0 0 -1>}
 deform_knee_in_l {parent=deform_knee_in_base_l, loc=<0 0 0>, rot=<0 0 0 1>}
 deform_knee_out_base_l {parent=calf_l, loc=<0 4 0>, rot=<0 0 0 -1>}
 deform_knee_out_l {parent=deform_knee_out_base_l, loc=<0 0 0>, rot=<0 0 0 1>}
 thigh_twist_01_l {parent=thigh_l, loc=<-27.952 0 0>, rot=<0.0013 0 0 -1>}
 deform_groin_base_l {parent=thigh_l, loc=<0.4801 5.6776 -0.2486>, rot=<0 0 0 -1>}
 deform_groin_l {parent=deform_groin_base_l, loc=<0 0 0>, rot=<0 0 0 1>}
 base_legBelt_01_F_l {parent=thigh_l, loc=<-15.2241 6.5295 -3.6446>, rot=<0.0626 -0.9501 0.2508 -0.1749>}
 dyn_legBelt_01_F_l {parent=base_legBelt_01_F_l, loc=<0 0 0>, rot=<0 0 0 1>}
 dyn_legBelt_02_F_l {parent=dyn_legBelt_01_F_l, loc=<1.0447 0.7131 0.3693>, rot=<0 0 0 1>}
 dyn_legBelt_03_F_l {parent=dyn_legBelt_02_F_l, loc=<4.503 0.2327 0.7424>, rot=<0 0 0 1>}
 dyn_legBelt_04_F_l {parent=dyn_legBelt_03_F_l, loc=<7.208 0.2187 -0.9661>, rot=<0 0 0 1>}
 dyn_legBelt_05_F_l {parent=dyn_legBelt_04_F_l, loc=<8.0574 -1.6008 -1.1691>, rot=<0 0 0 1>}
 dyn_legBelt_06_F_l {parent=dyn_legBelt_05_F_l, loc=<4.2282 -6.2485 0.6332>, rot=<0 0 0 1>}
 base_legBelt_01_B_l {parent=thigh_l, loc=<-3.6062 -10.205 -0.4868>, rot=<0.1876 -0.7073 0.1779 -0.658>}
 dyn_legBelt_01_B_l {parent=base_legBelt_01_B_l, loc=<0 0 0>, rot=<0 -0.5373 0 -0.8434>}
 dyn_legBelt_02_B_l {parent=dyn_legBelt_01_B_l, loc=<1.1094 -0.3553 0.1966>, rot=<0 0 0 1>}
 dyn_legBelt_03_B_l {parent=dyn_legBelt_02_B_l, loc=<4.059 -0.3273 -1.3318>, rot=<0 0 0 1>}
 dyn_legBelt_04_B_l {parent=dyn_legBelt_03_B_l, loc=<8.1003 -0.0008 -2.2249>, rot=<0 0 0 1>}
 dyn_legBelt_05_B_l {parent=dyn_legBelt_04_B_l, loc=<7.0454 0.6268 -2.1984>, rot=<0 0 0 1>}
 dyn_legBelt_06_B_l {parent=dyn_legBelt_05_B_l, loc=<7.9553 1.8725 -2.4476>, rot=<0 0 0 1>}
 dyn_legBelt_07_B_l {parent=dyn_legBelt_06_B_l, loc=<5.0796 7.9343 -1.6503>, rot=<0 0 0 1>}
 base_legBelt_aimChain_01_l {parent=thigh_l, loc=<-8.8591 -2.3226 -2.068>, rot=<0.0169 -0.9969 0.0015 -0.0771>}
 dyn_legBelt_aimChain_01_l {parent=base_legBelt_aimChain_01_l, loc=<0 0 0>, rot=<0 0 0 1>}
 dyn_legBelt_aimChain_02_l {parent=dyn_legBelt_aimChain_01_l, loc=<10.5405 0 -0.0286>, rot=<0 -0.0039 0 -1>}
 dyn_legBelt_aimChain_03_l {parent=dyn_legBelt_aimChain_02_l, loc=<7.8626 0 -0.0061>, rot=<0 0.0157 0 -0.9999>}
 dyn_legBelt_aimChain_04_l {parent=dyn_legBelt_aimChain_03_l, loc=<6.8952 0 -0.0188>, rot=<0 0.0301 0 -0.9995>}
 dyn_legBelt_aimChain_05_l {parent=dyn_legBelt_aimChain_04_l, loc=<4.6646 0 0>, rot=<0 0 0 1>}
 calf_l_ThighDeltaRef_null {parent=thigh_l, type=Null, loc=<-42.5284 -0.0019 -0.7297>, rot=<0.0006 -0.0161 -0.0391 -0.9991>}
 thigh_r {parent=pelvis, loc=<-1.0855 -0.2109 8.7053>, rot=<0.0355 -0.0009 0.9993 0.0155>}
 calf_r {parent=thigh_r, loc=<42.5284 0.0019 0.7297>, rot=<0.0006 -0.0161 -0.0391 -0.9991>}
 foot_r {parent=calf_r, loc=<46.7401 2.5014 0.592>, rot=<0.0427 -0.0238 0.0353 -0.9982>}
 ball_r {parent=foot_r, loc=<3.5658 -10.8066 0.1544>, rot=<0.0001 -0.0001 -0.7063 -0.7079>}
 calf_twist_02_r {parent=calf_r, loc=<19.1546 1.7943 0.2638>, rot=<0.0237 0.0095 -0.0035 -0.9997>}
 calf_twist_01_r {parent=calf_r, loc=<35.8483 1.9197 0.5794>, rot=<0.0237 0.0095 -0.0035 -0.9997>}
 deform_knee_in_base_r {parent=calf_r, loc=<-0.8031 5.0231 -0.2457>, rot=<0 0 0 -1>}
 deform_knee_in_r {parent=deform_knee_in_base_r, loc=<0 0 0>, rot=<0 0 0 1>}
 deform_knee_out_base_r {parent=calf_r, loc=<0 -4 0>, rot=<0 0 0 -1>}
 deform_knee_out_r {parent=deform_knee_out_base_r, loc=<0 0 0>, rot=<0 0 0 1>}
 thigh_twist_01_r {parent=thigh_r, loc=<27.952 0 0>, rot=<0.0013 0 0 -1>}
 deform_groin_base_r {parent=thigh_r, loc=<-0.4801 -5.6776 0.2486>, rot=<0 0 0 -1>}
 deform_groin_r {parent=deform_groin_base_r, loc=<0 0 0>, rot=<0 0 0 1>}
 base_regBelt_01_F_r {parent=thigh_r, loc=<15.2242 -6.5295 3.6446>, rot=<0.2487 0.2044 -0.0704 -0.9442>}
 dyn_regBelt_01_F_r {parent=base_regBelt_01_F_r, loc=<0 0 0>, rot=<0 0 0 1>}
 dyn_regBelt_02_F_r {parent=dyn_regBelt_01_F_r, loc=<1.0656 -0.7131 0.3037>, rot=<0 0 0 1>}
 dyn_regBelt_03_F_r {parent=dyn_regBelt_02_F_r, loc=<4.5404 -0.2326 0.4608>, rot=<0 0 0 1>}
 dyn_regBelt_04_F_r {parent=dyn_regBelt_03_F_r, loc=<7.134 -0.2188 -1.4126>, rot=<0 0 0 1>}
 dyn_regBelt_05_F_r {parent=dyn_regBelt_04_F_r, loc=<7.9691 1.6008 -1.6681>, rot=<0 0 0 1>}
 dyn_regBelt_06_F_r {parent=dyn_regBelt_05_F_r, loc=<4.2595 6.2485 0.3689>, rot=<0 0 0 1>}
 base_regBelt_01_B_r {parent=thigh_r, loc=<3.6062 10.205 0.4868>, rot=<0.2487 0.2044 -0.0704 -0.9442>}
 dyn_regBelt_01_B_r {parent=base_regBelt_01_B_r, loc=<0 0 0>, rot=<0 0 0 1>}
 dyn_regBelt_02_B_r {parent=dyn_regBelt_01_B_r, loc=<1.1194 0.3553 0.1273>, rot=<0 0 0 1>}
 dyn_regBelt_03_B_r {parent=dyn_regBelt_02_B_r, loc=<3.9682 0.3273 -1.5817>, rot=<0 0 0 1>}
 dyn_regBelt_04_B_r {parent=dyn_regBelt_03_B_r, loc=<7.9462 0.0008 -2.7245>, rot=<0 0 0 1>}
 dyn_regBelt_05_B_r {parent=dyn_regBelt_04_B_r, loc=<6.895 -0.6268 -2.6324>, rot=<0 0 0 1>}
 dyn_regBelt_06_B_r {parent=dyn_regBelt_05_B_r, loc=<7.7876 -1.8726 -2.9377>, rot=<0 0 0 1>}
 dyn_regBelt_07_B_r {parent=dyn_regBelt_06_B_r, loc=<4.9672 -7.9344 -1.9631>, rot=<0 0 0 1>}
 base_regBelt_aimChain_01_r {parent=thigh_r, loc=<8.8592 2.3226 2.068>, rot=<0.0015 0.0771 -0.0169 -0.9969>}
 dyn_regBelt_aimChain_01_r {parent=base_regBelt_aimChain_01_r, loc=<0 0 0>, rot=<0 0 0 1>}
 dyn_regBelt_aimChain_02_r {parent=dyn_regBelt_aimChain_01_r, loc=<10.5404 0 -0.0285>, rot=<0 -0.0039 0 -1>}
 dyn_regBelt_aimChain_03_r {parent=dyn_regBelt_aimChain_02_r, loc=<7.8627 0 -0.0061>, rot=<0 0.0157 0 -0.9999>}
 dyn_regBelt_aimChain_04_r {parent=dyn_regBelt_aimChain_03_r, loc=<6.8951 0 -0.0188>, rot=<0 0.0301 0 -0.9995>}
 dyn_regBelt_aimChain_05_r {parent=dyn_regBelt_aimChain_04_r, loc=<4.6647 0 -0.0001>, rot=<0 0 0 1>}
 deform_glute_base_l {parent=pelvis, loc=<-7.3466 -9.0213 -9.1245>, rot=<0.0172 0.031 -0.4477 -0.8935>}
 deform_glute_l {parent=deform_glute_base_l, loc=<0 0 0>, rot=<0 0 0 1>}
 deform_glute_base_r {parent=pelvis, loc=<-7.3464 -9.0213 9.1245>, rot=<0.031 -0.0172 0.8935 -0.4477>}
 deform_glute_r {parent=deform_glute_base_r, loc=<0 0 0>, rot=<0 0 0 1>}
 base_beltStrap_01 {parent=pelvis, loc=<0.6723 6.7325 8.2183>, rot=<0.7141 -0.0661 0.6908 -0.0922>}
 dyn_beltStrap_01 {parent=base_beltStrap_01, loc=<0 0 0>, rot=<0 0 0 1>}
 dyn_beltStrap_02 {parent=dyn_beltStrap_01, loc=<4.3695 -0.051 -1.2633>, rot=<-0.0731 -0.256 0.2933 -0.9182>}
 dyn_beltStrap_03 {parent=dyn_beltStrap_02, loc=<2.9635 0 0>, rot=<-0.0821 -0.0291 0.3331 -0.9389>}
 thigh_l_pelvisDeltaRef_null {parent=pelvis, type=Null, loc=<-1.0855 -0.2109 -8.7053>, rot=<0.0009 0.0355 0.0155 -0.9993>}
]
variables [
 IsCruchingf {type=double, default=0}
 L_Thigh_RotBack {type=double, default=0}
]
