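# Control Rig for the GeminiLink face accessory that applies pose-driven corrective offsets to back hair bones as the head and neck rotate.
# A reusable DeltaRotation function swing-twist decomposes rotation against reference nulls and remaps one Euler axis to a clamped weight.
# Head weights drive Modify Transforms directly, while products of two neck axes and one head axis act as combination-pose triggers.
# An optional debug function, disabled here, poses spine, neck and head from controls.
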ControlRigBlueprint /BRCosmetics/Characters/Player/Female/Medium/Bodies/F_MED_GeminiLink/Meshes/Parts/F_MED_GeminiLink_FaceAcc_CtrlRig
graphs [
 Multiply_1_ContainedGraph {nodes [
   n1 = RigVMFunctionEntryNode()
   Multiply = &RigVMFunction_MathDoubleMul(A=@n1.A, B=@n1.B)
   RigVMFunctionReturnNode(Result=&RigVMFunction_MathDoubleMul(B=@n1.C, A=@Multiply.Result).Result)
  ]}
 Multiply_2_ContainedGraph {nodes [
   n1 = RigVMFunctionEntryNode()
   RigVMFunctionReturnNode(Result=&RigVMFunction_MathDoubleMul(B=@n1.C, A=&RigVMFunction_MathDoubleMul(A=@n1.A, B=@n1.B).Result).Result)
  ]}
 New Function_ContainedGraph {nodes [
   n1 = RigVMFunctionEntryNode()
   RigVMFunction_MathBoolToInteger = &RigVMFunction_MathBoolToInteger(Value=@n1.onlyTwist)
   RigVMFunction_MathQuaternionSwingTwist = &RigVMFunction_MathQuaternionSwingTwist(
    TwistAxis=@n1.twistAxis
    Input=&RigUnit_GetRelativeTransformForItem(bChildInitial=False, bParentInitial=False, Child=@n1.child, Parent=@n1.parent).RelativeTransform.Rotation
   )
   n2 = RigVMDispatchNode(
    Values=((),())
    Index=@RigVMFunction_MathBoolToInteger.Result
    Values.0=@RigVMFunction_MathQuaternionSwingTwist.Swing
    Values.1=@RigVMFunction_MathQuaternionSwingTwist.Twist
   ) {template="DISPATCH_RigVMDispatch_SelectInt32(in Index,in Values,out Result)"}
   RigVMFunction_MathQuaternionToEuler = &RigVMFunction_MathQuaternionToEuler(Value=@n2.Result, RotationOrder=@n1.Rotation Order)
   > DISPATCH_RigVMDispatch_Print = RigVMDispatchNode(ScreenColor=(R=1,G=1,B=1,A=1), ScreenDuration=1.5, Enabled=@n1.debugRotation, Value=@RigVMFunction_MathQuaternionToEuler.Result) {template="DISPATCH_RigVMDispatch_Print(in Prefix,in Value,in Enabled,in ScreenDuration,in ScreenColor)"}
   Cast = RigVMDispatchNode(Value=@n1.triggerAxis) {template="DISPATCH_RigVMDispatch_CastEnumToInt(in Value,out Result)"}
   DISPATCH_RigVMDispatch_SelectInt32 = RigVMDispatchNode(
    Values=(0,0,0)
    Index=@Cast.Result
    Values.0=@RigVMFunction_MathQuaternionToEuler.Result.X
    Values.1=@RigVMFunction_MathQuaternionToEuler.Result.Y
    Values.2=@RigVMFunction_MathQuaternionToEuler.Result.Z
   ) {template="DISPATCH_RigVMDispatch_SelectInt32(in Index,in Values,out Result)"}
   Remap = &RigVMFunction_MathDoubleRemap(
    bClamp=False
    Value=@DISPATCH_RigVMDispatch_SelectInt32.Result
    TargetMinimum=@n1.outputMinMax.X
    TargetMaximum=@n1.outputMinMax.Y
    SourceMinimum=@n1.rotationMinMax.X
    SourceMaximum=@n1.rotationMinMax.Y
   )
   Absolute = &RigVMFunction_MathDoubleAbs(Value=@Remap.Result)
   Multiply = &RigVMFunction_MathDoubleMul(
    A=&RigVMFunction_MathDoubleClamp(Value=RigVMDispatchNode(True=@Absolute.Result, Condition=@n1.absoluteValue, False=@Remap.Result) {template="DISPATCH_RigVMDispatch_If(in Condition,in True,in False,out Result)"}.Result).Result
    B=@n1.premult
   )
   > RigVMFunctionReturnNode(outputQuat=@n2.Result, outputWeight=@Multiply.Result)
  ]}
 New Function_ContainedGraph {nodes [
   n1 = RigVMFunctionEntryNode()
   RigVMFunction_ControlFlowBranch = &RigVMFunction_ControlFlowBranch(Condition=@n1.Debug Controllers, ExecuteContext=@n1.ExecuteContext)
   RigVMFunctionReturnNode(ExecuteContext=@RigVMFunction_ControlFlowBranch.Completed)
   RigUnit_GetTransform = &RigUnit_GetTransform(Item=(Type=Control,Name="spine_05_ctrl"), bInitial=False)
   &RigUnit_SetTransform(
    Item=(Type=Bone,Name="spine_05")
    Weight=1
    bInitial=False
    Value=@RigUnit_GetTransform.Transform
    ExecutePin=@RigVMFunction_ControlFlowBranch.True
   )
   > RigUnit_SetTransform = &RigUnit_SetTransform(
    Item=(Type=Bone,Name="neck_01")
    Weight=1
    bInitial=False
    Value=&RigUnit_GetTransform(Item=(Type=Control,Name="neck_01_ctrl"), bInitial=False).Transform
   )
   > &RigUnit_SetTransform(
    Item=(Type=Bone,Name="neck_02")
    Weight=1
    bInitial=False
    Value=&RigUnit_GetTransform(Item=(Type=Control,Name="neck_02_ctrl"), bInitial=False).Transform
   )
   > &RigUnit_SetTransform(
    Item=(Type=Bone,Name="head")
    Weight=1
    bInitial=False
    Value=&RigUnit_GetTransform(Item=(Type=Control,Name="head_ctrl"), bInitial=False).Transform
   )
  ]}
 RigVMFunctionLibrary {nodes [
   Constrollers = RigVMCollapseNode(Debug Controllers=False, ExecuteContext=())
   DeltaRotation v4 = RigVMCollapseNode(
    ExecuteContext=()
    Rotation Order=ZYX
    child=(Type=Bone,Name="None")
    outputMinMax=(X=0,Y=1)
    parent=(Type=Bone,Name="None")
    premult=1
    rotationMinMax=(X=0,Y=0)
    twistAxis=(X=1,Y=0,Z=0)
   )
  ]}
 RigVMModel {nodes [
   RigUnit_BeginExecution = &RigUnit_BeginExecution()
   Constrollers = RigVMFunctionReferenceNode(Debug Controllers=false, ExecuteContext=@RigUnit_BeginExecution.ExecutePin)
   > DeltaRotation v4 = RigVMFunctionReferenceNode(
    Rotation Order=ZYX
    child=(Type=Bone,Name="head")
    debugRotation=false
    outputMinMax=(X=0,Y=1)
    parent=(Type=Null,Name="head_null")
    premult=1
    rotationMinMax=(X=0,Y=-25)
    triggerAxis=Z
    twistAxis=(X=1,Y=0,Z=0)
   )
   > &RigUnit_ModifyTransforms(
    ItemToModify="((Item=(Type=Bone,Name=\"base_hair_back_07\"),Transform=(Rotation=(X=-0.0971,Y=-0.2559,Z=-0.1571,W=0.9489),Translation=(X=-0.8561,Y=-0.2298,Z=2.9769),Scale3D=(X=1,Y=1,Z=1))),(Item=(Type=Bone,Name=\"base_hair_back_13\"),Transform=(Rotation=(X=-0.0519,Y=0.4274,Z ...(truncated)"
    WeightMinimum=0
    Weight=@"DeltaRotation v4".outputWeight
   )
   > DeltaRotation v4_1 = RigVMFunctionReferenceNode(
    Rotation Order=ZYX
    child=(Type=Bone,Name="head")
    debugRotation=false
    outputMinMax=(X=0,Y=1)
    parent=(Type=Null,Name="head_null")
    premult=1
    rotationMinMax=(X=0,Y=-45)
    triggerAxis=Z
    twistAxis=(X=1,Y=0,Z=0)
   )
   > &RigUnit_ModifyTransforms(
    ItemToModify="((Item=(Type=Bone,Name=\"base_hair_back_07\"),Transform=(Rotation=(X=0.1423,Y=-0.0027,Z=0.0689,W=-0.9874),Translation=(X=-1.1659,Y=0.1018,Z=1.61),Scale3D=(X=1,Y=1,Z=1))),(Item=(Type=Bone,Name=\"base_hair_back_13\"),Transform=(Rotation=(X=0.0146,Y=0.205,Z=0.069 ...(truncated)"
    WeightMinimum=0
    Weight=@"DeltaRotation v4_1".outputWeight
   )
   > DeltaRotation v4_2 = RigVMFunctionReferenceNode(
    Rotation Order=ZYX
    child=(Type=Bone,Name="neck_01")
    debugRotation=false
    outputMinMax=(X=0,Y=1)
    parent=(Type=Null,Name="neck_01_null")
    premult=1
    rotationMinMax=(X=0,Y=-23)
    triggerAxis=Z
    twistAxis=(X=1,Y=0,Z=0)
   )
   > DeltaRotation v4_3 = RigVMFunctionReferenceNode(
    Rotation Order=ZYX
    child=(Type=Bone,Name="neck_01")
    debugRotation=false
    outputMinMax=(X=0,Y=1)
    parent=(Type=Null,Name="neck_01_null")
    premult=1
    rotationMinMax=(X=0,Y=-17)
    triggerAxis=Y
    twistAxis=(X=1,Y=0,Z=0)
   )
   > DeltaRotation v4_4 = RigVMFunctionReferenceNode(
    Rotation Order=ZYX
    child=(Type=Bone,Name="head")
    debugRotation=false
    outputMinMax=(X=0,Y=1)
    parent=(Type=Null,Name="head_null")
    premult=1
    rotationMinMax=(X=0,Y=-30)
    triggerAxis=Z
    twistAxis=(X=1,Y=0,Z=0)
   )
   Multiply = RigVMAggregateNode(A=@"DeltaRotation v4_2".outputWeight, B=@"DeltaRotation v4_3".outputWeight, C=@"DeltaRotation v4_4".outputWeight)
   > &RigUnit_ModifyTransforms(
    ItemToModify="((Item=(Type=Bone,Name=\"base_hair_back_13\"),Transform=(Rotation=(X=0,Y=0,Z=0,W=1),Translation=(X=0,Y=0,Z=0),Scale3D=(X=1,Y=1,Z=1))),(Item=(Type=Bone,Name=\"base_hair_back_10\"),Transform=(Rotation=(X=0,Y=0,Z=0,W=1),Translation=(X=-3,Y=4,Z=0),Scale3D=(X=1,Y=1 ...(truncated)"
    WeightMinimum=0
    Weight=@Multiply.Result
   )
   > &RigUnit_ModifyTransforms(
    ItemToModify="((Item=(Type=Bone,Name=\"base_hair_back_14\"),Transform=(Rotation=(X=0,Y=-0.2588,Z=0,W=0.9659),Translation=(X=0,Y=0,Z=0),Scale3D=(X=1,Y=1,Z=1))),(Item=(Type=Bone,Name=\"base_hair_back_21\"),Transform=(Rotation=(X=0,Y=0,Z=0,W=1),Translation=(X=0,Y=0,Z=0),Scale3 ...(truncated)"
    WeightMinimum=0
    Weight=@Multiply.Result
   )
   > DeltaRotation v4_5 = RigVMFunctionReferenceNode(
    Rotation Order=ZYX
    child=(Type=Bone,Name="neck_01")
    debugRotation=false
    outputMinMax=(X=0,Y=1)
    parent=(Type=Null,Name="neck_01_null")
    premult=1
    rotationMinMax=(X=0,Y=-5)
    triggerAxis=Z
    twistAxis=(X=1,Y=0,Z=0)
   )
   > DeltaRotation v4_6 = RigVMFunctionReferenceNode(
    Rotation Order=ZYX
    child=(Type=Bone,Name="neck_01")
    debugRotation=false
    outputMinMax=(X=0,Y=1)
    parent=(Type=Null,Name="neck_01_null")
    premult=1
    rotationMinMax=(X=0,Y=-12)
    triggerAxis=Y
    twistAxis=(X=1,Y=0,Z=0)
   )
   > DeltaRotation v4_7 = RigVMFunctionReferenceNode(
    Rotation Order=ZYX
    child=(Type=Bone,Name="head")
    debugRotation=false
    outputMinMax=(X=0,Y=1)
    parent=(Type=Null,Name="head_null")
    premult=1
    rotationMinMax=(X=0,Y=-23)
    triggerAxis=Z
    twistAxis=(X=1,Y=0,Z=0)
   )
   Multiply_1 = RigVMAggregateNode(C=@"DeltaRotation v4_5".outputWeight, A=@"DeltaRotation v4_6".outputWeight, B=@"DeltaRotation v4_7".outputWeight)
   > &RigUnit_ModifyTransforms(
    ItemToModify="((Item=(Type=Bone,Name=\"base_hair_back_12\"),Transform=(Rotation=(X=0,Y=-0.0872,Z=0,W=0.9962),Translation=(X=0,Y=0,Z=0.5),Scale3D=(X=1,Y=1,Z=1))),(Item=(Type=Bone,Name=\"base_hair_back_20\"),Transform=(Rotation=(X=0,Y=-0.1736,Z=0,W=0.9848),Translation=(X=0,Y= ...(truncated)"
    WeightMinimum=0
    Weight=@Multiply_1.Result
   )
  ]}
]
hierarchy [
 root {loc=<0 0 0>, rot=<0 0 0 1>}
 pelvis {parent=root, loc=<0 -0.5301 98.0571>, rot=<0.0127 0.707 -0.0127 -0.707>}
 spine_01 {parent=pelvis, loc=<6.1208 -0.9328 0>, rot=<0 0 0.034 -0.9994>}
 spine_02 {parent=spine_01, loc=<6.5997 -0.2095 0>, rot=<0 0 0.0238 -0.9997>}
 spine_03 {parent=spine_02, loc=<6.4914 -0.1237 0>, rot=<0 0 0.0149 -0.9999>}
 spine_04 {parent=spine_03, loc=<6.9598 0.0794 0>, rot=<0 0 -0.0132 -0.9999>}
 spine_05 {parent=spine_04, loc=<6.9987 0 0>, rot=<0 0 -0.0015 -1>}
 neck_01 {parent=spine_05, loc=<11.3138 0.1529 0>, rot=<0 0 -0.2003 -0.9797>}
 neck_02 {parent=neck_01, loc=<4.5627 0 0>, rot=<0 0 0.0504 -0.9987>}
 head {parent=neck_02, loc=<5.3914 0 0>, rot=<0 0 0.115 -0.9934>}
 HairMain_01 {parent=head, loc=<7.1837 -2.5907 0>, rot=<0.101 -0.6999 0.6999 -0.101>}
 HairMain_02 {parent=HairMain_01, loc=<6.2714 0 0>, rot=<0 -0.1224 0 -0.9925>}
 HairMain_03 {parent=HairMain_02, loc=<5.8192 0 0>, rot=<0 0.0295 0 -0.9996>}
 base_HairSpike_A_01_l {parent=head, loc=<6.5319 1.5352 -7.4061>, rot=<0.0041 -1 0 0>}
 dyn_HairSpike_A_01_l {parent=base_HairSpike_A_01_l, loc=<0 0 0>, rot=<0 0 0 1>}
 base_HairSpike_B_01_l {parent=head, loc=<9.9494 0.1681 -7.3092>, rot=<0.0041 -1 0 0>}
 dyn_HairSpike_B_01_l {parent=base_HairSpike_B_01_l, loc=<0 0 0>, rot=<0 0 0 1>}
 base_HairSpike_C_01_l {parent=head, loc=<3.358 6.9803 -5.9>, rot=<0.0041 -1 0 0>}
 dyn_HairSpike_C_01_l {parent=base_HairSpike_C_01_l, loc=<0 0 0>, rot=<0 0 0 1>}
 base_HairSpike_D_01_l {parent=head, loc=<5.9304 9.9965 -4.1098>, rot=<0.0041 -1 0 0>}
 dyn_HairSpike_D_01_l {parent=base_HairSpike_D_01_l, loc=<0 0 0>, rot=<0 0 0 1>}
 base_HairSpike_E_01_l {parent=head, loc=<8.4906 10.8221 -1.9717>, rot=<0.0041 -1 0 0>}
 dyn_HairSpike_E_01_l {parent=base_HairSpike_E_01_l, loc=<0 0 0>, rot=<0 0 0 1>}
 base_HairSpike_F_01_c {parent=head, loc=<8.3628 10.6369 0.201>, rot=<0.0041 -1 0 0>}
 dyn_HairSpike_F_01_c {parent=base_HairSpike_F_01_c, loc=<0 0 0>, rot=<0 0 0 1>}
 base_HairSpike_G_01_l {parent=head, loc=<12.4869 5.0443 -4.8291>, rot=<0.0041 -1 0 0>}
 dyn_HairSpike_G_01_l {parent=base_HairSpike_G_01_l, loc=<0 0 0>, rot=<0 0 0 1>}
 base_HairSpike_H_01_l {parent=head, loc=<14.0049 5.3461 0.8886>, rot=<0.0041 -1 0 0>}
 dyn_HairSpike_H_01_l {parent=base_HairSpike_H_01_l, loc=<0 0 0>, rot=<0 0 0 1>}
 base_HairSpike_A_01_r {parent=head, loc=<4.4458 4.2077 6.6507>, rot=<0.0041 -1 0 0>}
 dyn_HairSpike_A_01_r {parent=base_HairSpike_A_01_r, loc=<0 0 0>, rot=<0 0 0 1>}
 base_HairSpike_B_01_r {parent=head, loc=<8.1868 2.9158 7.4996>, rot=<0.0041 -1 0 0>}
 dyn_HairSpike_B_01_r {parent=base_HairSpike_B_01_r, loc=<0 0 0>, rot=<0 0 0 1>}
 base_HairSpike_C_01_r {parent=head, loc=<6.212 7.0495 6.1391>, rot=<0.0041 -1 0 0>}
 dyn_HairSpike_C_01_r {parent=base_HairSpike_C_01_r, loc=<0 0 0>, rot=<0 0 0 1>}
 base_HairSpike_D_01_r {parent=head, loc=<5.7791 9.1564 4.9875>, rot=<0.0041 -1 0 0>}
 dyn_HairSpike_D_01_r {parent=base_HairSpike_D_01_r, loc=<0 0 0>, rot=<0 0 0 1>}
 base_HairSpike_E_01_r {parent=head, loc=<8.3614 10.0474 3.6366>, rot=<0.0041 -1 0 0>}
 dyn_HairSpike_E_01_r {parent=base_HairSpike_E_01_r, loc=<0 0 0>, rot=<0 0 0 1>}
 base_HairSpike_G_01_r {parent=head, loc=<15.7106 2.1727 2.9611>, rot=<0.0041 -1 0 0>}
 dyn_HairSpike_G_01_r {parent=base_HairSpike_G_01_r, loc=<0 0 0>, rot=<0 0 0 1>}
 base_hair_back_01 {parent=head, loc=<2.5256 -2.4689 -9.5681>, rot=<0.2165 -0.4721 0.0815 -0.8506>}
 dyn_hair_back_01 {parent=base_hair_back_01, loc=<0 0 0>, rot=<0 0 0 1>}
 base_hair_back_02 {parent=head, loc=<-0.9334 -3.9817 -8.794>, rot=<0.1124 -0.5631 -0.1037 -0.8121>}
 dyn_hair_back_02 {parent=base_hair_back_02, loc=<0 0 0>, rot=<0 0 0 1>}
 base_hair_back_03 {parent=head, loc=<-4.0775 -2.167 -8.4862>, rot=<0.1796 -0.6892 0.3038 -0.6328>}
 dyn_hair_back_03 {parent=base_hair_back_03, loc=<0 0 0>, rot=<0 0 0 1>}
 base_hair_back_04 {parent=head, loc=<-4.8 -5.8375 -6.967>, rot=<0.1958 -0.6671 0.1944 -0.692>}
 dyn_hair_back_04 {parent=base_hair_back_04, loc=<0 0 0>, rot=<0 0 0 1>}
 base_hair_back_05 {parent=head, loc=<4.9269 -7.2765 -5.2449>, rot=<0.4781 -0.5256 0.3614 -0.6038>}
 dyn_hair_back_05 {parent=base_hair_back_05, loc=<0 0 0>, rot=<0 0 0 1>}
 base_hair_back_06 {parent=head, loc=<0.8901 -7.7521 -6.3243>, rot=<0.3451 -0.683 0.1502 -0.626>}
 dyn_hair_back_06 {parent=base_hair_back_06, loc=<0 0 0>, rot=<0 0 0 1>}
 base_hair_back_07 {parent=head, loc=<-4.1951 -7.7694 -2.3913>, rot=<0.2003 -0.7561 0.4162 -0.4636>}
 dyn_hair_back_07 {parent=base_hair_back_07, loc=<0 0 0>, rot=<0 0 0 1>}
 base_hair_back_08 {parent=head, loc=<-1.5209 -8.513 -5.7025>, rot=<0.2228 -0.5753 0.2518 -0.7457>}
 dyn_hair_back_08 {parent=base_hair_back_08, loc=<0 0 0>, rot=<0 0 0 1>}
 base_hair_back_09 {parent=head, loc=<-1.992 -9.7902 -1.2837>, rot=<0.3717 -0.5796 0.3447 -0.638>}
 dyn_hair_back_09 {parent=base_hair_back_09, loc=<0 0 0>, rot=<0 0 0 1>}
 base_hair_back_10 {parent=head, loc=<-2.2882 -9.3973 2.4334>, rot=<0.5689 -0.3867 0.589 -0.4241>}
 dyn_hair_back_10 {parent=base_hair_back_10, loc=<0 0 0>, rot=<0 0 0 1>}
 base_hair_back_11 {parent=head, loc=<1.338 -10.3157 2.2132>, rot=<0.6511 -0.3906 0.5224 -0.388>}
 dyn_hair_back_11 {parent=base_hair_back_11, loc=<0 0 0>, rot=<0 0 0 1>}
 base_hair_back_12 {parent=head, loc=<-1.3085 -8.0283 5.9286>, rot=<0.687 -0.2997 0.5517 -0.366>}
 dyn_hair_back_12 {parent=base_hair_back_12, loc=<0 0 0>, rot=<0 0 0 1>}
 base_hair_back_13 {parent=head, loc=<-3.6263 -9.8413 4.359>, rot=<0.7511 -0.1815 0.4588 -0.4386>}
 dyn_hair_back_13 {parent=base_hair_back_13, loc=<0 0 0>, rot=<0 0 0 1>}
 base_hair_back_14 {parent=head, loc=<-4.3382 -4.041 7.8882>, rot=<0.7081 -0.1208 0.6892 -0.0946>}
 dyn_hair_back_14 {parent=base_hair_back_14, loc=<0 0 0>, rot=<0 0 0 1>}
 base_hair_back_15 {parent=head, loc=<-4.8102 -1.3637 7.6811>, rot=<0.7065 0.0069 0.7075 -0.0153>}
 dyn_hair_back_15 {parent=base_hair_back_15, loc=<0 0 0>, rot=<0 0 0 1>}
 base_hair_back_16 {parent=head, loc=<-0.4323 -4.0032 9.4716>, rot=<0.8191 0.1102 0.5606 -0.0506>}
 dyn_hair_back_16 {parent=base_hair_back_16, loc=<0 0 0>, rot=<0 0 0 1>}
 base_hair_back_17 {parent=head, loc=<2.0354 -2.4983 9.9434>, rot=<0.8483 -0.0221 0.5174 -0.1104>}
 dyn_hair_back_17 {parent=base_hair_back_17, loc=<0 0 0>, rot=<0 0 0 1>}
 base_hair_back_18 {parent=head, loc=<5.1142 -0.1204 9.5347>, rot=<0.8082 -0.0145 0.5077 -0.2981>}
 dyn_hair_back_18 {parent=base_hair_back_18, loc=<0 0 0>, rot=<0 0 0 1>}
 base_hair_back_19 {parent=head, loc=<4.8961 -9.3994 3.5997>, rot=<0.5427 -0.3952 0.6616 -0.3341>}
 dyn_hair_back_19 {parent=base_hair_back_19, loc=<0 0 0>, rot=<0 0 0 1>}
 base_hair_back_20 {parent=head, loc=<2.1551 -8.4085 7.1647>, rot=<0.6996 -0.187 0.6073 -0.3269>}
 dyn_hair_back_20 {parent=base_hair_back_20, loc=<0 0 0>, rot=<0 0 0 1>}
 base_hair_back_21 {parent=head, loc=<5.751 -7.9395 6.4144>, rot=<0.6421 -0.3186 0.5298 -0.4533>}
 dyn_hair_back_21 {parent=base_hair_back_21, loc=<0 0 0>, rot=<0 0 0 1>}
 head_null {parent=neck_01, type=Null, loc=<9.9266 -0.5431 0>, rot=<0 0 -0.1649 0.9863>}
 neck_01_null {parent=spine_05, type=Null, loc=<11.3138 0.1529 0>, rot=<0 0 -0.2003 -0.9797>}
 spine_05_ctrl {controlType=EULER_TRANSFORM, type=Control, loc=<0 0 0>, rot=<0 0 0 1>}
 neck_01_ctrl {controlType=EULER_TRANSFORM, parent=spine_05_ctrl, type=Control, loc=<0 0 0>, rot=<0 0 0 1>}
 neck_02_ctrl {controlType=EULER_TRANSFORM, parent=neck_01_ctrl, type=Control, loc=<0 0 0>, rot=<0 0 0 1>}
 head_ctrl {controlType=EULER_TRANSFORM, parent=neck_02_ctrl, type=Control, loc=<0 0 0>, rot=<0 0 0 1>}
]
